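# Material function that composes the Safflower season UI background: gradient sky, clouds, lightning, amusement park silhouettes and paper-textured mountains.
# Layers are placed by parameterized UV transforms and cut with anti-aliased distance-field masks, with offset re-samples adding cloud and ride shadows and highlights.
# MPC_Safflower_BackgroundStates supplies cloud-height, ground-height and hide-ground animation values.
# Vignettes, blue overlay, color-dodged procedural noise and top-of-screen darkening finish the composite.
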MaterialFunction /Safflower_SeasonData/UI/MaterialFunctions/MF_UI_Safflower_Background
graph {nodes [
  Constant()
  Constant() {r=1}
  n1 = NamedRerouteUsage()
  n2 = VectorParameter() {group=Color, parameterName=BlueLightColor, defaultValue=<rgba 0.0267 0.2174 0.4479 1>}
  n3 = NamedRerouteUsage()
  n4 = NamedRerouteUsage()
  n5 = VectorParameter() {parameterName=VignetteColor, defaultValue=<rgba 0.0052 0.0045 0.0052 1>}
  n6 = NamedRerouteUsage()
  n7 = ScalarParameter() {defaultValue=0.5, group=Park Rides, parameterName=AmusementShadowOpacity}
  n8 = NamedRerouteDeclaration(None=VectorParameter() {parameterName=CloudHighlightColor, defaultValue=<rgba 0.7605 0.6376 1 1>}.RGB) {name=Cloud Highlight Color, nodeColor=<rgba 0 0 0 1>}
  n9 = NamedRerouteDeclaration(None=VectorParameter() {group=Color, parameterName=CloudShadowColor, defaultValue=<rgba 0 0 0 1>}.RGB) {name=Cloud Shadow Color, nodeColor=<rgba 0 0 0 1>}
  n10 = ScalarParameter() {defaultValue=0.3, group=Clouds, parameterName=CloudShadowOpacity}
  n11 = NamedRerouteUsage()
  n12 = Add(A=@n11, B=Multiply(A=@n11, B=NamedRerouteUsage()))
  n13 = VectorParameter() {group=Color, parameterName=CloudColor, defaultValue=<rgba 0.1946 0.0437 0.227 1>}
  n14 = Add(A=@n13.RGB, B=Multiply(A=@n13.RGB, B=NamedRerouteUsage()))
  n15 = NamedRerouteUsage()
  n16 = Multiply(A=VectorParameter() {group=Color, parameterName=CloudOverlayColor, defaultValue=<rgba 0.0437 0.0065 0.0887 1>}.RGB, B=@n15)
  n17 = LinearInterpolate(
   A=LinearInterpolate(A=@n12, B=LinearInterpolate(A=@n12, B=@n9, Alpha=@n10), Alpha=NamedRerouteUsage()) {constB=0}
   B=LinearInterpolate(
    A=@n14
    B=LinearInterpolate(A=@n14, B=@n16, Alpha=@n15)
    Alpha=ScalarParameter() {defaultValue=0.8, group=Clouds, parameterName=CloudOverlayAlpha, sliderMax=1}
   )
   Alpha=NamedRerouteUsage()
  )
  n18 = Subtract(A=NamedRerouteUsage(), B=NamedRerouteUsage())
  n19 = LinearInterpolate(
   A=@n17
   B=LinearInterpolate(
    A=@n17
    B=MaterialFunctionCall!(Blend=@n8, Base=@n17) {materialFunction=/Engine/Functions/Engine_MaterialFunctions03/Blends/Blend_Screen}.Result
    Alpha=ScalarParameter() {defaultValue=0.15, group=Clouds, parameterName=CloudHighlightOpacity}
   )
   Alpha=LinearInterpolate(B=Saturate(None=@n18), Alpha=NamedRerouteUsage())
  )
  n20 = LinearInterpolate(
   A=@n19
   B=MaterialFunctionCall(Blend=VectorParameter() {parameterName=LightingGlowColor, defaultValue=<rgba 0.7157 0.0097 1 1>}.RGB, Base=@n19) {materialFunction=&Blend_Screen}.Result
   Alpha=ScalarParameter() {defaultValue=0.5, group=Lightning, parameterName=LightningGlowOpacity}
  )
  n21 = NamedRerouteUsage()
  n22 = NamedRerouteUsage()
  n23 = LinearInterpolate(
   A=@n19
   B=LinearInterpolate(
    A=LinearInterpolate(A=@n19, B=@n20, Alpha=@n21)
    B=VectorParameter() {parameterName=LightningWhiteColor, defaultValue=<rgba 1 1 1 1>}.RGB
    Alpha=Saturate(None=@n22)
   )
   Alpha=OneMinus(None=NamedRerouteUsage())
  ) {constAlpha=0}
  n24 = LinearInterpolate(
   A=LinearInterpolate(A=@n23, B=LinearInterpolate(A=@n23, B=@n6, Alpha=@n7), Alpha=NamedRerouteUsage())
   B=VectorParameter() {group=Color, parameterName=AmusementColor, defaultValue=<rgba 0.0781 0.0216 0.0548 1>}.RGB
   Alpha=NamedRerouteUsage()
  )
  n25 = MaterialFunctionCall(Blend=NamedRerouteUsage(), Base=@n24) {materialFunction=&Blend_Screen}
  n26 = ScalarParameter() {defaultValue=0.2, group=Park Rides, parameterName=AmusementHighlightOpacity, sliderMax=1}
  n27 = LinearInterpolate(
   A=LinearInterpolate(A=@n24, B=LinearInterpolate(A=@n24, B=@n25.Result, Alpha=@n26), Alpha=NamedRerouteUsage())
   B=NamedRerouteUsage()
   Alpha=NamedRerouteUsage()
  )
  n28 = LinearInterpolate(
   A=@n27
   B=MaterialFunctionCall(Base=@n27, Blend=NamedRerouteUsage()) {materialFunction=&Blend_Overlay}.Result
   Alpha=NamedRerouteUsage()
  )
  n29 = ScalarParameter() {defaultValue=1, group=Mountains, parameterName=MountainPaperTexture_Opacity}
  n30 = LinearInterpolate(
   A=@n23
   B=LinearInterpolate(A=LinearInterpolate(A=@n27, B=@n28, Alpha=@n29), B=NamedRerouteUsage(), Alpha=NamedRerouteUsage())
   Alpha=OneMinus(None=NamedRerouteUsage())
  )
  n31 = LinearInterpolate(
   A=@n30
   B=LinearInterpolate(
    A=@n30
    B=NamedRerouteDeclaration(None=@n5.RGB) {name=Vignette Color, nodeColor=<rgba 0 0 0 1>}
    Alpha=ScalarParameter() {defaultValue=1.2, group=Vignettes, parameterName=RadVigGrad_BaseOpacity}
   )
   Alpha=NamedRerouteUsage()
  )
  n32 = NamedRerouteUsage()
  n33 = LinearInterpolate(
   A=@n31
   B=LinearInterpolate(A=@n31, B=Multiply(A=@n31, B=@n32), Alpha=ScalarParameter() {defaultValue=1, group=Vignettes, parameterName=RadVigGrad_MultOpacity})
   Alpha=NamedRerouteUsage()
  )
  n34 = LinearInterpolate(
   A=@n33
   B=LinearInterpolate(A=@n33, B=Multiply(A=@n33, B=@n4), Alpha=ScalarParameter() {defaultValue=0.9, group=Vignettes, parameterName=BtmLinVig_MultOpacity})
   Alpha=NamedRerouteUsage()
  )
  n35 = MaterialFunctionCall!(
   Blend=Multiply(A=NamedRerouteUsage(), B=ScalarParameter() {defaultValue=0.4, group=Noise, parameterName=Noise_Opacity})
   Base=LinearInterpolate(
    A=@n34
    B=LinearInterpolate(A=@n34, B=@n2.RGB, Alpha=@n3)
    Alpha=ScalarParameter() {defaultValue=0.9, group=Vignettes, parameterName=BlueOverlay_Alpha}
   )
  ) {materialFunction=/Engine/Functions/Engine_MaterialFunctions03/Blends/Blend_ColorDodge}
  FunctionOutput(None=LinearInterpolate(
    A=@n35.Result
    B=LinearInterpolate(A=@n35.Result, Alpha=@n1) {constB=0}
    Alpha=ScalarParameter() {defaultValue=0.5, group=Vignettes, parameterName=TopScreenGradient_Opacity}
   ))
  n36 = Constant() {r=1}
  n37 = Constant() {r=0.5}
  n38 = TextureSample(UVs=MaterialFunctionCall(UVs=NamedRerouteUsage(), Texture Scale=AppendVector(A=@n36, B=@n37)) {materialFunction=&ScaleUVsByCenter}.UVs) {samplerType=SAMPLERTYPE_Alpha, texture=/Safflower_SeasonData/UI/Textures/T_UI_Safflower_MountainPaper}
  n39 = NamedRerouteUsage()
  LinearInterpolate(B=@n38.R, Alpha=@n39)
  n40 = Add(
   A=AppendVector(
    A=ScalarParameter() {defaultValue=0.92, group=Mountains, parameterName=MountXTransform, sliderMax=1, sliderMin=-1}
    B=ScalarParameter() {defaultValue=-0.1, group=Mountains, parameterName=MountYTransform, sliderMax=1, sliderMin=-1}
   )
   B=Add(A=NamedRerouteUsage(), B=NamedRerouteUsage())
  )
  n41 = AppendVector(
   A=ScalarParameter() {defaultValue=1, group=Mountains, parameterName=MountXScale, sliderMax=2}
   B=ScalarParameter() {defaultValue=0.16, group=Mountains, parameterName=MountYScale, sliderMax=2}
  )
  n42 = ScalarParameter() {defaultValue=1.825, group=Mountains, parameterName=MountOverallScale, sliderMax=4, sliderMin=1}
  n43 = TextureSample(UVs=MaterialFunctionCall(UVs=NamedRerouteDeclaration(None=@n40) {name=Mountain Transform, nodeColor=<rgba 0 0 0 1>}, Texture Scale=Multiply(A=@n41, B=@n42)) {materialFunction=&ScaleUVsByCenter}.UVs) {texture=/Safflower_SeasonData/UI/Textures/T_UI_Safflower_Mount}
  LinearInterpolate(
   A=@n43.RGB
   B=Add(A=@n43.RGB, B=VectorParameter() {group=Color, parameterName=HilltopColorCorrection, defaultValue=<rgba 0.6979 0.6979 0.6979 1>}.RGB)
   Alpha=@n43.RGB
  )
  n44 = NamedRerouteUsage()
  n45 = ScalarParameter() {defaultValue=1, group=Background, parameterName=SkyPaperTextureX}
  n46 = ScalarParameter() {defaultValue=1, group=Background, parameterName=SkyPaperTextureY}
  MaterialFunctionCall(
   UVs=@n44
   Texture Scale=Multiply(A=AppendVector(A=@n45, B=@n46), B=ScalarParameter() {defaultValue=1, group=Background, parameterName=SkyPaperTextureScale})
  ) {materialFunction=&ScaleUVsByCenter}
  n47 = MaterialFunctionCall!(UV=NamedRerouteUsage(), Seed) {materialFunction=&MF_UI_Random}
  Multiply(A=@n47.Result, B=ScalarParameter() {defaultValue=0.4, group=Noise, parameterName=Noise_Opacity, sliderMax=1}) {constB=0.65}
  NamedRerouteDeclaration(None=MaterialFunctionCall!(Use Local UV, Texture Size, Scale Pivot Point) {materialFunction=/Game/Environments/FrontEnd/Blueprints/ItemPreview/VaultWorldTransitions/MaterialFunctions/MF_UI_VaultWorld_UV}.Scale Y) {name=ScaleY, nodeColor=<rgba 0 0 0 1>}
  NamedRerouteDeclaration(None=Saturate(None=CollectionParameter() {collection=/Safflower_SeasonData/UI/Data/MPC_Safflower_BackgroundStates, parameterName=Anim_GroundHeight})) {name=Ground Height Anim, nodeColor=<rgba 0 0 0 1>}
  NamedRerouteDeclaration(None=Saturate(None=CollectionParameter() {collection=&MPC_Safflower_BackgroundStates, parameterName=Anim_HideGround})) {name=Hide Ground Anim, nodeColor=<rgba 0 0 0 1>}
  NamedRerouteDeclaration(None=MaterialFunctionCall(DistanceField=@n43.A, Threshold=Constant() {r=0.5}, Smoothing=Constant()) {materialFunction=&MF_UI_AntiAliasing}.Outer) {name=MountainsMask, nodeColor=<rgba 0 0 0 1>}
  n48 = NamedRerouteUsage()
  n49 = ScalarParameter() {group=Clouds, parameterName=CloudOverlayGradOffset}
  NamedRerouteDeclaration(None=MaterialFunctionCall(
    UVs=NamedRerouteUsage()
    GradientStart=ScalarParameter() {defaultValue=0.8, group=Clouds, parameterName=CloudOverlayGradStart}
    GradientEnd=ScalarParameter() {defaultValue=1, group=Clouds, parameterName=CloudOverlayGradEnd}
    RotationAngle=Constant() {r=0.25}
    GradientOffset=Add(A=@n48, B=@n49)
   ) {materialFunction=&MF_UI_LinearGradient}.Result) {name=Cloud Overlay Gradient, nodeColor=<rgba 0 0 0 1>}
  n50 = ScalarParameter() {group=AnimatedStates, parameterName=GroundHeightNormal, sliderMin=-0.25}
  n51 = ScalarParameter() {defaultValue=-0.25, group=AnimatedStates, parameterName=GroundHeightMoved, sliderMin=-0.25}
  n52 = NamedRerouteUsage()
  NamedRerouteDeclaration(None=AppendVector(A=Constant(), B=LinearInterpolate(A=@n50, B=@n51, Alpha=@n52))) {name=Horizon Height, nodeColor=<rgba 0 0 0 1>}
  n53 = AppendVector(A=Constant(), B=ScalarParameter() {defaultValue=0.37, group=Clouds, parameterName=CloudYTransform, sliderMax=1, sliderMin=-1})
  n54 = Add(A=NamedRerouteUsage(), B=NamedRerouteUsage())
  n55 = AppendVector(
   A=ScalarParameter() {defaultValue=1, group=Clouds, parameterName=CloudXScale, sliderMax=2}
   B=ScalarParameter() {defaultValue=0.2, group=Clouds, parameterName=CloudYScale, sliderMax=2}
  )
  n56 = ScalarParameter() {defaultValue=2.15, group=Clouds, parameterName=CloudOverallScale, sliderMax=4, sliderMin=1}
  n57 = MaterialFunctionCall(
   UVs=NamedRerouteDeclaration(None=Add(A=@n53, B=@n54)) {name=Cloud Transform, nodeColor=<rgba 0 0 0 1>}
   Texture Scale=NamedRerouteDeclaration(None=Multiply(A=@n55, B=@n56)) {name=Cloud Scale, nodeColor=<rgba 0 0 0 1>}
  ) {materialFunction=&ScaleUVsByCenter}
  NamedRerouteDeclaration(None=MaterialFunctionCall(
    DistanceField=TextureSample(UVs=@n57.UVs) {samplerType=SAMPLERTYPE_Alpha, texture=/Safflower_SeasonData/UI/Textures/T_UI_Safflower_Clouds}.R
    Threshold=Constant() {r=0.5}
    Smoothing=Constant()
   ) {materialFunction=&MF_UI_AntiAliasing}.Inner) {name=Cloud Mask, nodeColor=<rgba 0 0 0 1>}
  n58 = ScalarParameter() {group=AnimatedStates, parameterName=CloudHeightNormal, sliderMax=0.15}
  n59 = ScalarParameter() {defaultValue=0.15, group=AnimatedStates, parameterName=CloudHeightMoved, sliderMax=0.15}
  n60 = NamedRerouteUsage()
  NamedRerouteDeclaration(None=AppendVector(A=Constant(), B=LinearInterpolate(A=@n58, B=@n59, Alpha=@n60))) {name=Cloud Height, nodeColor=<rgba 0 0 0 1>}
  n61 = Constant()
  n62 = ScalarParameter() {defaultValue=-0.003, group=Clouds, parameterName=CloudShadowYAdd}
  n63 = MaterialFunctionCall(UVs=Add(A=NamedRerouteUsage(), B=AppendVector(A=@n61, B=@n62)), Texture Scale=NamedRerouteUsage()) {materialFunction=&ScaleUVsByCenter}
  NamedRerouteDeclaration(None=MaterialFunctionCall(
    DistanceField=TextureSample(UVs=@n63.UVs) {samplerType=SAMPLERTYPE_Alpha, texture=&T_UI_Safflower_Clouds}.R
    Threshold=Constant() {r=0.5}
    Smoothing=Constant()
   ) {materialFunction=&MF_UI_AntiAliasing}.Inner) {name=Cloud Shadow, nodeColor=<rgba 0 0 0 1>}
  n64 = VectorParameter() {parameterName=BkgTopGrad, defaultValue=<rgba 0.062 0.0138 0.0677 1>}
  n65 = VectorParameter() {parameterName=BkgBtmGrad, defaultValue=<rgba 0.6042 0.2077 0.4509 1>}
  n66 = MaterialFunctionCall("Value (0-1)"=MaterialFunctionCall(UVs=NamedRerouteUsage(), GradientStart=Constant() {r=0.1}, GradientEnd=Constant() {r=0.8}, RotationAngle=Constant() {r=0.75}) {materialFunction=&MF_UI_LinearGradient}.Result) {materialFunction=&MF_UI_EaseCurves}
  NamedRerouteDeclaration(None=MaterialFunctionCall!(
    Base=LinearInterpolate(A=@n64.RGB, B=@n65.RGB, Alpha=@n66.EaseInOutSine)
    Hue=Constant()
    Saturation=ScalarParameter() {defaultValue=1.25, group=Background, parameterName=BackgroundSaturation, sliderMax=2, sliderMin=1}
    Value
    Effect Amount
   ) {materialFunction=/BRCosmetics/Weapons/FORT_Melee/Pickaxe_QuailWink/MF_DM_Effect_Base_HSV}.Result) {name=Background Gradient, nodeColor=<rgba 0 0 0 1>}
  n67 = Add(
   A=NamedRerouteUsage()
   B=AppendVector(A=Constant(), B=ScalarParameter() {defaultValue=0.001, group=Clouds, parameterName=CloudHighlightYAdd})
  )
  n68 = NamedRerouteUsage()
  n69 = MaterialFunctionCall(
   DistanceField=TextureSample(UVs=MaterialFunctionCall(UVs=@n67, Texture Scale=@n68) {materialFunction=&ScaleUVsByCenter}.UVs) {samplerType=SAMPLERTYPE_Alpha, texture=&T_UI_Safflower_Clouds}.R
   Threshold=Constant() {r=0.5}
   Smoothing=Constant()
  ) {materialFunction=&MF_UI_AntiAliasing}
  n70 = NamedRerouteUsage()
  NamedRerouteDeclaration(None=Saturate(None=Subtract(A=@n69.Inner, B=@n70))) {name=Cloud Highlight, nodeColor=<rgba 0 0 0 1>}
  n71 = NamedRerouteUsage()
  n72 = AppendVector(
   A=ScalarParameter() {defaultValue=-0.0125, group=Lightning, parameterName=LightningXTransform, sliderMax=1, sliderMin=-1}
   B=ScalarParameter() {defaultValue=0.3168, group=Lightning, parameterName=LightningYTransform, sliderMax=1, sliderMin=-1}
  )
  n73 = AppendVector(
   A=ScalarParameter() {defaultValue=0.65, group=Lightning, parameterName=LightningXScale, sliderMax=2}
   B=ScalarParameter() {defaultValue=1.4, group=Lightning, parameterName=LightningYScale, sliderMax=2}
  )
  n74 = ScalarParameter() {defaultValue=0.4, group=Lightning, parameterName=LightningOverallScale, sliderMax=1}
  n75 = TextureSample(UVs=MaterialFunctionCall(UVs=Add(A=@n71, B=@n72), Texture Scale=Multiply(A=@n73, B=@n74)) {materialFunction=&ScaleUVsByCenter}.UVs) {samplerType=SAMPLERTYPE_Alpha, texture=/Safflower_SeasonData/UI/Textures/T_UI_Safflower_Lightning_sdf}
  NamedRerouteDeclaration(None=MaterialFunctionCall(DistanceField=@n75.R, Threshold=Constant() {r=0.5}, Smoothing=Constant()) {materialFunction=&MF_UI_AntiAliasing}.Outer) {name=Lightning Mask, nodeColor=<rgba 0 0 0 1>}
  n76 = MaterialFunctionCall(DistanceField=@n75.R, Threshold=Constant() {r=0.6}, Smoothing=Constant() {r=1}) {materialFunction=&MF_UI_AntiAliasing}
  NamedRerouteDeclaration(None=@n76.Outer) {name=Lightning Glow Mask, nodeColor=<rgba 0 0 0 1>}
  n77 = NamedRerouteUsage()
  n78 = Multiply(
   A=AppendVector(A=ScalarParameter() {defaultValue=1.5, group=Vignettes, parameterName=RadVigGrad_XScale}, B=Constant() {r=1})
   B=ScalarParameter() {defaultValue=4, group=Vignettes, parameterName=RadVigGrad_OverallScale}
  )
  NamedRerouteDeclaration(None=MaterialFunctionCall(
    UV=MaterialFunctionCall(UVs=@n77, Texture Scale=@n78) {materialFunction=&ScaleUVsByCenter}.UVs
    GradientStart=Constant() {r=1}
    GradientEnd=Constant()
   ) {materialFunction=&MF_UI_RadialGradient}.Linear) {name=Radial Vignette Mask, nodeColor=<rgba 0 0 0 1>}
  NamedRerouteDeclaration(None=MaterialFunctionCall(
    UVs=NamedRerouteUsage()
    GradientStart=ScalarParameter() {group=Vignettes, parameterName=BtmLinVig_GradStart}
    GradientEnd=Constant() {r=1}
    RotationAngle=Constant() {r=0.75}
   ) {materialFunction=&MF_UI_LinearGradient}.Result) {name=Bottom Linear Gradient Mask, nodeColor=<rgba 0 0 0 1>}
  n79 = ScalarParameter() {defaultValue=3, group=Vignettes, parameterName=BlueOverlay_XScale}
  n80 = Constant() {r=1}
  n81 = MaterialFunctionCall(
   UVs=NamedRerouteUsage()
   Texture Scale=Multiply(A=AppendVector(A=@n79, B=@n80), B=ScalarParameter() {defaultValue=3, group=Vignettes, parameterName=BlueOverlay_OverallScale})
  ) {materialFunction=&ScaleUVsByCenter}
  n82 = Constant() {r=1}
  n83 = Constant()
  n84 = NamedRerouteUsage()
  n85 = Constant() {r=0.55}
  n86 = Constant() {r=0.25}
  NamedRerouteDeclaration(None=LinearInterpolate(
    B=MaterialFunctionCall(UV=@n81.UVs, GradientStart=@n82, GradientEnd=@n83) {materialFunction=&MF_UI_RadialGradient}.Linear
    Alpha=MaterialFunctionCall(UVs=@n84, GradientStart=@n85, RotationAngle=@n86) {materialFunction=&MF_UI_LinearGradient}.Result
   )) {name=Blue Overlay Mask, nodeColor=<rgba 0 0 0 1>}
  n87 = ScalarParameter() {defaultValue=0.5, group=Noise, parameterName=Noise_FoldsOpacity, sliderMax=1}
  n88 = NamedRerouteUsage()
  n89 = Sine(None=Multiply(A=ComponentMask(None=@n88) {r=true}, B=ScalarParameter() {defaultValue=2.5, group=Noise, parameterName=Noise_FoldFrequency, sliderMax=10}) {constB=2})
  n90 = Power(Base=OneMinus(None=ComponentMask(None=@n88) {g=true}))
  n91 = Multiply(A=Power(Base=Abs(None=@n89)) {constExponent=1.2}, B=@n90)
  n92 = LinearInterpolate(Alpha=@n90) {constA=0.1}
  NamedRerouteDeclaration(None=Max(A=Multiply(A=@n87, B=@n91) {constB=0.5}, B=Multiply(A=@n47.Result, B=@n92))) {name=Noise Overlay, nodeColor=<rgba 0 0 0 1>}
  NamedRerouteDeclaration(None=TextureSample(UVs=@n44) {samplerType=SAMPLERTYPE_Alpha, texture=/Safflower_SeasonData/UI/Textures/T_UI_Safflower_PaperTexture}.R) {name=SkyPaperTexture, nodeColor=<rgba 0 0 0 1>}
  n93 = ScalarParameter() {defaultValue=0.015, group=Park Rides, parameterName=GateShadowX, sliderMax=1, sliderMin=-1}
  n94 = ScalarParameter() {defaultValue=-0.015, group=Park Rides, parameterName=GateShadowY, sliderMax=1, sliderMin=-1}
  n95 = TextureSample(UVs=Add(A=NamedRerouteUsage(), B=AppendVector(A=@n93, B=@n94))) {samplerType=SAMPLERTYPE_Alpha, texture=/Safflower_SeasonData/UI/Textures/T_UI_Safflower_Sign}
  n96 = NamedRerouteDeclaration(None=MaterialFunctionCall(
    DistanceField=OneMinus(None=@n95.R)
    Threshold=Constant() {r=0.5}
    Smoothing=ScalarParameter() {defaultValue=0.1, group=Park Rides, parameterName=AmusementShadowSoftness, sliderMax=1}
   ) {materialFunction=&MF_UI_AntiAliasing}.Inner) {name=Gate Shadow, nodeColor=<rgba 0.0094 0.0066 0 1>}
  n97 = ScalarParameter() {defaultValue=0.015, group=Park Rides, parameterName=WheelShadowX, sliderMax=1, sliderMin=-1}
  n98 = ScalarParameter() {defaultValue=-0.015, group=Park Rides, parameterName=WheelShadowY, sliderMax=1, sliderMin=-1}
  n99 = TextureSample(UVs=Add(A=NamedRerouteUsage(), B=AppendVector(A=@n97, B=@n98))) {samplerType=SAMPLERTYPE_Alpha, texture=/Safflower_SeasonData/UI/Textures/T_UI_Safflower_Wheel}
  n100 = NamedRerouteDeclaration(None=MaterialFunctionCall(
    DistanceField=OneMinus(None=@n99.R)
    Threshold=Constant() {r=0.5}
    Smoothing=ScalarParameter() {defaultValue=0.1, group=Park Rides, parameterName=AmusementShadowSoftness, sliderMax=1}
   ) {materialFunction=&MF_UI_AntiAliasing}.Inner) {name=Wheel Shadow, nodeColor=<rgba 0.0094 0.0066 0 1>}
  n101 = ScalarParameter() {defaultValue=0.015, group=Park Rides, parameterName=CoasterShadowX, sliderMax=1, sliderMin=-1}
  n102 = ScalarParameter() {defaultValue=-0.015, group=Park Rides, parameterName=CoasterShadowY, sliderMax=1, sliderMin=-1}
  n103 = TextureSample(UVs=Add(A=NamedRerouteUsage(), B=AppendVector(A=@n101, B=@n102))) {samplerType=SAMPLERTYPE_Alpha, texture=/Safflower_SeasonData/UI/Textures/T_UI_Safflower_Coaster}
  n104 = NamedRerouteDeclaration(None=MaterialFunctionCall(
    DistanceField=OneMinus(None=@n103.R)
    Threshold=Constant() {r=0.5}
    Smoothing=ScalarParameter() {defaultValue=0.1, group=Park Rides, parameterName=AmusementShadowSoftness, sliderMax=1}
   ) {materialFunction=&MF_UI_AntiAliasing}.Inner) {name=Coaster Shadow, nodeColor=<rgba 0.0094 0.0066 0 1>}
  NamedRerouteDeclaration(None=LinearInterpolate(A=LinearInterpolate(A=@n96, B=@n100, Alpha=@n100), B=@n104, Alpha=@n104)) {name=Amusement Park Shadows, nodeColor=<rgba 0 0 0 1>}
  n105 = Add(
   A=NamedRerouteUsage()
   B=AppendVector(
    A=ScalarParameter() {defaultValue=0.744, group=Park Rides, parameterName=CoasterHighlightMask_X}
    B=ScalarParameter() {defaultValue=-0.024, group=Park Rides, parameterName=CoasterHighlightMask_Y}
   )
  )
  n106 = ScalarParameter() {defaultValue=0.1, group=Park Rides, parameterName=CoasterHighlightMask_Scale}
  NamedRerouteDeclaration(None=MaterialFunctionCall(
    UVs=MaterialFunctionCall(UVs=@n105, Texture Scale=@n106) {materialFunction=&ScaleUVsByCenter}.UVs
    GlowMin=ScalarParameter() {group=Park Rides, parameterName=CoasterHighlightMask_GlowMin}
    GlowMax=ScalarParameter() {defaultValue=1, group=Park Rides, parameterName=CoasterHighlightMask_GlowMax}
   ) {materialFunction=&MF_UI_Circle}.Fill) {name=Coaster Highlight Mask, nodeColor=<rgba 0 0 0 1>}
  n107 = ScalarParameter() {defaultValue=-0.015, group=Park Rides, parameterName=GateHighlightX, sliderMax=1, sliderMin=-1}
  n108 = ScalarParameter() {defaultValue=0.015, group=Park Rides, parameterName=GateHighlightY, sliderMax=1, sliderMin=-1}
  n109 = TextureSample(UVs=Add(A=NamedRerouteUsage(), B=AppendVector(A=@n107, B=@n108))) {samplerType=SAMPLERTYPE_Alpha, texture=&T_UI_Safflower_Sign}
  n110 = LinearInterpolate(
   B=MaterialFunctionCall(
    DistanceField=OneMinus(None=@n109.R)
    Threshold=Constant() {r=0.5}
    Smoothing=ScalarParameter() {defaultValue=0.2, group=Park Rides, parameterName=AmusementHighlightSoftness, sliderMax=1}
   ) {materialFunction=&MF_UI_AntiAliasing}.Inner
   Alpha=NamedRerouteUsage()
  )
  n111 = NamedRerouteUsage()
  n112 = AppendVector(
   A=ScalarParameter() {defaultValue=-0.015, group=Park Rides, parameterName=WheelHighlightX, sliderMax=1, sliderMin=-1}
   B=ScalarParameter() {defaultValue=0.015, group=Park Rides, parameterName=WheelHighlightY, sliderMax=1, sliderMin=-1}
  )
  n113 = OneMinus(None=TextureSample(UVs=Add(A=@n111, B=@n112)) {samplerType=SAMPLERTYPE_Alpha, texture=&T_UI_Safflower_Wheel}.R)
  n114 = Constant() {r=0.5}
  n115 = ScalarParameter() {defaultValue=0.2, group=Park Rides, parameterName=AmusementHighlightSoftness, sliderMax=1}
  n116 = NamedRerouteDeclaration(None=LinearInterpolate(
    B=MaterialFunctionCall(DistanceField=@n113, Threshold=@n114, Smoothing=@n115) {materialFunction=&MF_UI_AntiAliasing}.Inner
    Alpha=NamedRerouteUsage()
   )) {name=Wheel Highlight, nodeColor=<rgba 0 0 0 1>}
  n117 = Add(
   A=NamedRerouteUsage()
   B=AppendVector(
    A=ScalarParameter() {defaultValue=-0.002, group=Park Rides, parameterName=CoasterHighlightX, sliderMax=1, sliderMin=-1}
    B=ScalarParameter() {defaultValue=0.012, group=Park Rides, parameterName=CoasterHighlightY, sliderMax=1, sliderMin=-1}
   )
  )
  n118 = MaterialFunctionCall(
   DistanceField=OneMinus(None=TextureSample(UVs=@n117) {samplerType=SAMPLERTYPE_Alpha, texture=&T_UI_Safflower_Coaster}.R)
   Threshold=Constant() {r=0.5}
   Smoothing=ScalarParameter() {defaultValue=0.2, group=Park Rides, parameterName=AmusementHighlightSoftness, sliderMax=1}
  ) {materialFunction=&MF_UI_AntiAliasing}
  n119 = NamedRerouteUsage()
  n120 = NamedRerouteUsage()
  n121 = NamedRerouteDeclaration(None=LinearInterpolate(B=LinearInterpolate(B=@n118.Inner, Alpha=@n119), Alpha=OneMinus(None=@n120))) {name=Coaster Highlight, nodeColor=<rgba 0 0 0 1>}
  n122 = LinearInterpolate(
   A=LinearInterpolate(A=NamedRerouteDeclaration(None=@n110) {name=Gate Highlight, nodeColor=<rgba 0 0 0 1>}, B=@n116, Alpha=@n116)
   B=@n121
   Alpha=@n121
  )
  n123 = NamedRerouteUsage()
  NamedRerouteDeclaration(None=Saturate(None=Subtract(A=@n122, B=@n123))) {name=Amusement Park Highlights, nodeColor=<rgba 0 0 0 1>}
  NamedRerouteDeclaration(None=MaterialFunctionCall(UVs=NamedRerouteUsage(), GradientStart=Constant() {r=0.85}, RotationAngle=Constant() {r=0.25}) {materialFunction=&MF_UI_LinearGradient}.Result) {name=Top of Screen Gradient, nodeColor=<rgba 0 0 0 1>}
  n124 = ComponentMask(None=Add(A=ScalarParameter() {defaultValue=-0.6, group=Mountains, parameterName=MountainBottom_Height}, B=NamedRerouteUsage()) {constA=-0.7}) {g=true}
  NamedRerouteDeclaration(None=Saturate(None=Ceil(None=@n124))) {name=Mountain Bottom, nodeColor=<rgba 0.0052 0 0.0023 1>}
  NamedRerouteDeclaration(None=VectorParameter() {group=Color, parameterName=MountainColor, defaultValue=<rgba 0.0704 0.008 0.0452 1>}.RGB) {name=Mountain Color, nodeColor=<rgba 0 0 0 1>}
  NamedRerouteDeclaration(None=@n38.R) {name=Mountain Paper Texture, nodeColor=<rgba 0 0 0 1>}
  n125 = Add(A=NamedRerouteUsage(), B=NamedRerouteUsage())
  n126 = Add(
   A=AppendVector(
    A=ScalarParameter() {defaultValue=0.5725, group=Park Rides, parameterName=GateX, sliderMax=1, sliderMin=-1}
    B=ScalarParameter() {defaultValue=-0.11, group=Park Rides, parameterName=GateY, sliderMax=1, sliderMin=-1}
   )
   B=@n125
  )
  n127 = ScalarParameter() {defaultValue=0.13, group=Park Rides, parameterName=GateScale, sliderMax=1}
  n128 = TextureSample(UVs=NamedRerouteDeclaration(None=MaterialFunctionCall(UVs=@n126, Texture Scale=@n127) {materialFunction=&ScaleUVsByCenter}.UVs) {name=Gate transform, nodeColor=<rgba 0 0 0 1>}) {samplerType=SAMPLERTYPE_Alpha, texture=&T_UI_Safflower_Sign}
  n129 = NamedRerouteDeclaration(None=MaterialFunctionCall(DistanceField=OneMinus(None=@n128.R), Threshold=Constant() {r=0.5}, Smoothing=Constant()) {materialFunction=&MF_UI_AntiAliasing}.Inner) {name=Gate, nodeColor=<rgba 0 0 0 1>}
  n130 = Add(
   A=@n125
   B=AppendVector(
    A=ScalarParameter() {defaultValue=0.46, group=Park Rides, parameterName=WheelX, sliderMax=1, sliderMin=-1}
    B=ScalarParameter() {defaultValue=-0.092, group=Park Rides, parameterName=WheelY, sliderMax=1, sliderMin=-1}
   )
  )
  n131 = ScalarParameter() {defaultValue=0.12, group=Park Rides, parameterName=WheelScale, sliderMax=1}
  n132 = TextureSample(UVs=NamedRerouteDeclaration(None=MaterialFunctionCall(UVs=@n130, Texture Scale=@n131) {materialFunction=&ScaleUVsByCenter}.UVs) {name=Wheel Transform, nodeColor=<rgba 0 0 0 1>}) {samplerType=SAMPLERTYPE_Alpha, texture=&T_UI_Safflower_Wheel}
  n133 = NamedRerouteDeclaration(None=MaterialFunctionCall(DistanceField=OneMinus(None=@n132.R), Threshold=Constant() {r=0.5}, Smoothing=Constant()) {materialFunction=&MF_UI_AntiAliasing}.Inner) {name=Wheel, nodeColor=<rgba 0 0 0 1>}
  n134 = Add(
   A=@n125
   B=AppendVector(
    A=ScalarParameter() {defaultValue=0.73, group=Park Rides, parameterName=CoasterX, sliderMax=1, sliderMin=-1}
    B=ScalarParameter() {defaultValue=-0.025, group=Park Rides, parameterName=CoasterY, sliderMax=1, sliderMin=-1}
   )
  )
  n135 = Constant() {r=0.2}
  n136 = TextureSample(UVs=NamedRerouteDeclaration(None=MaterialFunctionCall(UVs=@n134, Texture Scale=@n135) {materialFunction=&ScaleUVsByCenter}.UVs) {name=Coaster Transform, nodeColor=<rgba 0 0 0 1>}) {samplerType=SAMPLERTYPE_Alpha, texture=&T_UI_Safflower_Coaster}
  n137 = NamedRerouteDeclaration(None=MaterialFunctionCall(DistanceField=OneMinus(None=@n136.R), Threshold=Constant() {r=0.5}, Smoothing=Constant()) {materialFunction=&MF_UI_AntiAliasing}.Inner) {name=Coaster, nodeColor=<rgba 0 0 0 1>}
  NamedRerouteDeclaration(None=LinearInterpolate(A=LinearInterpolate(A=@n129, B=@n133, Alpha=@n133), B=@n137, Alpha=@n137)) {name=AmusementParkPieces, nodeColor=<rgba 0 0 0 1>}
  NamedRerouteDeclaration(None=@n43.RGB) {name=Mountains, nodeColor=<rgba 0 0 0 1>}
  NamedRerouteDeclaration(None=Saturate(None=CollectionParameter() {collection=&MPC_Safflower_BackgroundStates, parameterName=Anim_CloudHeight})) {name=Cloud Height Anim, nodeColor=<rgba 0 0 0 1>}
  OneMinus(None=@n39)
  Saturate(None=@n76.Outer)
 ]}
