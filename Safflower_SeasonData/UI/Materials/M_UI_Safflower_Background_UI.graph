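# M_UI_Safflower_Background_UI is a translucent UI-domain material, named as a background for the Safflower season interface.
# Its only node is a call to the MF_UI_Safflower_Background material function, so any visual logic lives in that function.
# The graph shows no parameters and no wiring from the function to the material outputs.
# Automatic usage setting in the editor and static mesh usage are both disabled.
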Material /Safflower_SeasonData/UI/Materials/M_UI_Safflower_Background_UI
graph {
 properties {
  bAutomaticallySetUsageInEditor=false
  bUsedWithStaticMesh=false
  blendMode=BLEND_Translucent
  materialDomain=MD_UI
  usedShadingModels=Default Lit
 }
 nodes [MaterialFunctionCall() {materialFunction=/Safflower_SeasonData/UI/MaterialFunctions/MF_UI_Safflower_Background}]
}
