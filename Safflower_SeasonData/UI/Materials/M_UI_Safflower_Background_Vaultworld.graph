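# Safflower season background material for the front-end item preview Vault World, using Masked blending and Default Lit shading.
# The Safflower background function supplies the color, which passes through the shared Vault World post-process calibration function.
# The calibrated color then feeds the shared Vault World transition function.
# All visual logic lives in these three material functions, so the material itself only chains them.
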Material /Safflower_SeasonData/UI/Materials/M_UI_Safflower_Background_Vaultworld
graph {
 properties {bAutomaticallySetUsageInEditor=false, blendMode=BLEND_Masked, usedShadingModels=Default Lit}
 nodes [
  n1 = MaterialFunctionCall!(Vault World Color=MaterialFunctionCall() {materialFunction=/Safflower_SeasonData/UI/MaterialFunctions/MF_UI_Safflower_Background}.Result) {materialFunction=/Game/Environments/FrontEnd/Blueprints/ItemPreview/VaultWorldTransitions/MaterialFunctions/MF_UI_VaultWorld_PostProcessCalibration}
  MaterialFunctionCall!(Vault World Color=@n1.Calibrated Color) {materialFunction=/Game/Environments/FrontEnd/Blueprints/ItemPreview/VaultWorldTransitions/MaterialFunctions/MF_UI_VaultWorld_Transition}
 ]
}
