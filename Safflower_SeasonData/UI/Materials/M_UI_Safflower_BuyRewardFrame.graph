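# Translucent UI material for the Safflower season's buy-reward frame, drawing a soft-edged frame with a drop shadow.
# Nine-slice, center-scaled UVs sample the frame texture's alpha as a distance field, anti-aliased into a mask with adjustable softness.
# A vertically offset second sample forms the shadow and a faint addition to the combined shape.
# A lightened foil texture and shadow color are declared alongside nested color blends.
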Material /Safflower_SeasonData/UI/Materials/M_UI_Safflower_BuyRewardFrame
graph {
 properties {
  bAutomaticallySetUsageInEditor=false
  bUsedWithStaticMesh=false
  blendMode=BLEND_Translucent
  materialDomain=MD_UI
  usedShadingModels=Default Lit
 }
 nodes [
  LinearInterpolate(B=LinearInterpolate(A=NamedRerouteUsage(), B=NamedRerouteUsage(), Alpha=NamedRerouteUsage()), Alpha=NamedRerouteUsage())
  n1 = NamedRerouteDeclaration(None=MaterialFunctionCall(UVs=MaterialFunctionCall() {materialFunction=&GetUserInterfaceUV}.9-Slice UV) {materialFunction=&ScaleUVsByCenter}.UVs) {name=Frame UV, nodeColor=<rgba 0 0 0 1>}
  NamedRerouteDeclaration(None=MaterialFunctionCall(
    DistanceField=TextureSample(UVs=@n1) {texture=/Safflower_SeasonData/UI/Textures/T_UI_Safflower_BuyRewardFrame}.A
    Threshold=Constant() {r=0.5}
    Smoothing=ScalarParameter() {group=Shape, parameterName=FrameSoftness, sliderMax=1}
   ) {materialFunction=&MF_UI_AntiAliasing}.Outer) {name=Frame Mask, nodeColor=<rgba 0 0 0 1>}
  n2 = MaterialFunctionCall(UVs=TextureCoordinate(), Texture Scale=Constant() {r=3}) {materialFunction=&ScaleUVsByCenter}
  NamedRerouteDeclaration(None=LinearInterpolate(A=TextureSample(UVs=@n2.UVs) {texture=/Safflower_SeasonData/UI/Textures/T_UI_Safflower_FoilTexture}.RGB) {constAlpha=0.2}) {name=Foil Texture, nodeColor=<rgba 0 0 0 1>}
  n3 = NamedRerouteUsage()
  n4 = AppendVector(A=Constant(), B=ScalarParameter() {defaultValue=-0.015, parameterName=FrameY})
  n5 = MaterialFunctionCall(
   DistanceField=TextureSample(UVs=Add(A=@n3, B=@n4)) {texture=&T_UI_Safflower_BuyRewardFrame}.A
   Threshold=Constant() {r=0.5}
   Smoothing=ScalarParameter() {group=Shape, parameterName=FrameSoftness, sliderMax=1}
  ) {materialFunction=&MF_UI_AntiAliasing}
  n6 = NamedRerouteUsage()
  NamedRerouteDeclaration(None=Saturate(None=Subtract(A=@n5.Outer, B=@n6))) {name=Frame Shadow, nodeColor=<rgba 0 0 0 1>}
  n7 = NamedRerouteUsage()
  n8 = LinearInterpolate(B=@n5.Outer) {constAlpha=0.25}
  NamedRerouteDeclaration(None=Saturate(None=Add(A=@n7, B=@n8))) {name=Combined Frame Shape, nodeColor=<rgba 0 0 0 1>}
  NamedRerouteDeclaration(None=VectorParameter() {group=Color, parameterName=ShadowColor}.RGB) {name=Shadow Color, nodeColor=<rgba 0 0 0 1>}
 ]
}
