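# UI material for Safflower season chase reward icons that draws animated, talking berry characters.
# ID picks a tile from a 3x3 atlas and gates a composited overlay, apparently the face, off at zero.
# Flipbooks over a shared mouth-and-eye texture drive timed blinks, offset randomly by ID, and mouth cycling.
# IsTalking switches on mouth animation, eye bobbing and a pulsing vertical squash, with explicit derivatives for tile sampling.
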Material /Safflower_SeasonData/UI/Materials/M_UI_Safflower_ChaseRewardIcons
graph {
 properties {
  bAutomaticallySetUsageInEditor=false
  bUsedWithStaticMesh=false
  blendMode=BLEND_Translucent
  materialDomain=MD_UI
  usedShadingModels=Default Lit
 }
 nodes [
  n1 = NamedRerouteUsage()
  n2 = NamedRerouteUsage()
  n3 = NamedRerouteUsage()
  LinearInterpolate(
   A=@n1
   B=LinearInterpolate(A=LinearInterpolate(A=@n1, B=@n2, Alpha=@n3), B=Constant3Vector() {constant=<rgba 0.0931 0.0232 0.0176 1>}, Alpha=NamedRerouteUsage()) {constAlpha=1}
   Alpha=Saturate(None=ScalarParameter() {defaultValue=3, parameterName=ID})
  )
  n4 = MaterialFunctionCall("Animation  Phase (0-1)"=Constant() {r=0.4}, Number of Rows=Constant() {r=3}, Number of Columns=Constant() {r=3}) {materialFunction=&FlipBook}
  n5 = Frac(None=Multiply(A=Time(), B=ScalarParameter() {defaultValue=2, parameterName=MouthSpeed}) {constB=2})
  n6 = TextureObject() {texture=/Safflower_SeasonData/UI/Textures/T_UI_Safflower_ChaseReward_MouthsEyes}
  n7 = MaterialFunctionCall(
   "Animation  Phase (0-1)"=LinearInterpolate(Alpha=OneMinus(None=@n5)) {constB=0.44}
   Number of Rows=Constant() {r=3}
   Number of Columns=Constant() {r=3}
   Texture=@n6
  ) {materialFunction=&FlipBook}
  n8 = ComponentMask(None=@n7.UVs) {g=true}
  n9 = Constant() {r=-0.01}
  n10 = OneMinus(None=NamedRerouteUsage())
  n11 = AppendVector(A=ComponentMask(None=@n7.UVs) {r=true}, B=Add(A=Add(A=@n8, B=@n9), B=Multiply(A=@n10) {constB=0.005}))
  n12 = ScalarParameter() {parameterName=IsTalking}
  NamedRerouteDeclaration(None=TextureSample(UVs=LinearInterpolate(A=@n4.UVs, B=@n11, Alpha=@n12), Tex=@n6) {texture=&T_UI_Safflower_ChaseReward_MouthsEyes}.R) {name=MouthAlpha, nodeColor=<rgba 0 0 0 1>}
  n13 = ScalarParameter() {defaultValue=3, parameterName=ID}
  n14 = Multiply(
   A=Add(A=Time(), B=MaterialFunctionCall!(UV=@n13, Seed) {materialFunction=&MF_UI_Random}.Result)
   B=ScalarParameter() {defaultValue=0.1, parameterName=EyeSpeed}
  )
  n15 = Multiply(A=Distance(A=Frac(None=@n14), B=Constant() {r=0.5})) {constB=2}
  n16 = TextureObject() {texture=&T_UI_Safflower_ChaseReward_MouthsEyes}
  n17 = MaterialFunctionCall(
   "Animation  Phase (0-1)"=LinearInterpolate(Alpha=SmoothStep(Value=@n15) {constMin=0.965}) {constA=0.5, constB=0.95}
   Number of Rows=Constant() {r=3}
   Number of Columns=Constant() {r=3}
   Texture=@n16
  ) {materialFunction=&FlipBook}
  n18 = ComponentMask(None=@n17.UVs) {g=true}
  n19 = LinearInterpolate(
   B=Multiply(A=NamedRerouteUsage(), B=ScalarParameter() {defaultValue=-0.0015, parameterName=TalkingEyeMovement}) {constB=0.005}
   Alpha=ScalarParameter() {parameterName=IsTalking}
  ) {constB=0}
  n20 = TextureSample(UVs=AppendVector(A=ComponentMask(None=@n17.UVs) {r=true}, B=Add(A=@n18, B=@n19)), Tex=@n16) {texture=&T_UI_Safflower_ChaseReward_MouthsEyes}
  NamedRerouteDeclaration(None=LinearInterpolate(A=@n20.R, Alpha=@n20.R)) {name=EyesAlpha, nodeColor=<rgba 0 0 0 1>}
  n21 = Constant3Vector() {constant=<rgba 0.0116 0.0116 0.0116 1>}
  n22 = Constant3Vector() {constant=<rgba 1 1 1 0>}
  NamedRerouteDeclaration(None=LinearInterpolate(A=LinearInterpolate(B=@n21, Alpha=@n20.R), Alpha=LinearInterpolate(B=@n22, Alpha=@n20.G))) {name=EyesColor, nodeColor=<rgba 0 0 0 1>}
  n23 = ComponentMask(None=MaterialFunctionCall(UVs=TextureCoordinate(), Texture Scale=Constant() {r=1}) {materialFunction=&ScaleUVsByCenter}.UVs) {g=true, r=true}
  n24 = AppendVector(A=ScalarParameter() {parameterName=Position X}, B=ScalarParameter() {parameterName=Position Y})
  n25 = ScalarParameter() {defaultValue=3, parameterName=ID}
  n26 = AppendVector(A=Fmod(A=@n25, B=Constant() {r=3}), B=Floor(None=Divide(A=@n25) {constB=3}))
  n27 = Add(A=Divide(A=Add(A=@n23, B=@n24), B=Constant() {r=3.1}) {constB=0}, B=Multiply(A=Floor(None=@n26), B=Constant() {r=0.3333}) {constB=0.33})
  n28 = TextureSample!(
   UVs=MaterialFunctionCall(UVs=@n27, Texture Scale=NamedRerouteUsage()) {materialFunction=&ScaleUVsByCenter}.UVs
   Tex
   "DDX(UVs)"=DDX(Value=@n27)
   "DDY(UVs)"=DDY(Value=@n27)
   Apply View MipBias
  ) {
   mipValueMode=TMVM_Derivative
   samplerSource=SSM_Clamp_WorldGroupSettings
   texture=/Safflower_SeasonData/UI/Textures/T_UI_Safflower_ChaseReward_Tiles
  }
  NamedRerouteDeclaration(None=@n28.RGB) {name=BerryColor, nodeColor=<rgba 0 0 0 1>}
  NamedRerouteDeclaration(None=@n28.A) {name=BerryAlpha, nodeColor=<rgba 0 0 0 1>}
  n29 = Constant() {r=1}
  n30 = Frac(None=Multiply(A=Time(), B=ScalarParameter() {defaultValue=3, parameterName=BerryTalkingSpeed}))
  n31 = Constant() {r=0.5}
  n32 = Power(Base=Multiply(A=Distance(A=@n30, B=@n31)) {constB=2})
  n33 = TextureCoordinate()
  n34 = LinearInterpolate(
   A=LinearInterpolate(
    A=ScalarParameter() {defaultValue=0.95, parameterName=TalkingScaleFactor}
    Alpha=NamedRerouteDeclaration(None=@n32) {name=TalkingSpeed, nodeColor=<rgba 0 0 0 1>}
   ) {constA=0.9}
   Alpha=OneMinus(None=ComponentMask(None=@n33) {g=true})
  )
  NamedRerouteDeclaration(None=LinearInterpolate(B=AppendVector(A=@n29, B=@n34), Alpha=ScalarParameter() {parameterName=IsTalking}) {constA=1, constAlpha=1}) {name=TalkingScale, nodeColor=<rgba 0 0 0 1>}
  NamedRerouteUsage()
 ]
}
