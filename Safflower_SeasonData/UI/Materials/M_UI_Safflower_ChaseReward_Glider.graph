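# Translucent UI material for the Safflower season chase reward glider, built to show the glider texture gently floating.
# UVs are fitted through MF_UI_UVScaling, nudged horizontally and scaled about the centre.
# A small vertical offset and CustomRotator then adjust them, with speed-scaled Time and two sine waves declared as named reroutes.
# The texture sample feeds colour and alpha reroutes; no connection to the material outputs appears.
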Material /Safflower_SeasonData/UI/Materials/M_UI_Safflower_ChaseReward_Glider
graph {
 properties {
  bAutomaticallySetUsageInEditor=false
  bUsedWithStaticMesh=false
  blendMode=BLEND_Translucent
  materialDomain=MD_UI
  usedShadingModels=Default Lit
 }
 nodes [
  NamedRerouteDeclaration(None=TextureObjectParameter() {parameterName=Reward Texture, texture=/Safflower_SeasonData/UI/Textures/T_UI_Safflower_Glider}) {name=Reward Texture, nodeColor=<rgba 0.7813 0 1 1>}
  n1 = NamedRerouteDeclaration(None=Multiply(A=Time(), B=ScalarParameter() {defaultValue=0.1, parameterName=Time Speed})) {name=Time, nodeColor=<rgba 0 1 0.4609 1>}
  NamedRerouteDeclaration(None=Sine(None=@n1)) {name=Sine Time, nodeColor=<rgba 1 0.1406 0 1>}
  NamedRerouteDeclaration(None=Sine(None=Multiply(A=@n1) {constB=0.5})) {name=Sine Double Time, nodeColor=<rgba 0 0.1797 1 1>}
  n2 = MaterialFunctionCall() {materialFunction=&MF_UI_UVScaling}
  n3 = Add(A=ComponentMask(None=@n2.Scaled to Fit) {r=true}) {constB=0.05}
  n4 = ComponentMask(None=@n2.Scaled to Fit) {g=true}
  n5 = MaterialFunctionCall(
   UVs=NamedRerouteDeclaration(None=AppendVector(A=@n3, B=@n4)) {name=UV, nodeColor=<rgba 0.5 1 0 1>}
   Texture Scale=ScalarParameter() {defaultValue=0.91, parameterName=Texture Scale}
  ) {materialFunction=&ScaleUVsByCenter}
  n6 = ComponentMask(None=@n5.UVs) {g=true}
  n7 = Multiply(A=NamedRerouteUsage(), B=ScalarParameter() {defaultValue=0.025, parameterName=Floating Parameter})
  n8 = NamedRerouteDeclaration(None=AppendVector(A=ComponentMask(None=@n5.UVs) {r=true}, B=Add(A=@n6, B=@n7))) {name=Floating UV, nodeColor=<rgba 1 0 0.8203 1>}
  n9 = Multiply(A=Divide(A=NamedRerouteUsage()) {constB=4}) {constB=0.05}
  n10 = TextureSample(
   UVs=NamedRerouteDeclaration(None=MaterialFunctionCall(UVs=@n8, "Rotation Angle (0-1)"=@n9) {materialFunction=&CustomRotator}.Rotated Values) {name=Reward UV, nodeColor=<rgba 0 1 0.8594 1>}
   Tex=NamedRerouteUsage()
  ) {texture=&T_UI_Safflower_Glider}
  NamedRerouteDeclaration(None=@n10.RGB) {name=Reward Texture Colour, nodeColor=<rgba 1 0.5391 0 1>}
  NamedRerouteDeclaration(None=@n10.A) {name=Reward Texture Alpha, nodeColor=<rgba 0.2188 0 1 1>}
 ]
}
