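# Translucent UI material for Safflower season key character art, with adjustable texture scale and position and animatable effects.
# Offset samples of the character texture's alpha form a coloured drop shadow and a two-colour edge highlight, refined by a channel mask and smoothsteps.
# Eased linear-gradient functions drive front and back-character overlays, bottom fade and mask wipe.
# Named reroutes share intermediates, including a strip wipe and distorted UVs whose offset is multiplied by zero.
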Material /Safflower_SeasonData/UI/Materials/M_UI_Safflower_KeyCharacters
graph {
 properties {
  bAutomaticallySetUsageInEditor=false
  bUsedWithStaticMesh=false
  blendMode=BLEND_Translucent
  materialDomain=MD_UI
  usedShadingModels=Default Lit
 }
 nodes [
  n1 = ScalarParameter() {defaultValue=1, group=Texture, parameterName=TextureXScale}
  n2 = ScalarParameter() {defaultValue=1, group=Texture, parameterName=TextureYScale}
  n3 = MaterialFunctionCall(
   UVs=TextureCoordinate()
   Texture Scale=Multiply(A=ScalarParameter() {defaultValue=1, group=Texture, parameterName=TextureScale}, B=AppendVector(A=@n1, B=@n2))
  ) {materialFunction=&ScaleUVsByCenter}
  n4 = AppendVector(A=ScalarParameter() {group=Texture, parameterName=TextureXPosition}, B=ScalarParameter() {group=Texture, parameterName=TextureYPosition})
  n5 = TextureSampleParameter2D(UVs=NamedRerouteDeclaration(None=Add(A=@n3.UVs, B=@n4)) {name=CharacterUV, nodeColor=<rgba 0 0 0 1>}) {
   group=Texture
   parameterName=Texture_Asset
   samplerSource=SSM_Clamp_WorldGroupSettings
   texture=/Safflower_SeasonData/UI/Textures/T_UI_Safflower_OzoneCredit
  }
  n6 = ScalarParameter() {defaultValue=0.1, parameterName=OverlayIntensity}
  n7 = ScalarParameter() {parameterName=OverlayPosition}
  n8 = MaterialFunctionCall(
   GradientStart=ScalarParameter() {defaultValue=0.25, parameterName=OverlayWidth}
   GradientEnd=Constant()
   RotationAngle=ScalarParameter() {defaultValue=0.25, parameterName=OverlayAngle}
   GradientOffset=@n7
  ) {materialFunction=&MF_UI_LinearGradient}
  n9 = Saturate(None=OneMinus(None=MaterialFunctionCall("Value (0-1)"=@n8.Result) {materialFunction=&MF_UI_EaseCurves}.EaseOutSine))
  n10 = Ceil(None=@n7)
  n11 = NamedRerouteDeclaration(None=LinearInterpolate(A=LinearInterpolate(A=@n5.RGB, B=@n5.RGB, Alpha=@n6), B=@n5.RGB, Alpha=LinearInterpolate(B=@n9, Alpha=@n10) {constA=1})) {name=CharacterColour, nodeColor=<rgba 0 0 0 1>}
  n12 = VectorParameter() {group=Highlight, parameterName=HighlightColour02, defaultValue=<rgba 0 1 0.9612 1>}
  n13 = VectorParameter() {group=Highlight, parameterName=HighlightColour01, defaultValue=<rgba 1 0.3718 0 1>}
  n14 = Power(Base=NamedRerouteUsage()) {constExponent=4}
  LinearInterpolate(
   A=@n11
   B=LinearInterpolate(A=@n11, B=LinearInterpolate(A=@n12.RGB, B=@n13.RGB, Alpha=@n14), Alpha=NamedRerouteUsage())
   Alpha=ScalarParameter() {group=Highlight, parameterName=HighlightMainController}
  )
  n15 = TextureSampleParameter2D(UVs=NamedRerouteUsage()) {
   group=Texture
   parameterName=Texture_Asset
   samplerSource=SSM_Clamp_WorldGroupSettings
   texture=&T_UI_Safflower_OzoneCredit
  }
  n16 = MaterialFunctionCall(UVs=TextureCoordinate(), Texture Scale=Constant() {r=1}) {materialFunction=&ScaleUVsByCenter}
  n17 = ScalarParameter() {defaultValue=-0.5, parameterName=DropShadowPanX}
  n18 = ScalarParameter() {defaultValue=1, parameterName=DropShadowPanY}
  n19 = Add(A=Divide(A=AppendVector(A=@n17, B=@n18)) {constB=100})
  n20 = TextureSampleParameter2D(UVs=Add(A=ComponentMask(None=@n16.UVs) {g=true, r=true}, B=OneMinus(None=@n19))) {
   group=Texture
   parameterName=Texture_Asset
   samplerSource=SSM_Clamp_WorldGroupSettings
   texture=&T_UI_Safflower_OzoneCredit
  }
  n21 = Multiply(A=NamedRerouteUsage(), B=Add(A=@n20.A) {constB=0})
  n22 = LinearInterpolate(A=@n15.RGB, B=NamedRerouteUsage(), Alpha=Saturate(None=Subtract(A=@n21, B=@n15.A)))
  n23 = ScalarParameter() {defaultValue=0.25, parameterName=OverlayWidth_BackCharacter}
  n24 = Constant()
  n25 = ScalarParameter() {defaultValue=0.25, parameterName=OverlayAngle_BackCharacter}
  n26 = ScalarParameter() {parameterName=OverlayPosition_BackCharacter}
  n27 = OneMinus(None=MaterialFunctionCall("Value (0-1)"=MaterialFunctionCall(GradientStart=@n23, GradientEnd=@n24, RotationAngle=@n25, GradientOffset=@n26) {materialFunction=&MF_UI_LinearGradient}.Result) {materialFunction=&MF_UI_EaseCurves}.EaseOutSine)
  n28 = LinearInterpolate(
   A=LinearInterpolate(A=@n22, B=NamedRerouteUsage(), Alpha=ScalarParameter() {parameterName=OverlayIntensity_BackCharacter})
   B=@n22
   Alpha=LinearInterpolate(B=Saturate(None=@n27), Alpha=Ceil(None=@n26)) {constA=1}
  )
  n29 = ScalarParameter() {defaultValue=0.25, parameterName=OverlayWidth}
  n30 = Constant()
  n31 = ScalarParameter() {defaultValue=0.25, parameterName=OverlayAngle}
  n32 = ScalarParameter() {parameterName=OverlayPosition}
  n33 = OneMinus(None=MaterialFunctionCall("Value (0-1)"=MaterialFunctionCall(GradientStart=@n29, GradientEnd=@n30, RotationAngle=@n31, GradientOffset=@n32) {materialFunction=&MF_UI_LinearGradient}.Result) {materialFunction=&MF_UI_EaseCurves}.EaseOutSine)
  LinearInterpolate(
   A=LinearInterpolate(A=@n28, B=NamedRerouteUsage(), Alpha=ScalarParameter() {defaultValue=0.1, parameterName=OverlayIntensity})
   B=@n28
   Alpha=LinearInterpolate(B=Saturate(None=@n33), Alpha=Ceil(None=@n32)) {constA=1}
  )
  n34 = TextureCoordinate()
  n35 = ComponentMask(None=@n34) {g=true}
  n36 = ScalarParameter() {defaultValue=0.9, parameterName=DistortionOffset}
  n37 = ScalarParameter() {defaultValue=1, parameterName=DistortionProgress}
  n38 = ScalarParameter() {defaultValue=8, parameterName=DistortionSharpness}
  n39 = Multiply(A=Add(A=Subtract(A=@n35, B=@n36), B=@n37), B=@n38)
  n40 = OneMinus(None=ComponentMask(None=@n34) {g=true})
  n41 = Multiply(A=Add(A=Subtract(A=@n40, B=@n36), B=OneMinus(None=@n37)), B=@n38)
  NamedRerouteDeclaration(None=Saturate(None=Multiply(A=@n39, B=@n41))) {name=StripWipe, nodeColor=<rgba 0 0 0 1>}
  NamedRerouteDeclaration(None=LinearInterpolate(
    A=Constant3Vector() {constant=<rgba 0.5841 0.0137 0.4125 1>}
    B=Constant3Vector() {constant=<rgba 0.0319 0.0065 0.0222 1>}
    Alpha=ScalarParameter() {parameterName=PurpleToPinkGradient}
   )) {name=PurpleGradientShadow, nodeColor=<rgba 0 0 0 1>}
  n42 = VectorParameter() {parameterName=DropShadow/OverlayColour, defaultValue=<rgba 0.0152 0.0066 0.0156 1>}
  NamedRerouteDeclaration(None=@n42.A) {name=DropShadowAlpha, nodeColor=<rgba 0 0 0 1>}
  NamedRerouteDeclaration(None=LinearInterpolate(A=Constant4Vector() {constant=<rgba 1 1 1 0>}.A, B=@n42.RGB, Alpha=ScalarParameter() {parameterName=WhiteToColourLerpAnim})) {name=DropShadowColour, nodeColor=<rgba 0 0 0 1>}
  n43 = Constant()
  n44 = NamedRerouteUsage()
  NamedRerouteDeclaration(None=Add(A=TextureCoordinate(), B=Multiply(A=@n43, B=@n44))) {name=DistortedUVs, nodeColor=<rgba 1 0 0.6328 1>}
  n45 = Rotator(Coordinate=NamedRerouteUsage(), Time=Constant() {r=1})
  n46 = Add(A=Subtract(A=ComponentMask(None=@n45) {r=true}, B=ComponentMask(None=@n45) {g=true}))
  n47 = SmoothStep(Value=NamedRerouteDeclaration(None=Divide(A=@n46)) {name=HighlightGradient, nodeColor=<rgba 0 0 0 1>}) {constMax=0.7, constMin=0.5}
  n48 = ScalarParameter() {defaultValue=0.002, group=Highlight, parameterName=HighlightPositionX}
  n49 = ScalarParameter() {defaultValue=-0.006, group=Highlight, parameterName=HighlightPositionY}
  n50 = TextureSampleParameter2D(UVs=Add(A=NamedRerouteUsage(), B=AppendVector(A=@n48, B=@n49))) {
   group=Texture
   parameterName=Texture_Asset
   samplerSource=SSM_Clamp_WorldGroupSettings
   texture=&T_UI_Safflower_OzoneCredit
  }
  n51 = Subtract(A=NamedRerouteUsage(), B=@n50.A)
  NamedRerouteDeclaration(None=Saturate(None=Multiply(A=@n47, B=@n51))) {name=HighlightEdge, nodeColor=<rgba 0 0 0 1>}
  n52 = ScalarParameter() {defaultValue=0.0001, parameterName=BottomFadePosition}
  n53 = MaterialFunctionCall(
   GradientStart=ScalarParameter() {defaultValue=0.2, parameterName=BottomFadeWidth}
   GradientEnd=Constant()
   RotationAngle=ScalarParameter() {defaultValue=0.25, parameterName=BottomFadeAngle}
   GradientOffset=@n52
  ) {materialFunction=&MF_UI_LinearGradient}
  n54 = LinearInterpolate(
   B=Saturate(None=MaterialFunctionCall("Value (0-1)"=@n53.Result) {materialFunction=&MF_UI_EaseCurves}.EaseOutSine)
   Alpha=Ceil(None=@n52)
  ) {constA=1}
  NamedRerouteDeclaration(None=LinearInterpolate(A=@n5.A, B=Subtract(A=@n5.A, B=@n54), Alpha=ScalarParameter() {defaultValue=1, group=BottomFade, parameterName=MaskOff})) {name=CharacterOpacity, nodeColor=<rgba 0 0 0 1>}
  n55 = SmoothStep(
   Min=ScalarParameter() {defaultValue=0.3, group=Highlight, parameterName=SoftHighlightFinish}
   Max=ScalarParameter() {defaultValue=0.65, group=Highlight, parameterName=SoftHighlightStart}
   Value=NamedRerouteUsage()
  )
  n56 = StaticComponentMaskParameter(None=NamedRerouteUsage()) {defaultR=true, parameterName=HighlightMask}
  n57 = Multiply(
   A=LinearInterpolate(A=@n56, B=OneMinus(None=@n56), Alpha=ScalarParameter() {group=Highlight, parameterName=MaskInverter})
   B=ScalarParameter() {defaultValue=1, group=Highlight, parameterName=HighlightMaskMultiplier}
  )
  n58 = SmoothStep(
   Min=ScalarParameter() {defaultValue=0.2, group=Highlight, parameterName=SoftHighlightColourInfluenceStart}
   Max=ScalarParameter() {defaultValue=1, group=Highlight, parameterName=SoftHighlightColourInfluenceFinish}
   Value=Rotator(Coordinate=Multiply(A=@n55, B=@n57), Time=Constant() {r=0.5})
  )
  n59 = ScalarParameter() {defaultValue=1.25, group=Highlight, parameterName=SoftHighlightIntensity}
  n60 = Add(A=NamedRerouteUsage(), B=Saturate(None=Multiply(A=@n58, B=@n59)))
  n61 = OneMinus(None=ComponentMask(None=NamedRerouteUsage()) {g=true})
  NamedRerouteDeclaration(None=Multiply(A=Saturate(None=@n60), B=SmoothStep(Value=@n61) {constMax=0.8, constMin=0.4})) {name=Highlight, nodeColor=<rgba 0 0 0 1>}
  NamedRerouteUsage()
  n62 = Max(A=@n21, B=@n15.A)
  n63 = ScalarParameter() {defaultValue=0.0001, parameterName=BottomFadePosition}
  n64 = MaterialFunctionCall("Value (0-1)"=MaterialFunctionCall(
    GradientStart=ScalarParameter() {defaultValue=0.2, parameterName=BottomFadeWidth}
    GradientEnd=Constant()
    RotationAngle=ScalarParameter() {defaultValue=0.25, parameterName=BottomFadeAngle}
    GradientOffset=@n63
   ) {materialFunction=&MF_UI_LinearGradient}.Result) {materialFunction=&MF_UI_EaseCurves}
  n65 = LinearInterpolate(B=Saturate(None=OneMinus(None=@n64.EaseOutSine)), Alpha=Ceil(None=@n63)) {constA=1}
  n66 = Constant() {r=0.08}
  n67 = Constant()
  n68 = Constant() {r=0.25}
  n69 = ScalarParameter() {defaultValue=1, parameterName=MaskWipeProgress}
  Saturate(None=Multiply(
    A=Multiply(A=@n62, B=@n65)
    B=MaterialFunctionCall(GradientStart=@n66, GradientEnd=@n67, RotationAngle=@n68, GradientOffset=@n69) {materialFunction=&MF_UI_LinearGradient}.Result
   ))
 ]
}
