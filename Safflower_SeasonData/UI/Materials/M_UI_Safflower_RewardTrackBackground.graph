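# Translucent UI material for the Safflower season reward track background, drawing a liquid-style progress fill with bubbles and glows.
# A rotated animated sine-wave SDF forms the progress mask and glowing edge, alongside gradient colors, inner glow and a pixel-aware rounded-corner mask.
# Randomized particle-grid circles form rising bubbles, mapped mostly in viewport space with aspect correction.
# Panned, rotated noise samples supply water distortion, and a texture parameter provides an item display.
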Material /Safflower_SeasonData/UI/Materials/M_UI_Safflower_RewardTrackBackground
graph {
 properties {
  bAutomaticallySetUsageInEditor=false
  bUsedWithStaticMesh=false
  blendMode=BLEND_Translucent
  materialDomain=MD_UI
  usedShadingModels=Default Lit
 }
 nodes [
  n1 = MaterialFunctionCall(
   UVs=MaterialFunctionCall(UV=NamedRerouteUsage()) {materialFunction=&MF_UI_UVScaling}.ScaleX
   "Rotation Angle (0-1)"=Constant() {r=0.25}
  ) {materialFunction=&CustomRotator}
  n2 = NamedRerouteUsage()
  n3 = Constant() {r=0.04}
  Add(A=@n1.Rotated Values, B=AppendVector(A=Constant(), B=Multiply(A=@n2, B=@n3) {constB=0.02, desc=Progression Wave Mask Intensity}))
  n4 = NamedRerouteUsage()
  n5 = Multiply(A=Multiply(A=NamedRerouteUsage()) {constB=2.5}, B=ScalarParameter() {defaultValue=1, group=Glow Mask, parameterName=Outline Glow Intensity}) {constB=1.5}
  n6 = NamedRerouteUsage()
  n7 = NamedRerouteUsage()
  n8 = ScalarParameter() {defaultValue=0.1, group=Bubbles, parameterName=Bubble Opacity}
  n9 = NamedRerouteUsage()
  n10 = Add(A=@n9, B=NamedRerouteUsage()) {constB=0.2}
  n11 = Max(A=NamedRerouteUsage()) {constB=0}
  n12 = ScalarParameter() {defaultValue=2, parameterName=InnerGlowIntensity}
  n13 = Multiply(A=Multiply(A=Power(Base=@n11, Exp=@n12), B=ScalarParameter() {parameterName=InnerGlowOpacity})) {constB=0.2}
  n14 = MaterialFunctionCall(UV=NamedRerouteUsage()) {materialFunction=&MF_UI_RadialGradient}
  LinearInterpolate(
   A=LinearInterpolate(A=LinearInterpolate(A=@n4, B=@n5, Alpha=@n6), B=Add(A=@n7, B=@n8), Alpha=@n8) {constB=0.1}
   B=LinearInterpolate(A=LinearInterpolate(A=@n9, B=@n10, Alpha=@n13) {constB=0}, B=NamedRerouteUsage(), Alpha=Multiply(A=@n14.Smooth) {constB=0})
   Alpha=NamedRerouteUsage()
  ) {constB=0}
  n15 = NamedRerouteUsage()
  n16 = ScalarParameter() {defaultValue=0.8, group=Bubbles, parameterName=Bubbles_Scale, sliderMax=2, sliderMin=0.1}
  n17 = NamedRerouteUsage()
  n18 = Constant()
  n19 = Multiply(A=Add(A=Time(), B=NamedRerouteUsage()), B=ScalarParameter() {defaultValue=0.025, group=Bubbles, parameterName=Bubbles_SpeedMultiply}) {constB=0.1}
  n20 = NamedRerouteUsage()
  n21 = TextureSample(UVs=Rotator(Coordinate=@n15, Time=@n19) {speed=6.2832}, Tex=@n20) {texture=&T_Smooth_tile}
  n22 = TextureSample(
   UVs=Rotator(
    Coordinate=MaterialFunctionCall(UV=@n15) {materialFunction=&MF_UI_UVScaling}.ScaleX
    Time=Multiply(A=@n19) {constB=-0.05}
   ) {speed=6.2832}
   Tex=@n20
  ) {texture=&T_Smooth_tile}
  n23 = Add(
   A=Add(
    A=MaterialFunctionCall(UVs=@n15, Texture Scale=@n16) {materialFunction=&ScaleUVsByCenter}.UVs
    B=AppendVector(A=@n17, B=@n18)
   )
   B=Multiply(A=DotProduct(A=@n21.R, B=@n22.R), B=ScalarParameter() {group=Bubbles, parameterName=Bubbles_Distortion, sliderMax=0.5}) {constB=0.1, desc=Intensity}
  )
  n24 = Panner(Coordinate=@n23, Speed=AppendVector(A=Constant(), B=Constant() {r=0.01})) {desc=Panning Speed, speedY=0.025}
  MaterialFunctionCall(UV=@n24) {materialFunction=&MF_UI_UVScaling}
  n25 = Power(Base=Constant3Vector() {constant=<rgba 0.4766 1 0.3881 1>}) {constExponent=0.8}
  MaterialFunctionCall(Hue Shift Percentage=Constant() {r=0.05}, Texture=@n25) {materialFunction=&HueShift}
  Multiply(
   A=ScalarParameter() {defaultValue=0.2, parameterName=ScreenBuffer-Multiplier}
   B=MaterialFunctionCall!("UVs (Optional)") {materialFunction=/Engine/Functions/UserInterface/GetSlatePost0}.LinearRGB
  )
  NamedRerouteDeclaration(None=ScalarParameter() {group=Edge Noise, parameterName=RandomSeed, sliderMax=1}) {name=Random Seed, nodeColor=<rgba 0 0 0 1>}
  n26 = NamedRerouteUsage()
  n27 = OneMinus(None=NamedRerouteUsage())
  n28 = MaterialFunctionCall!(
   UV=Add(A=@n1.Rotated Values, B=AppendVector(A=@n26, B=@n27))
   Wave Height=Constant() {r=0.06}
   Wave Elevation
   Wave Length=Constant() {r=0.8}
   Wave Speed=Constant() {r=-0.2}
   GlowMin
   GlowMax=Constant() {r=0.025}
   Outline Thickness
   Outline Alignment
   Custom Curve
  ) {materialFunction=/FortUILibrary/MaterialFunctions/Shapes/MF_UI_Wave}
  NamedRerouteDeclaration(None=@n28.Raw UV) {name=SineWaveRaw, nodeColor=<rgba 0 0 0 1>}
  n29 = Multiply(A=Time(), B=ScalarParameter() {defaultValue=0.25, parameterName=Bubble Speed}) {constB=0.5}
  n30 = LinearInterpolate(Alpha=NamedRerouteUsage()) {constA=0.05, constB=0.1}
  n31 = MaterialFunctionCall() {materialFunction=&ScreenResolution}
  n32 = Divide(A=ComponentMask(None=@n31.Visible Resolution) {r=true}, B=ComponentMask(None=@n31.Visible Resolution) {g=true})
  n33 = Constant() {r=1}
  n34 = Constant()
  n35 = MaterialFunctionCall!(
   UV=LinearInterpolate(A=NamedRerouteUsage(), B=ScreenPosition().ViewportUV) {constAlpha=0.97}
   Global Offset=AppendVector(A=Constant(), B=Multiply(A=@n29, B=@n30) {constB=0.05})
   Global Rotation
   Tiling=Multiply(A=NamedRerouteUsage(), B=AppendVector(A=@n32, B=@n33))
   Cell Offset=@n34
   Cell Scale=Constant() {r=1}
   Cell Rotation=@n34
   Random Seed
  ) {materialFunction=/FortUILibrary/MaterialFunctions/Basics/MF_UI_ParticleGrid}
  NamedRerouteDeclaration(None=@n35.Random Rows) {name=RandomRows, nodeColor=<rgba 0 0 0 1>}
  n36 = NamedRerouteUsage()
  n37 = OneMinus(None=@n36)
  n38 = MaterialFunctionCall(DistanceField=NamedRerouteDeclaration(None=@n28.SDF) {name=SineWaveEdge, nodeColor=<rgba 0 0 0 1>}) {materialFunction=&MF_UI_AntiAliasing}
  n39 = Ceil(None=@n36)
  n40 = NamedRerouteUsage()
  NamedRerouteDeclaration(None=LinearInterpolate(A=LinearInterpolate(A=@n37, B=@n38.Outer, Alpha=@n39), B=@n37, Alpha=Floor(None=@n40))) {name=Progress Mask, nodeColor=<rgba 0 0 0 1>}
  n41 = SmoothStep(Min=ScalarParameter() {group=Glow Mask, parameterName=Wiggle Glow Mask Thickness}, Value=NamedRerouteUsage()) {constMax=0.1, constMin=-0.5}
  n42 = NamedRerouteUsage()
  n43 = Multiply(A=Multiply(A=Power(Base=@n41) {constExponent=3}) {constB=0.8}, B=SmoothStep(Value=OneMinus(None=@n42)) {constMax=0.01})
  NamedRerouteDeclaration(None=Saturate(None=Multiply(A=@n43) {constB=1.5})) {name=Outline Glow Mask, nodeColor=<rgba 0 0 0 1>}
  n44 = ScalarParameter() {defaultValue=20, group=Layout, parameterName=TopLeftCornerRadius, sortPriority=4}
  n45 = ScalarParameter() {defaultValue=20, group=Layout, parameterName=BottomLeftCornerRadius, sortPriority=7}
  n46 = ScalarParameter() {defaultValue=20, group=Layout, parameterName=TopRightCornerRadius, sortPriority=5}
  n47 = ScalarParameter() {defaultValue=20, group=Layout, parameterName=BottomRightCornerRadius, sortPriority=6}
  n48 = ScalarParameter() {
   defaultValue=1
   desc="If enabled reads corner radius values in pixels.\r\nbool"
   group=Layout
   parameterName=UseValueInPixels
   sliderMax=1
   sortPriority=1
  }
  NamedRerouteDeclaration(None=MaterialFunctionCall(
    Corner Radius=MaterialFunctionCall(R=@n44, G=@n45, B=@n46, A=@n47) {materialFunction=&AppendMany}.RGBA
    Radius in Pixels=@n48
   ) {materialFunction=&MF_UI_CoreUI_RoundedMask}.Shape Mask) {name=Shape Mask, nodeColor=<rgba 0 0 0 1>}
  n49 = TextureObjectParameter() {parameterName=ItemDisplay, texture=/Game/Global/Textures/Default/Blanks/T_Blank_White_ZeroAlpha_sRGBColor}
  NamedRerouteDeclaration(None=TextureProperty(Tex=@n49)) {name=Item Dimensions, nodeColor=<rgba 0 0 0 1>}
  NamedRerouteDeclaration(None=TextureObject() {samplerType=SAMPLERTYPE_Grayscale, texture=/FortWater/Textures/LowResBlurredNoise}) {name=Noise Texture, nodeColor=<rgba 0 0 0 1>}
  NamedRerouteDeclaration(None=ScalarParameter() {defaultValue=30, group=Bubbles, parameterName=Bubble Amount}) {name=Bubble Amount, nodeColor=<rgba 0 0 0 1>}
  n50 = Time()
  n51 = NamedRerouteUsage()
  n52 = NamedRerouteUsage()
  n53 = Floor(None=Multiply(A=ComponentMask(None=@n51) {r=true}, B=@n52))
  n54 = LinearInterpolate(Alpha=MaterialFunctionCall!(UV=Divide(A=@n53, B=@n52), Seed=Constant() {r=2}) {materialFunction=&MF_UI_Random}.Result) {constA=1}
  NamedRerouteDeclaration(None=Panner(Time=Multiply(A=@n50, B=@n54)) {speedY=0.025}) {name=Bubble UV, nodeColor=<rgba 0 0 0 1>}
  n55 = Constant()
  n56 = MaterialFunctionCall(
   UV=NamedRerouteUsage()
   Tiling=NamedRerouteUsage()
   Cell Offset=@n55
   Cell Scale=Constant() {r=1}
   Cell Rotation=@n55
   Random Seed=Constant() {r=0.5}
  ) {materialFunction=&MF_UI_ParticleGrid}
  NamedRerouteDeclaration(None=@n56.Random Cell Values) {name=Random Bubble, nodeColor=<rgba 0 0 0 1>}
  n57 = Distance(A=Add(A=TextureCoordinate(), B=Constant2Vector() {g=-0.5}), B=Constant() {r=0.5})
  n58 = VectorParameter() {group=Color, parameterName=BkgGradientColor, defaultValue=<rgba 0 0.5938 0.2595 1>}
  NamedRerouteDeclaration(None=LinearInterpolate(
    A=@n25
    B=NamedRerouteDeclaration(None=@n58.RGB) {name=BackColor, nodeColor=<rgba 0 0 0 1>}
    Alpha=MaterialFunctionCall("Value (0-1)"=@n57) {materialFunction=&MF_UI_EaseCurves}.EaseInOutSine
   )) {name=Gradient, nodeColor=<rgba 0 0 0 1>}
  n59 = LinearInterpolate(
   A=ScalarParameter() {defaultValue=0.3, group=Bubbles, parameterName=Bubble Size Min}
   B=ScalarParameter() {defaultValue=0.8, group=Bubbles, parameterName=Bubble Size Max}
   Alpha=NamedRerouteDeclaration(None=@n35.Random Cell Values) {name=Random Cell Values, nodeColor=<rgba 0 0 0 1>}
  ) {constA=0.3, constB=0.6}
  n60 = MaterialFunctionCall(UV=NamedRerouteUsage(), Seed=Constant() {r=4}) {materialFunction=&MF_UI_Random}
  NamedRerouteDeclaration(None=Multiply(
    A=MaterialFunctionCall(UVs=@n35.Cells UV, Size=@n59) {materialFunction=&MF_UI_Circle}.Fill
    B=Step(Y=@n60.Result) {constX=0.3}
   ) {constA=1}) {name=Bubble, nodeColor=<rgba 0 0 0 1>}
  n61 = MaterialFunctionCall() {materialFunction=&GetUserInterfaceUV}
  n62 = ComponentMask(None=@n61.9-Slice UV) {r=true}
  n63 = Constant() {r=0.5}
  n64 = SmoothStep(Value=Multiply(A=Distance(A=@n62, B=@n63)) {constB=2}) {constMin=0.6}
  n65 = ComponentMask(None=@n61.9-Slice UV) {g=true}
  n66 = Constant() {r=0.5}
  n67 = SmoothStep(Value=Multiply(A=Distance(A=@n65, B=@n66)) {constB=2}) {constMin=0.6}
  n68 = OneMinus(None=Multiply(A=OneMinus(None=@n64), B=OneMinus(None=@n67)))
  NamedRerouteDeclaration(None=Saturate(None=Power(Base=@n68))) {name=Inner Glow Mask, nodeColor=<rgba 0 0 0 1>}
  NamedRerouteDeclaration(None=LinearInterpolate(
    A=Constant3Vector() {constant=<rgba 0.0482 0.1022 0.1529 1>}
    B=Constant3Vector() {constant=<rgba 0.016 0.0369 0.0578 1>}
    Alpha=MaterialFunctionCall() {materialFunction=&LinearGradient}.VGradient
   )) {name=Background Colors, nodeColor=<rgba 0 0 0 1>}
  n69 = NamedRerouteUsage()
  n70 = Rotator(
   Coordinate=MaterialFunctionCall(UV=Panner(Coordinate=@n23) {desc=Panning Speed, speedY=0.001}) {materialFunction=&MF_UI_UVScaling}.ScaleY
   Time=@n69
  ) {speed=6.2832}
  n71 = ScalarParameter() {parameterName=WaterDistortion}
  NamedRerouteDeclaration(None=TextureSample(UVs=Multiply(A=@n70, B=@n71) {constB=0.1, desc=Intensity}, Tex=NamedRerouteUsage()) {texture=&T_Smooth_tile}.R) {name=Water Noise, nodeColor=<rgba 0 0 0 1>}
  NamedRerouteDeclaration(None=ScalarParameter() {defaultValue=0.8, group=Shape, parameterName=OverlayProgress, sliderMax=1, sortPriority=2}) {name=Progression Value, nodeColor=<rgba 0 0 0 1>}
  NamedRerouteDeclaration(None=MaterialFunctionCall() {materialFunction=&GetUserInterfaceUV}.Normalized UV) {name=Normalized UV, nodeColor=<rgba 0 0 0 1>}
  NamedRerouteDeclaration(None=@n49) {name=Item Display, nodeColor=<rgba 0 0 0 1>}
  NamedRerouteDeclaration(None=@n35.Random Columns) {name=RandomColumns, nodeColor=<rgba 0 0 0 1>}
  NamedRerouteUsage()
  OneMinus(None=Round())
  Rotator(Coordinate=@n24, Time=@n69) {speed=6.2832}
  Step(Y=@n56.Random Cell Values, X=ScalarParameter() {defaultValue=0.3, group=Bubbles, parameterName=Bubble Random Mask}) {constX=0.6}
 ]
}
