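# Translucent UI material for the straight vine graphic in the Safflower season UI, tinting a channel-packed texture green.
# The texture's G channel scales a dark green to form the base of a lerp.
# The R channel scales a brighter green that serves as the per-channel lerp alpha toward the unconnected B default, rather than as a second color.
# The graph shows no links to material outputs, and one dark green constant is unconnected.
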Material /Safflower_SeasonData/UI/Materials/M_UI_Safflower_VineStraight
graph {
 properties {bAutomaticallySetUsageInEditor=false, blendMode=BLEND_Translucent, materialDomain=MD_UI, usedShadingModels=Default Lit}
 nodes [
  Constant3Vector() {constant=<rgba 0.035 0.0625 0.0085 1>}
  n1 = TextureSample() {samplerType=SAMPLERTYPE_LinearColor, texture=/Safflower_SeasonData/UI/Textures/T_UI_Safflower_VineStraight}
  LinearInterpolate(
   A=Multiply(A=@n1.G, B=Constant3Vector() {constant=<rgba 0.035 0.0625 0.0085 0>})
   Alpha=Multiply(A=Constant3Vector() {constant=<rgba 0.1812 0.3231 0.0437 1>}, B=@n1.R)
  )
 ]
}
