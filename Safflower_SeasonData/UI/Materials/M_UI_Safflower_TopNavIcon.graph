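# Translucent UI material for the Safflower season top navigation icon, drawn from a signed distance field texture.
# The texture parameter feeds a channel-packed distance field function, with glow bounds set by a Threshold parameter plus and minus TextureFalloff.
# The function's Fill output is inverted and passed through the UI anti-aliasing function with constant threshold and smoothing.
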Material /Safflower_SeasonData/UI/Materials/M_UI_Safflower_TopNavIcon
graph {
 properties {
  bAutomaticallySetUsageInEditor=false
  bUsedWithStaticMesh=false
  blendMode=BLEND_Translucent
  materialDomain=MD_UI
  usedShadingModels=Default Lit
 }
 nodes [
  Constant3Vector() {constant=<rgba 1 1 1 0>}
  Constant() {r=0.45}
  Constant() {r=0.2}
  n1 = TextureObjectParameter() {parameterName=SDF_Texture, samplerType=SAMPLERTYPE_Alpha, texture=/Safflower_SeasonData/UI/Textures/T_UI_Safflower_TopNavIcon_SDF}
  n2 = TextureSample(Tex=@n1) {samplerType=SAMPLERTYPE_Alpha, texture=&T_UI_Safflower_TopNavIcon_SDF}
  n3 = ScalarParameter() {defaultValue=0.362, parameterName=Threshold}
  n4 = ScalarParameter() {defaultValue=0.029, parameterName=TextureFalloff}
  n5 = Subtract(A=@n3, B=@n4)
  n6 = Add(A=@n4, B=@n3)
  MaterialFunctionCall(
   DistanceField=OneMinus(None=MaterialFunctionCall!(RGBA Channels=@n2.RGBA, Glow Min=@n5, Glow Max=@n6, Outline Thickness) {materialFunction=/FortUILibrary/MaterialFunctions/Basics/MF_UI_ChannelPackedDistanceField}.Fill)
   Threshold=Constant() {r=0.8}
   Smoothing=Constant() {r=0.2}
  ) {materialFunction=&MF_UI_AntiAliasing}
 ]
}
