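# SafflowerCustomTab is a top bar tab button widget, derived from FortTopBarTabButton, that shows the Safflower season navigation icon.
# A SizeBox root centers an Overlay holding a single Image, drawn with the M_UI_Safflower_TopNavIcon material.
# It inherits an input action widget from the parent class and adds no logic, as its event graph holds only unconnected PreConstruct, Construct and Tick events.
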WidgetBlueprint /Safflower_SeasonData/UI/SafflowerCustomTab
graphs [EventGraph {nodes [
   Event PreConstruct()
   Event Construct()
   Event Tick()
  ]}]
variables []
widgets {
 info {inheritedWidgetCount=1, parentClass=/Script/FortniteUI.FortTopBarTabButton, rootWidgetClass=&SizeBox, widgetCount=4}
 widgets [
  SizeBox_Mobile {widgetClassPath=&SizeBox, widget=SizeBox_Mobile}
  Overlay_31 {
   widgetClassPath=&Overlay
   parent=SizeBox_Mobile
   widget=Overlay_31
   slot {class=SizeBoxSlot, properties {horizontalAlignment=HAlign_Center, verticalAlignment=VAlign_Center}}
  }
  Image_Safflower_TopNavIcon {
   widgetClassPath=&Image
   bIsVariable=true
   parent=Overlay_31
   widget=Image_Safflower_TopNavIcon
   properties {brush {
     bIsDynamicallyLoaded=false
     drawAs=Image
     imageType=FullColor
     mirroring=NoMirror
     resourceName=None
     resourceObject=/Safflower_SeasonData/UI/Materials/M_UI_Safflower_TopNavIcon
     tiling=NoTile
     imageSize=<44 44>
     margin=<ltrb 0 0 0 0>
     outlineSettings {
      bUseBrushTransparency=false
      roundingType=HalfHeightRadius
      width=0
      color {colorUseRule=UseColor_Specified, specifiedColor=<rgba 0 0 0 0>}
      cornerRadii=<0 0 0 0>
     }
     tintColor {colorUseRule=UseColor_Specified, specifiedColor=<rgba 1 1 1 1>}
     uVRegion {bIsValid=false, max=<0 0>, min=<0 0>}
    }}
   slot {class=OverlaySlot, properties {horizontalAlignment=HAlign_Center, verticalAlignment=VAlign_Center}}
  }
  InputActionWidget {widgetClassPath=&CommonActionWidget, bInherited=true, widget=null}
 ]
}
