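# Common UI button for a chevron pagination arrow in the Safflower season UI, with an input action icon.
# PreConstruct uses Flip to mirror the arrow overlay horizontally and switch the icon's alignment.
# Hover and unhover call PlayHover, which plays AnimHover forward or reversed at high speed unless Disabled.
# AnimHover is a null-default variable separate from the Anim_Hover squash animation, and the idle-animation switch is unconnected.
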WidgetBlueprint /Safflower_SeasonData/UI/WBP_Arrow_Pagination
animations [
 Anim_Hover {
  binding_count=1
  length_seconds=0.75
  object=&self:Anim_Hover
  sequence {bindings [LightningArrow {
     id=#1
     possessed_class=Image
     tracks [{
       type=MovieScene2DTransformTrack
       sections [{
         type=MovieScene2DTransformSection
         end_frame=15
         start_frame=0
         channels [
          Translation.X {type=MovieSceneScriptingFloatChannel, const [0,1]}
          Translation.Y {type=MovieSceneScriptingFloatChannel, const [0,1]}
          Angle {type=MovieSceneScriptingFloatChannel, const [0,1]}
          Scale.X {type=MovieSceneScriptingFloatChannel, const [1,1]}
          Scale.Y {type=MovieSceneScriptingFloatChannel, keys [[0,1],[15,0.9]]}
          Shear.X {type=MovieSceneScriptingFloatChannel, const [0,1]}
          Shear.Y {type=MovieSceneScriptingFloatChannel, const [0,1]}
         ]
        }]
      }]
    }]}
 }
 Anim_Clicked {
  binding_count=1
  length_seconds=0.75
  object=&self:Anim_Clicked
  sequence {bindings [LightningArrow {
     id=#2
     possessed_class=Image
     tracks [{
       type=MovieScene2DTransformTrack
       sections [{
         type=MovieScene2DTransformSection
         end_frame=15
         start_frame=0
         channels [
          Translation.X {type=MovieSceneScriptingFloatChannel, const [0,1]}
          Translation.Y {type=MovieSceneScriptingFloatChannel, const [0,1]}
          Angle {type=MovieSceneScriptingFloatChannel, const [0,1]}
          Scale.X {type=MovieSceneScriptingFloatChannel, const [1,1]}
          Scale.Y {type=MovieSceneScriptingFloatChannel, keys [[0,1],[15,0.9]]}
          Shear.X {type=MovieSceneScriptingFloatChannel, const [0,1]}
          Shear.Y {type=MovieSceneScriptingFloatChannel, const [0,1]}
         ]
        }]
      }]
    }]}
 }
 Anim_Disabled {binding_count=0, length_seconds=5, object=&self:Anim_Disabled, sequence {bindings []}}
 Anim_Normal {binding_count=0, length_seconds=5, object=&self:Anim_Normal, sequence {bindings []}}
 Anim_Idle {binding_count=0, length_seconds=5, object=&self:Anim_Idle, sequence {bindings []}}
]
graphs [EventGraph {nodes [
   n1 = PlayHover!() -> (OutputDelegate, then, PlayMode:EUMGSequencePlayMode Enum=Forward) {is=title K2Node_CustomEvent}
   > n2 = Branch(Condition=Get Disabled() -> (Disabled:bool))
   PlayAnimation(
    execute=@n2.else
    InAnimation=Get AnimHover() -> (AnimHover:Widget Animation Object Reference)
    PlayMode=@n1.PlayMode
    PlaybackSpeed=float / float(A=1, B=Make Literal Float(Value=0.1))
   )
   Event PreConstruct()
   > SetRenderScale!(
    execute
    self:Widget Object Reference=Get ArrowOverlay() -> (ArrowOverlay:Overlay Object Reference)
    Scale:Vector 2D Structure=Select(Option 0:Vector 2D Structure=(X=1,Y=1), Option 1:Vector 2D Structure=(X=-1,Y=1), Index=Get Flip() -> (Flip:bool)) -> (ReturnValue:Vector 2D Structure)
   ) -> (then)
   n3 = SlotAsOverlaySlot(Widget=Get CommonActionWidget_98() -> (CommonActionWidget_98:Common Action Widget Object Reference))
   > SetPadding!(
    execute
    self:Overlay Slot Object Reference=@n3
    InPadding_Left:float=0
    InPadding_Top:float=5
    InPadding_Right:float=0
    InPadding_Bottom:float=0
   ) -> (then)
   > SetHorizontalAlignment(
    self=@n3
    InHorizontalAlignment=Select(Option 0:EHorizontalAlignment Enum=HAlign_Left, Option 1:EHorizontalAlignment Enum=HAlign_Right, Index=Get Flip()) -> (ReturnValue:EHorizontalAlignment Enum)
   )
   > Branch(Condition=Get Disabled())
   Event On Hovered()
   > PlayHover!(execute, self, PlayMode:EUMGSequencePlayMode Enum) -> (then) {is=title K2Node_CallFunction}
   Event On Unhovered()
   > PlayHover(PlayMode=Reverse)
   n4 = Switch on EUMGSequencePlayMode!(execute, Selection:EUMGSequencePlayMode Enum=@n1.PlayMode) -> (Forward:Exec, Reverse:Exec, PingPong:Exec) {is=title K2Node_SwitchEnum}
   PlayAnimation(execute=@n4.Reverse, InAnimation=Get AnimIdle() -> (AnimIdle:Widget Animation Object Reference), PlaybackSpeed=2)
   Event Construct()
   Event Tick()
  ]}]
variables [
 AnimHover {type=WidgetAnimation, default=null}
 AnimIdle {type=WidgetAnimation, default=null}
 Disabled {type=boolean}
 Flip {type=boolean}
]
widgets {
 info {inheritedWidgetCount=1, parentClass=&CommonButtonBase, rootWidgetClass=&Overlay, widgetCount=8}
 widgets [
  Overlay {widgetClassPath=&Overlay, bIsVariable=true, widget=Overlay, properties {visibility=Visible}}
  ArrowOverlay {
   widgetClassPath=&Overlay
   bIsVariable=true
   parent=Overlay
   widget=ArrowOverlay
   slot {class=OverlaySlot, properties {horizontalAlignment=HAlign_Center, verticalAlignment=VAlign_Center}}
  }
  Spacer_34 {widgetClassPath=&Spacer, parent=ArrowOverlay, widget=Spacer_34, properties {size=<80 200>}, slot {class=OverlaySlot}}
  FortVisualAttachment_173 {
   widgetClassPath=&FortVisualAttachment
   parent=ArrowOverlay
   widget=FortVisualAttachment_173
   properties {contentAnchor=<0.5 0.5>}
   slot {class=OverlaySlot, properties {horizontalAlignment=HAlign_Center, verticalAlignment=VAlign_Center}}
  }
  Overlay_0 {widgetClassPath=&Overlay, parent=FortVisualAttachment_173, widget=Overlay_0, slot {class=SizeBoxSlot}}
  LightningArrow {
   widgetClassPath=&Image
   bIsVariable=true
   parent=Overlay_0
   widget=LightningArrow
   properties {brush {
     bIsDynamicallyLoaded=false
     drawAs=Image
     imageType=FullColor
     mirroring=NoMirror
     resourceName=None
     resourceObject=/Safflower_SeasonData/UI/Textures/T_UI_Safflower_Chevron
     tiling=NoTile
     imageSize=<58 120>
     margin=<ltrb 0 0 0 0>
     outlineSettings {
      bUseBrushTransparency=false
      roundingType=HalfHeightRadius
      width=0
      color {colorUseRule=UseColor_Specified, specifiedColor=<rgba 0 0 0 0>}
      cornerRadii=<0 0 0 0>
     }
     tintColor {colorUseRule=UseColor_Specified, specifiedColor=<rgba 1 1 1 1>}
     uVRegion {bIsValid=false, max=<0 0>, min=<0 0>}
    }}
   slot {class=OverlaySlot}
  }
  CommonActionWidget_98 {
   widgetClassPath=&CommonActionWidget
   bIsVariable=true
   parent=ArrowOverlay
   widget=CommonActionWidget_98
   properties {
    iconRimBrush {
     bIsDynamicallyLoaded=false
     drawAs=NoDrawType
     imageType=NoImage
     mirroring=NoMirror
     resourceName=None
     resourceObject=null
     tiling=NoTile
     imageSize=<44 44>
     margin=<ltrb 0 0 0 0>
     outlineSettings {
      bUseBrushTransparency=false
      roundingType=HalfHeightRadius
      width=0
      color {colorUseRule=UseColor_Specified, specifiedColor=<rgba 0 0 0 0>}
      cornerRadii=<0 0 0 0>
     }
     tintColor {colorUseRule=UseColor_Specified, specifiedColor=<rgba 1 1 1 1>}
     uVRegion {bIsValid=false, max=<0 0>, min=<0 0>}
    }
    renderTransform {angle=0, scale=<1 1>, shear=<0 0>, translation=<0 24>}
   }
   slot {class=OverlaySlot, properties {verticalAlignment=VAlign_Center, padding=<ltrb 0 5 0 0>}}
  }
  InputActionWidget {widgetClassPath=&CommonActionWidget, bInherited=true, widget=null}
 ]
}
