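# UI-domain translucent material for the short reward line in the Safflower season UI, judging by its name.
# It scales texture coordinates by constant U and V tiling factors and passes them through a Panner into a sample of the Safflower reward frame texture.
# The Panner has no speed set, and a separate warm yellow color constant sits unwired to the other nodes.
# The graph shows no links to the material outputs.
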Material /Safflower_SeasonData/UI/Materials/M_UI_Safflower_RewardLine_Short
graph {
 properties {
  bAutomaticallySetUsageInEditor=false
  bUsedWithStaticMesh=false
  blendMode=BLEND_Translucent
  materialDomain=MD_UI
  usedShadingModels=Default Lit
 }
 nodes [
  Constant3Vector() {constant=<rgba 0.956 0.7758 1 1>}
  n1 = ComponentMask(None=TextureCoordinate()) {g=true, r=true}
  n2 = AppendVector(A=Constant() {r=3}, B=Constant() {r=5})
  TextureSample(UVs=Panner(Coordinate=Multiply(A=@n1, B=@n2))) {texture=/Safflower_SeasonData/UI/Textures/T_UI_Safflower_RewardFrame}
 ]
}
