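# Plus/minus button widget for the Safflower season UI, built on CommonButtonBase with a minus sign as its default label.
# An input action prompt overlay with negative padding sits beside a centered, user-scaled ScaleBox named for mobile.
# The ScaleBox layers the label, a Spacer that sets minimum size, and a right-aligned, rotated reward-line material image as a slanted separator.
# The event graph holds only unconnected stub events, with no logic or variables.
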WidgetBlueprint /Safflower_SeasonData/UI/WBP_PlusMinusButton_Safflower
graphs [EventGraph {nodes [
   Event PreConstruct()
   Event Construct()
   Event Tick()
  ]}]
variables []
widgets {
 info {inheritedWidgetCount=1, parentClass=&CommonButtonBase, rootWidgetClass=&Overlay, widgetCount=8}
 widgets [
  Overlay_31 {widgetClassPath=&Overlay, widget=Overlay_31}
  Overlay_ButtonPrompt {
   widgetClassPath=&Overlay
   parent=Overlay_31
   widget=Overlay_ButtonPrompt
   slot {class=OverlaySlot, properties {verticalAlignment=VAlign_Center, padding=<ltrb -50 -16 -50 -16>}}
  }
  InputActionWidget {
   widgetClassPath=&CommonActionWidget
   bInherited=true
   bIsVariable=true
   parent=Overlay_ButtonPrompt
   widget=InputActionWidget
   slot {class=OverlaySlot}
  }
  ScaleBox_Mobile {
   widgetClassPath=&ScaleBox
   parent=Overlay_31
   widget=ScaleBox_Mobile
   properties {stretch=UserSpecified}
   slot {class=OverlaySlot, properties {horizontalAlignment=HAlign_Center, verticalAlignment=VAlign_Center}}
  }
  Overlay_105 {widgetClassPath=&Overlay, parent=ScaleBox_Mobile, widget=Overlay_105, slot {class=ScaleBoxSlot}}
  ButtonText {
   widgetClassPath=&CommonTextBlock
   parent=Overlay_105
   widget=ButtonText
   properties {
    text=-
    font {
     bForceMonospaced=false
     bMaterialIsStencil=false
     fontMaterial=null
     fontObject=&BurbankSmall
     letterSpacing=0
     monospacedWidth=1
     size=54
     skewAmount=0
     typefaceFontName=Medium
     outlineSettings {
      bApplyOutlineToDropShadows=false
      bMiteredCorners=false
      bSeparateFillAlpha=false
      outlineMaterial=null
      outlineSize=0
      outlineColor=<rgba 0 0 0 1>
     }
    }
    shadowOffset=<0 0>
   }
   slot {class=OverlaySlot, properties {horizontalAlignment=HAlign_Center, verticalAlignment=VAlign_Center}}
  }
  Spacer_31 {widgetClassPath=&Spacer, parent=Overlay_105, widget=Spacer_31, properties {size=<64 65>}, slot {class=OverlaySlot}}
  Image_Separator {
   widgetClassPath=&Image
   bIsVariable=true
   parent=Overlay_105
   widget=Image_Separator
   properties {
    brush {
     bIsDynamicallyLoaded=false
     drawAs=Image
     imageType=FullColor
     mirroring=NoMirror
     resourceName=None
     resourceObject=/Safflower_SeasonData/UI/Materials/M_UI_Safflower_RewardLine_Short
     tiling=NoTile
     imageSize=<256.5029 15.755>
     margin=<ltrb 0 0 0 0>
     outlineSettings {
      bUseBrushTransparency=false
      roundingType=HalfHeightRadius
      width=0
      color {colorUseRule=UseColor_Specified, specifiedColor=<rgba 0 0 0 0>}
      cornerRadii=<0 0 0 0>
     }
     tintColor {colorUseRule=UseColor_Specified, specifiedColor=<rgba 1 1 1 1>}
     uVRegion {bIsValid=false, max=<0 0>, min=<0 0>}
    }
    renderTransform {angle=74, scale=<1 1>, shear=<0 0>, translation=<0 0>}
   }
   slot {class=OverlaySlot, properties {horizontalAlignment=HAlign_Right, verticalAlignment=VAlign_Center}}
  }
 ]
}
