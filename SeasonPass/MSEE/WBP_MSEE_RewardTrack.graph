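# WBP_MSEE_RewardTrack is a widget showing the MSEE season pass reward track as item cards in a Dynamic Entry Box.
# PopulateItemcards builds cards from the first track's rewards, optionally hiding owned ones, with widths scaled by tile size.
# Owned cards show full progress, the first unowned shows wallet currency progress since the previous threshold, and later ones show none.
# It relays card hover and click through dispatchers, offers item lookup helpers, and can populate fake preview entries.
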WidgetBlueprint /SeasonPass/MSEE/WBP_MSEE_RewardTrack
graphs [
 EventGraph {nodes [
   n1 = OnItemcardHovered!() -> (OutputDelegate, then, RewardVM:Season Pass Reward VM Object Reference, FortItemVM:Fort Item VM Object Reference) {is=title K2Node_CustomEvent}
   > Call On Item Hovered!(
    execute:Exec
    self:WBP MSEE Reward Track Object Reference
    RewardVM:Season Pass Reward VM Object Reference=@n1.RewardVM
    FortItemVM:Fort Item VM Object Reference=@n1.FortItemVM
    ItemIndex:int=Find Item!(
     TargetArray:Array of Season Pass Reward VM Object References=Get RewardVMs() -> (RewardVMs:Array of Season Pass Reward VM Object References)
     ItemToFind:"Season Pass Reward VM Object Reference (by ref)"=@n1.RewardVM
    ) -> (ReturnValue)
   ) -> (then:Exec) {is=title K2Node_CallDelegate}
   n2 = ItemcardClicked!() -> (OutputDelegate, then, RewardVM:Season Pass Reward VM Object Reference, FortItemVM:Fort Item VM Object Reference) {is=title K2Node_CustomEvent}
   > Call On Item Clicked!(
    execute:Exec
    self:WBP MSEE Reward Track Object Reference
    RewardVM:Season Pass Reward VM Object Reference=@n2.RewardVM
    ItemVM:Fort Item VM Object Reference=@n2.FortItemVM
   ) -> (then:Exec) {is=title K2Node_CallDelegate}
   Event PreConstruct()
  ]}
 FindItemByIndex {nodes [
   n1 = FindItemByIndex!() -> (then, IndexToFind:int) {is=title K2Node_FunctionEntry}
   n2 = "Get (a copy)"(Array:Array of Season Pass Reward VM Object References=Get RewardVMs(), Dimension 1=@n1) -> (Output:Season Pass Reward VM Object Reference)
   > Return Node!(
    execute
    SeasonPassRewardVM:Season Pass Reward VM Object Reference=@n2
    FortItemVM:Fort Item VM Object Reference="Get (a copy)"(Array:Array of Fort Item VM Object References=Get ItemVMs!(self:Season Pass Reward VM Object Reference=@n2) -> (ItemVMs:Array of Fort Item VM Object References)) -> (Output:Fort Item VM Object Reference)
   )
  ]}
 FindItemVM {nodes [
   n1 = FindItemVM!() -> (then, ItemToFind:"Season Pass Reward VM Object Reference (by ref)") {is=title K2Node_FunctionEntry}
   > n2 = FindItemByIndex!(execute, self, IndexToFind:int=Find Item(TargetArray=Get RewardVMs(), ItemToFind=@n1)) -> (then, SeasonPassRewardVM:Season Pass Reward VM Object Reference, FortItemVM:Fort Item VM Object Reference) {is=title K2Node_CallFunction}
   > Return Node!(execute, Output:Fort Item VM Object Reference=@n2.FortItemVM)
  ]}
 OnItemClicked {nodes [OnItemClicked!() -> (then, RewardVM:Season Pass Reward VM Object Reference, ItemVM:Fort Item VM Object Reference) {is=title K2Node_FunctionEntry}]}
 OnItemHovered {nodes [OnItemHovered!() -> (then, RewardVM:Season Pass Reward VM Object Reference, FortItemVM:Fort Item VM Object Reference, ItemIndex:int) {is=title K2Node_FunctionEntry}]}
 OnRewardTrackPopulated {nodes [OnRewardTrackPopulated!() -> (then, RewardTrack:Dynamic Entry Box Object Reference) {is=title K2Node_FunctionEntry}]}
 PopulateFakeEntries {nodes [
   n1 = PopulateFakeEntries!() -> (then, Rewards:Array of Season Pass Track VM Object References) {is=title K2Node_FunctionEntry}
   n2 = Get RewardTrack() -> (RewardTrack:Dynamic Entry Box Object Reference)
   > Reset!(execute, self:Dynamic Entry Box Object Reference=@n2, bDeleteWidgets:bool=true) -> (then)
   > n3 = For Each Loop(Array:Array of Season Pass Track VM Object References=@n1) -> (Array Element:Season Pass Track VM Object Reference)
   > n4 = Create Entry of Class!(
    execute
    self:Dynamic Entry Box Object Reference=@n2
    EntryClass:User Widget Class Reference=Get MSEE Itemcard() -> (MSEE Itemcard:User Widget Class Reference)
   ) -> (then, ReturnValue:User Widget Object Reference)
   > n5 = Cast To WBP_MSEE_Itemcard!(execute, Object=@n4) -> (then, CastFailed, AsWBP MSEE Itemcard:WBP MSEE Itemcard Object Reference) {is=title K2Node_DynamicCast}
   n6 = Last Index!(TargetArray:Array of Season Pass Track VM Object References=Get Rewards() -> (Rewards:Array of Season Pass Track VM Object References)) -> (ReturnValue=0)
   n7 = Round(A=SafeDivide(A="To Float (Integer)"(InInt=@n6), B=2)) {is=title K2Node_CallFunction}
   n8 = "integer > integer"(A=@n3.Array Index, B=@n7)
   > Set ProgressPercentPreview!(
    execute
    ProgressPercentPreview:double=SelectFloat(A=Get ProgressPreview() -> (ProgressPreview:double), B=SelectFloat(A=0, B=1, bPickA=@n8), bPickA="Equal (Integer)"(A=@n3.Array Index, B=@n7))
    self:WBP MSEE Itemcard Object Reference=@n5
   ) -> (then, Output_Get:double)
  ]}
 PopulateItemcards {nodes [
   n1 = PopulateItemcards!() -> (then, Rewards:Array of Season Pass Track VM Object References) {is=title K2Node_FunctionEntry}
   > UnbindEvents()
   n2 = Get RewardTrack()
   > Reset(self=@n2, bDeleteWidgets=true)
   n3 = Get Rewards!(self:Season Pass Track VM Object Reference="Get (a copy)"(Array:Array of Season Pass Track VM Object References=@n1) -> (Output:Season Pass Track VM Object Reference)) -> (Rewards:Array of Season Pass Reward VM Object References)
   > Set RewardVMs(RewardVMs:Array of Season Pass Reward VM Object References=@n3) -> (Output_Get:Array of Season Pass Reward VM Object References)
   > n4 = For Each Loop(Array:Array of Season Pass Reward VM Object References=@n3) -> (Array Element:Season Pass Reward VM Object Reference)
   > n5 = Branch(Condition=Get HideOwnedItems() -> (HideOwnedItems:bool))
   > n6 = Branch(Condition=NOT Boolean(A=Get bIsRewardOwned!(self:Season Pass Reward VM Object Reference=@n4) -> (bIsRewardOwned:bool)))
   n7 = Create Entry of Class(execute=[@n6.then, @n5.else], self=@n2, EntryClass=Get MSEE Itemcard())
   > n8 = Cast To WBP_MSEE_Itemcard(Object=@n7)
   > SetSeasonPassRewardVM!(execute, self:WBP MSEE Itemcard Object Reference=@n8, Viewmodel:Season Pass Reward VM Object Reference=@n4) -> (then)
   n9 = Get TileSize!(self:Athena Season Item Entry Base Object Reference=GetReward!(self:Season Pass Reward VM Object Reference=@n4) -> (ReturnValue:Athena Season Item Entry Base Object Reference)) -> (TileSize:EPageItemTileSize Enum=Size_1_x_1)
   > SetItemCardSize!(
    execute
    self:WBP UIKit Item Card Base Object Reference=@n8
    ItemSize_X:double=float * float(
     A=Get BaseTileSizeX() -> (BaseTileSizeX:double)
     B=Select!(Size_1_x_1:double=1, Size_2_x_1:double=2, Size_2_x_2:double=2, Size_3_x_2:double=3, Size_3_x_3:double=3, Index:EPageItemTileSize Enum=@n9) -> (ReturnValue:double)
    ) {is=title K2Node_PromotableOperator}
    ItemSize_Y:double=144
   ) -> (then)
   > n10 = Branch(Condition=Get bIsRewardOwned(self=@n4))
   > n11 = Set with Broadcast ProgressPercent!(execute, ProgressPercent:double=1, self:WBP MSEE Itemcard Object Reference=@n8) -> (then, Output_Get:double)
   Bind Event to Item Clicked(execute=[@n16.then, @n11.then, @n13.then], self:WBP MSEE Itemcard Object Reference=@n8, Delegate=Create Event()) {is=title K2Node_AddDelegate}
   > Bind Event to Item Hovered(self:WBP MSEE Itemcard Object Reference=@n8, Delegate=Create Event()) {is=title K2Node_AddDelegate}
   > Set PreviousLevelToUnlock(PreviousLevelToUnlock:int=Get ProgressCountForUnlock!(self:Season Pass Reward VM Object Reference=Get SeasonPassRewardVM!(self:WBP MSEE Itemcard Object Reference=@n8) -> (SeasonPassRewardVM:Season Pass Reward VM Object Reference)) -> (ProgressCountForUnlock:int=0)) -> (Output_Get:int=0)
   n12 = Branch(execute=@n10.else, Condition=Get FirstObtainablePassed() -> (FirstObtainablePassed:bool))
   > n13 = Set with Broadcast ProgressPercent(ProgressPercent=0, self=@n8)
   n14 = int - int(A=Get ProgressCountForUnlock(self=@n4) -> (ProgressCountForUnlock=0), B=Get PreviousLevelToUnlock() -> (PreviousLevelToUnlock:int=0))
   n15 = int - int(
    A=Get CurrentCurrency!(self:MSEEWallet VM Object Reference=Get MSEEWalletVM() -> (MSEEWalletVM:MSEEWallet VM Object Reference)) -> (CurrentCurrency:int=0)
    B=Get PreviousLevelToUnlock() -> (PreviousLevelToUnlock=0)
   )
   Set with Broadcast ProgressPercent(execute=@n12.else, ProgressPercent=SafeDivide(A="To Float (Integer)"(InInt=@n15), B="To Float (Integer)"(InInt=@n14)), self=@n8)
   > n16 = Set FirstObtainablePassed(FirstObtainablePassed:bool=true) -> (Output_Get:bool)
   Call On Reward Track Populated!(execute:Exec=@n4.Completed, self:WBP MSEE Reward Track Object Reference, RewardTrack:Dynamic Entry Box Object Reference=Get RewardTrack()) -> (then:Exec) {is=title K2Node_CallDelegate}
  ]}
 UnbindEvents {nodes [
   UnbindEvents() {is=title K2Node_FunctionEntry}
   > n1 = For Each Loop(Array:Array of User Widget Object References=GetAllEntries!(self:Dynamic Entry Box Base Object Reference=Get RewardTrack()) -> (ReturnValue:Array of User Widget Object References)) -> (Array Element:User Widget Object Reference)
   Return Node!(execute=@n1.Completed)
   > n2 = Cast To WBP_MSEE_Itemcard(Object=@n1)
   > Unbind Event from Item Clicked(self:WBP MSEE Itemcard Object Reference=@n2, Delegate=Create Event()) {is=title K2Node_RemoveDelegate}
   > Unbind Event from Item Hovered(self:WBP MSEE Itemcard Object Reference=@n2, Delegate=Create Event()) {is=title K2Node_RemoveDelegate}
  ]}
]
variables [
 BaseTileSizeX {type=number, default=112}
 HideOwnedItems {type=boolean}
 InspectScreen {type=Class, default=&/WBP_MSEE_InspectScreen.WBP_MSEE_InspectScreen_C}
 MSEE Itemcard {type=Class, default=&/WBP_MSEE_Itemcard.WBP_MSEE_Itemcard_C}
 OnItemClicked {}
 OnItemHovered {}
 OnRewardTrackPopulated {}
 ProgressPreview {type=number, default=0.5}
 RewardVMs {type="array<SeasonPassRewardVM>", default []}
]
widgets {
 info {inheritedWidgetCount=0, parentClass=&UserWidget, rootWidgetClass=&Overlay, widgetCount=3}
 widgets [
  Overlay_0 {widgetClassPath=&Overlay, widget=Overlay_0}
  TrackBackground_Slot {
   widgetClassPath=&NamedSlot
   bIsVariable=true
   parent=Overlay_0
   widget=TrackBackground_Slot
   slot {class=OverlaySlot, properties {horizontalAlignment=HAlign_Fill, verticalAlignment=VAlign_Fill}}
  }
  RewardTrack {
   widgetClassPath=&DynamicEntryBox
   bIsVariable=true
   parent=Overlay_0
   widget=RewardTrack
   properties {
    entryWidgetClass=&WBP_MSEE_Itemcard_C
    navigation=&self:WidgetTree.RewardTrack.WidgetNavigation_1
    numDesignerPreviewEntries=12
    entrySpacing=<16 16>
   }
   slot {class=OverlaySlot, properties {horizontalAlignment=HAlign_Fill}}
  }
 ]
}
