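# Safflower season variant of the MSEE reward track widget, derived from WBP_MSEE_RewardTrack in the SeasonPass content.
# It adds an Overlay with a solid pink-tinted fill Image, apparently as content for a named slot of the parent.
# PreConstruct calls the parent, then branches on design time to populate fake entries from a twelve-element array of unassigned track view models for editor preview.
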WidgetBlueprint /Safflower_SeasonData/UI/WBP_Safflower_RewardTrack
graphs [EventGraph {nodes [
   n1 = Event PreConstruct()
   > Parent: PreConstruct(IsDesignTime=@n1.IsDesignTime)
   > Branch(Condition=@n1.IsDesignTime)
   > PopulateFakeEntries!(
    execute
    self:WBP MSEE Reward Track Object Reference
    Rewards:Array of Season Pass Track VM Object References=Make Array!(
     "[0]":Season Pass Track VM Object Reference
     "[1]":Season Pass Track VM Object Reference
     "[2]":Season Pass Track VM Object Reference
     "[3]":Season Pass Track VM Object Reference
     "[4]":Season Pass Track VM Object Reference
     "[5]":Season Pass Track VM Object Reference
     "[6]":Season Pass Track VM Object Reference
     "[7]":Season Pass Track VM Object Reference
     "[8]":Season Pass Track VM Object Reference
     "[9]":Season Pass Track VM Object Reference
     "[10]":Season Pass Track VM Object Reference
     "[11]":Season Pass Track VM Object Reference
    ) -> (Array:Array of Season Pass Track VM Object References)
   ) -> (then)
   Event Construct()
   Event Tick()
  ]}]
variables []
widgets {
 info {
  inheritedWidgetCount=0
  namedSlotCount=1
  parentClass=/SeasonPass/MSEE/WBP_MSEE_RewardTrack.WBP_MSEE_RewardTrack_C
  rootWidgetClass=null
  widgetCount=2
 }
 widgets [
  Overlay_78 {widgetClassPath=&Overlay, widget=Overlay_78}
  Image_51 {
   widgetClassPath=&Image
   bIsVariable=true
   parent=Overlay_78
   widget=Image_51
   properties {brush {
     bIsDynamicallyLoaded=false
     drawAs=Image
     imageType=NoImage
     mirroring=NoMirror
     resourceName=None
     resourceObject=null
     tiling=NoTile
     imageSize=<32 32>
     margin=<ltrb 0 0 0 0>
     outlineSettings {
      bUseBrushTransparency=false
      roundingType=HalfHeightRadius
      width=0
      color {colorUseRule=UseColor_Specified, specifiedColor=<rgba 0 0 0 0>}
      cornerRadii=<0 0 0 0>
     }
     tintColor {colorUseRule=UseColor_Specified, specifiedColor=<rgba 1 0.1801 0.7156 1>}
     uVRegion {bIsValid=false, max=<0 0>, min=<0 0>}
    }}
   slot {class=OverlaySlot, properties {horizontalAlignment=HAlign_Fill, verticalAlignment=VAlign_Fill}}
  }
 ]
}
