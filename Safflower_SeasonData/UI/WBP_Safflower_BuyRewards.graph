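# Safflower season pass screen for buying rewards, a child of WBP_MSEE_BuyRewards that supplies the Safflower reward track widget.
# When that track is populated, its entry box is passed to SetRewardTrack.
# The intro waits one tick, then picks a from-main or from-inspect animation using FromInspectScreen.
# Both animations drive only MPC_Safflower_BackgroundStates parameters for cloud height, ground height and ground hiding, animating the background through the material collection.
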WidgetBlueprint /Safflower_SeasonData/UI/WBP_Safflower_BuyRewards
animations [
 S_Intro_FromInspect {
  binding_count=0
  length_seconds=0.3
  object=&self:S_Intro_FromInspect
  sequence {
   bindings []
   tracks [{
     type=MovieSceneMaterialParameterCollectionTrack
     display_name=MPC_Safflower_BackgroundStates
     sections [{
       type=MovieSceneParameterSection
       channels [
        Anim_CloudHeight {type=MovieSceneScriptingFloatChannel, const [1,2]}
        Anim_GroundHeight {type=MovieSceneScriptingFloatChannel, const [1,2]}
        Anim_HideGround {type=MovieSceneScriptingFloatChannel, keys [[0,1],[6,0]]}
       ]
      }]
    }]
  }
 }
 S_Intro_FromMain {
  binding_count=0
  length_seconds=0.3
  object=&self:S_Intro_FromMain
  sequence {
   bindings []
   tracks [{
     type=MovieSceneMaterialParameterCollectionTrack
     display_name=MPC_Safflower_BackgroundStates
     sections [{
       type=MovieSceneParameterSection
       channels [
        Anim_CloudHeight {type=MovieSceneScriptingFloatChannel, keys [[0,0],[4,1]]}
        Anim_GroundHeight {type=MovieSceneScriptingFloatChannel, keys [[0,0],[6,1]]}
        Anim_HideGround {type=MovieSceneScriptingFloatChannel, const [0,2]}
       ]
      }]
    }]
  }
 }
]
graphs [EventGraph {nodes [
   n1 = "On Reward Track Populated (RewardTrack_A)"!() -> (OutputDelegate, then, RewardTrack:Dynamic Entry Box Object Reference) {is=title K2Node_ComponentBoundEvent}
   > SetRewardTrack!(execute, self:WBP MSEE Buy Rewards Object Reference, RewardTrack:Dynamic Entry Box Object Reference=@n1.RewardTrack) -> (then)
   Event PreConstruct()
   > Parent: PreConstruct(IsDesignTime=false)
   Event On Activated()
   > Parent: On Activated() {is=title K2Node_CallParentFunction}
   Event PlayIntroAnimation() {is=title K2Node_Event}
   > DelayUntilNextTick!(execute) -> (then)
   > PlayAnimationForward(InAnimation=Select(
     Option 0:Widget Animation Object Reference=Get S_Intro_FromMain() -> (S_Intro_FromMain:Widget Animation Object Reference)
     Option 1:Widget Animation Object Reference=Get S_Intro_FromInspect() -> (S_Intro_FromInspect:Widget Animation Object Reference)
     Index=Get FromInspectScreen() -> (FromInspectScreen:bool)
    ) -> (ReturnValue:Widget Animation Object Reference))
  ]}]
variables []
widgets {
 info {
  inheritedWidgetCount=0
  namedSlotCount=1
  parentClass=/SeasonPass/MSEE/WBP_MSEE_BuyRewards.WBP_MSEE_BuyRewards_C
  rootWidgetClass=null
  widgetCount=1
 }
 widgets [RewardTrack_A {widgetClassPath=&/WBP_Safflower_RewardTrack, bIsVariable=true, widget=RewardTrack_A}]
}
