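# Season key art widget for Safflower, layering two character images (MorningSoak and OzoneCredit) over a collapsed CMS main image.
# CharacterReveal staggers them apart with slide, tilt, fade and animated material overlay and drop shadow.
# UpdateImageRefs passes each character image to SetImage, which looks up its key in the supplied image table, falling back to an error icon.
# The event graph is empty, so nothing here plays the animation or calls the update.
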WidgetBlueprint /Safflower_SeasonData/UI/WBP_Safflower_KeyArt
animations [CharacterReveal {
  binding_count=2
  length_seconds=0.6
  object=&self:CharacterReveal
  sequence {bindings [
    KeyArt_OzoneCredit {
     id=#1
     possessed_class=FortLazyImage
     tracks [
      {
       type=MovieScene2DTransformTrack
       sections [{
         type=MovieScene2DTransformSection
         end_frame=10
         start_frame=3
         channels [
          Translation.X {type=MovieSceneScriptingFloatChannel, keys [[3,-24.7329],[10,-105.4458]]}
          Translation.Y {type=MovieSceneScriptingFloatChannel, const [70,2]}
          Angle {type=MovieSceneScriptingFloatChannel, keys [[3,10],[8,-2]]}
          Scale.X {type=MovieSceneScriptingFloatChannel, const [0.87,2]}
          Scale.Y {type=MovieSceneScriptingFloatChannel, const [0.87,1]}
          Shear.X {type=MovieSceneScriptingFloatChannel, const [0,1]}
          Shear.Y {type=MovieSceneScriptingFloatChannel, const [0,1]}
         ]
        }]
      }
      {
       type=MovieSceneColorTrack
       display_name=Color and Opacity
       sections [{
         type=MovieSceneColorSection
         end_frame=10
         start_frame=3
         channels [
          Color.R {type=MovieSceneScriptingFloatChannel, default=1}
          Color.G {type=MovieSceneScriptingFloatChannel, default=1}
          Color.B {type=MovieSceneScriptingFloatChannel, default=1}
          Color.A {type=MovieSceneScriptingFloatChannel, keys [[3,0],[5,1],[10,1]]}
         ]
        }]
      }
      {
       type=MovieSceneWidgetMaterialTrack
       sections [{
         type=MovieSceneParameterSection
         end_frame=12
         start_frame=3
         channels [
          OverlayIntensity {type=MovieSceneScriptingFloatChannel, keys [[3,1],[12,0.9698]]}
          WhiteToColourLerpAnim {type=MovieSceneScriptingFloatChannel, const [1,3]}
          OverlayPosition {type=MovieSceneScriptingFloatChannel, keys [[3,1],[12,0.5533]]}
          OverlayWidth {type=MovieSceneScriptingFloatChannel, keys [[3,0.2584],[12,0.106]]}
          OverlayAngle {type=MovieSceneScriptingFloatChannel, const [0.25,1]}
          DropShadowPanX {type=MovieSceneScriptingFloatChannel, keys [[3,8],[11,0]]}
          DropShadowPanY {type=MovieSceneScriptingFloatChannel, const [0,1]}
          DropShadow/OverlayColourR {type=MovieSceneScriptingFloatChannel, keys [[3,1],[12,0.0208]]}
          DropShadow/OverlayColourG {type=MovieSceneScriptingFloatChannel, keys [[3,1],[12,0.0068]]}
          DropShadow/OverlayColourB {type=MovieSceneScriptingFloatChannel, keys [[3,1],[12,0.0204]]}
          DropShadow/OverlayColourA {type=MovieSceneScriptingFloatChannel, const [1,2]}
         ]
        }]
      }
      {
       type=MovieSceneFloatTrack
       display_name=Render Opacity
       sections [{type=MovieSceneFloatSection, end_frame=4, start_frame=0, channels [None {type=MovieSceneScriptingFloatChannel, keys [[0,0],[2,0],[3,1]]}]}]
      }
     ]
    }
    KeyArt_MorningSoak {
     id=#2
     possessed_class=FortLazyImage
     tracks [
      {
       type=MovieScene2DTransformTrack
       sections [{
         type=MovieScene2DTransformSection
         end_frame=7
         start_frame=0
         channels [
          Translation.X {type=MovieSceneScriptingFloatChannel, keys [[0,15],[7,117.1163]]}
          Translation.Y {type=MovieSceneScriptingFloatChannel, const [55,1]}
          Angle {type=MovieSceneScriptingFloatChannel, keys [[0,-10],[5,0]]}
          Scale.X {type=MovieSceneScriptingFloatChannel, keys [[0,0.85],[4,0.95]]}
          Scale.Y {type=MovieSceneScriptingFloatChannel, keys [[0,1],[4,0.95]]}
          Shear.X {type=MovieSceneScriptingFloatChannel, const [0,1]}
          Shear.Y {type=MovieSceneScriptingFloatChannel, const [0,1]}
         ]
        }]
      }
      {
       type=MovieSceneColorTrack
       display_name=Color and Opacity
       sections [{
         type=MovieSceneColorSection
         end_frame=7
         start_frame=0
         channels [
          Color.R {type=MovieSceneScriptingFloatChannel, const [1,1]}
          Color.G {type=MovieSceneScriptingFloatChannel, const [1,1]}
          Color.B {type=MovieSceneScriptingFloatChannel, const [1,1]}
          Color.A {type=MovieSceneScriptingFloatChannel, keys [[0,0],[2,1],[7,1]]}
         ]
        }]
      }
      {
       type=MovieSceneWidgetMaterialTrack
       sections [{
         type=MovieSceneParameterSection
         end_frame=12
         start_frame=0
         channels [
          OverlayIntensity {type=MovieSceneScriptingFloatChannel, keys [[0,1],[9,0.9602],[12,0.9202]]}
          WhiteToColourLerpAnim {type=MovieSceneScriptingFloatChannel, const [1,3]}
          OverlayPosition {type=MovieSceneScriptingFloatChannel, keys [[0,1],[9,0.4813]]}
          OverlayWidth {type=MovieSceneScriptingFloatChannel, keys [[0,0.2584],[9,0.106]]}
          OverlayAngle {type=MovieSceneScriptingFloatChannel, const [0.25,2]}
          DropShadowPanX {type=MovieSceneScriptingFloatChannel, keys [[0,-8],[8,0]]}
          DropShadowPanY {type=MovieSceneScriptingFloatChannel, const [0,1]}
          DropShadow/OverlayColourR {type=MovieSceneScriptingFloatChannel, keys [[0,1],[9,0.026]]}
          DropShadow/OverlayColourG {type=MovieSceneScriptingFloatChannel, keys [[0,1],[9,0.0088]]}
          DropShadow/OverlayColourB {type=MovieSceneScriptingFloatChannel, keys [[0,1],[9,0.026]]}
          DropShadow/OverlayColourA {type=MovieSceneScriptingFloatChannel, const [1,1]}
         ]
        }]
      }
     ]
    }
   ]}
 }]
graphs [
 EventGraph {nodes [
   Event PreConstruct()
   Event Construct()
   Event Tick()
  ]}
 SetImage {nodes [
   n1 = SetImage!() -> (
    then
    ImageObject:Fort Lazy Image Object Reference
    LookupName:String
    ImageTable:"Fort Image Reference Table Structure (by ref)"
    Image Size:Vector 2D Structure
   ) {is=title K2Node_FunctionEntry}
   n2 = TryLookupImage!(Table:"Fort Image Reference Table Structure (by ref)"=@n1.ImageTable, Name:String=@n1.LookupName) -> (OutImageReference:Fort Image Reference Structure, ReturnValue:bool)
   > SetBrushFromImageReference!(
    execute
    self:Fort Lazy Image Object Reference=@n1
    ImageReference:"Fort Image Reference Structure (by ref)"=Select(
     Option 0:"Fort Image Reference Structure (by ref)"=Make Fort Image Reference!(
      Image:Soft Object Reference=/FortUILibrary/Textures/MaterialInstances/MI_UIKit_Icon_ImageError
      ImageUrl:String
      MaterialTextureParamName:Name=None
     ) -> (FortImageReference:Fort Image Reference Structure) {is=title K2Node_MakeStruct}
     Option 1:"Fort Image Reference Structure (by ref)"=@n2
     Index=@n2.ReturnValue
    ) -> (ReturnValue:"Fort Image Reference Structure (by ref)")
    bMatchSize:bool=false
    DownloadPriority:ECMSDownloadPriority Enum=Normal
   ) -> (then)
  ]}
 UpdateImageRefs {nodes [
   UpdateImageRefs!() -> (then, ImageTable:"Fort Image Reference Table Structure (by ref)") {is=title K2Node_FunctionEntry}
   n1 = Get KeyArt_OzoneCredit() -> (KeyArt_OzoneCredit:Fort Lazy Image Object Reference)
   > SetImage!(
    execute
    self
    ImageObject:Fort Lazy Image Object Reference=@n1
    LookupName:String=KeyArt_Main_1
    ImageTable:"Fort Image Reference Table Structure (by ref)"=Get ImageTable() -> (ImageTable:"Fort Image Reference Table Structure (by ref)")
    Image Size:Vector 2D Structure=Get!(self:Image Object Reference=@n1) -> (
     Brush_TintColor:Slate Color Structure
     Brush_DrawAs:ESlateBrushDrawType Enum
     Brush_Tiling:ESlateBrushTileType Enum
     Brush_Mirroring:ESlateBrushMirrorType Enum
     Brush_ImageSize:Vector 2D Structure
     Brush_Margin:Margin Structure
     Brush_ResourceObject:Object Reference
     Brush_OutlineSettings:Slate Brush Outline Settings Structure
    ).Brush_ImageSize
   ) -> (then) {is=title K2Node_CallFunction}
   n2 = Get KeyArt_MorningSoak() -> (KeyArt_MorningSoak:Fort Lazy Image Object Reference)
   > SetImage(ImageObject=@n2, LookupName=KeyArt_Main_2, ImageTable=Get ImageTable(), Image Size=Get(self=@n2).Brush_ImageSize)
  ]}
]
variables []
widgets {
 info {inheritedWidgetCount=0, parentClass=&UserWidget, rootWidgetClass=&GridPanel, widgetCount=5}
 widgets [
  GridPanel_23 {
   widgetClassPath=&GridPanel
   widget=GridPanel_23
   properties {columnFill [1], renderTransform {angle=0, scale=<0.95 0.95>, shear=<0 0>, translation=<0 0>}, renderTransformPivot=<0.5 1>, rowFill [1]}
  }
  Overlay_49 {
   widgetClassPath=&Overlay
   parent=GridPanel_23
   widget=Overlay_49
   properties {renderTransform {angle=0, scale=<2.1 2.1>, shear=<0 0>, translation=<0 -25>}}
   slot {class=GridSlot, properties {horizontalAlignment=HAlign_Center, verticalAlignment=VAlign_Bottom}}
  }
  CMS_KeyArt_Main {
   widgetClassPath=&FortLazyImage
   bIsVariable=true
   parent=Overlay_49
   widget=CMS_KeyArt_Main
   properties {
    visibility=Collapsed
    brush {
     bIsDynamicallyLoaded=false
     drawAs=Image
     imageType=FullColor
     mirroring=NoMirror
     resourceName=None
     resourceObject=/Safflower_SeasonData/UI/MaterialInstances/MI_UI_Safflower_OzoneCredit
     tiling=NoTile
     imageSize=<725 725>
     margin=<ltrb 0 0 0 0>
     outlineSettings {
      bUseBrushTransparency=false
      roundingType=HalfHeightRadius
      width=0
      color {colorUseRule=UseColor_Specified, specifiedColor=<rgba 0 0 0 0>}
      cornerRadii=<0 0 0 0>
     }
     tintColor {colorUseRule=UseColor_Specified, specifiedColor=<rgba 1 1 1 1>}
     uVRegion {bIsValid=false, max=<0 0>, min=<0 0>}
    }
    loadingBackgroundBrush {
     bIsDynamicallyLoaded=false
     drawAs=Image
     imageType=FullColor
     mirroring=NoMirror
     resourceName=None
     resourceObject=null
     tiling=NoTile
     imageSize=<32 32>
     margin=<ltrb 0 0 0 0>
     outlineSettings {
      bUseBrushTransparency=false
      roundingType=HalfHeightRadius
      width=0
      color {colorUseRule=UseColor_Specified, specifiedColor=<rgba 0 0 0 0>}
      cornerRadii=<0 0 0 0>
     }
     tintColor {colorUseRule=UseColor_Specified, specifiedColor=<rgba 1 1 1 1>}
     uVRegion {bIsValid=false, max=<0 0>, min=<0 0>}
    }
   }
   slot {class=OverlaySlot, properties {horizontalAlignment=HAlign_Center, verticalAlignment=VAlign_Center}}
  }
  KeyArt_MorningSoak {
   widgetClassPath=&FortLazyImage
   bIsVariable=true
   parent=Overlay_49
   widget=KeyArt_MorningSoak
   properties {
    brush {
     bIsDynamicallyLoaded=false
     drawAs=Image
     imageType=FullColor
     mirroring=NoMirror
     resourceName=None
     resourceObject=/Safflower_SeasonData/UI/MaterialInstances/MI_UI_Safflower_MorningSoak_Cook
     tiling=NoTile
     imageSize=<363 725>
     margin=<ltrb 0 0 0 0>
     outlineSettings {
      bUseBrushTransparency=false
      roundingType=HalfHeightRadius
      width=0
      color {colorUseRule=UseColor_Specified, specifiedColor=<rgba 0 0 0 0>}
      cornerRadii=<0 0 0 0>
     }
     tintColor {colorUseRule=UseColor_Specified, specifiedColor=<rgba 1 1 1 1>}
     uVRegion {bIsValid=false, max=<0 0>, min=<0 0>}
    }
    loadingBackgroundBrush {
     bIsDynamicallyLoaded=false
     drawAs=Image
     imageType=FullColor
     mirroring=NoMirror
     resourceName=None
     resourceObject=null
     tiling=NoTile
     imageSize=<32 32>
     margin=<ltrb 0 0 0 0>
     outlineSettings {
      bUseBrushTransparency=false
      roundingType=HalfHeightRadius
      width=0
      color {colorUseRule=UseColor_Specified, specifiedColor=<rgba 0 0 0 0>}
      cornerRadii=<0 0 0 0>
     }
     tintColor {colorUseRule=UseColor_Specified, specifiedColor=<rgba 1 1 1 1>}
     uVRegion {bIsValid=false, max=<0 0>, min=<0 0>}
    }
    renderTransform {angle=0, scale=<1 1>, shear=<0 0>, translation=<92 59>}
   }
   slot {class=OverlaySlot, properties {horizontalAlignment=HAlign_Center, verticalAlignment=VAlign_Center}}
  }
  KeyArt_OzoneCredit {
   widgetClassPath=&FortLazyImage
   bIsVariable=true
   parent=Overlay_49
   widget=KeyArt_OzoneCredit
   properties {
    brush {
     bIsDynamicallyLoaded=false
     drawAs=Image
     imageType=FullColor
     mirroring=NoMirror
     resourceName=None
     resourceObject=&MI_UI_Safflower_OzoneCredit
     tiling=NoTile
     imageSize=<363 725>
     margin=<ltrb 0 0 0 0>
     outlineSettings {
      bUseBrushTransparency=false
      roundingType=HalfHeightRadius
      width=0
      color {colorUseRule=UseColor_Specified, specifiedColor=<rgba 0 0 0 0>}
      cornerRadii=<0 0 0 0>
     }
     tintColor {colorUseRule=UseColor_Specified, specifiedColor=<rgba 1 1 1 1>}
     uVRegion {bIsValid=false, max=<0 0>, min=<0 0>}
    }
    colorAndOpacity=<rgba 1 1 1 0>
    loadingBackgroundBrush {
     bIsDynamicallyLoaded=false
     drawAs=Image
     imageType=FullColor
     mirroring=NoMirror
     resourceName=None
     resourceObject=null
     tiling=NoTile
     imageSize=<32 32>
     margin=<ltrb 0 0 0 0>
     outlineSettings {
      bUseBrushTransparency=false
      roundingType=HalfHeightRadius
      width=0
      color {colorUseRule=UseColor_Specified, specifiedColor=<rgba 0 0 0 0>}
      cornerRadii=<0 0 0 0>
     }
     tintColor {colorUseRule=UseColor_Specified, specifiedColor=<rgba 1 1 1 1>}
     uVRegion {bIsValid=false, max=<0 0>, min=<0 0>}
    }
    renderTransform {angle=10, scale=<0.88 0.88>, shear=<0 0>, translation=<-24.7329 70>}
   }
   slot {class=OverlaySlot, properties {horizontalAlignment=HAlign_Center, verticalAlignment=VAlign_Center}}
  }
 ]
}
