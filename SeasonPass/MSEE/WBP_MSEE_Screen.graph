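# Main screen for an MSEE event pass, showing a header, reward track, chase reward and key art.
# On activation it looks up the season pass by template ID, fills selection viewmodels and plays an intro, or shows a load error.
# Buttons push buy and inspect screens, quests and an About dialog built from theme lookup tables.
# Setters inject track and chase widgets; helpers handle themed text, focus navigation, async backgrounds and input blocking.
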WidgetBlueprint /SeasonPass/MSEE/WBP_MSEE_Screen
animations [
 Intro_Default {
  binding_count=4
  length_seconds=1
  object=&self:Intro_Default
  sequence {bindings [
    ChaseWidget_Container {
     id=#1
     possessed_class=Overlay
     tracks [{
       type=MovieSceneFloatTrack
       display_name=Render Opacity
       sections [{type=MovieSceneFloatSection, end_frame=20, start_frame=0, channels [None {type=MovieSceneScriptingFloatChannel, keys [[12,0],[20,1]]}]}]
      }]
    }
    WBP_EventPassHeader {
     id=#2
     possessed_class=WBP_EventPassHeader_C
     tracks [{
       type=MovieSceneFloatTrack
       display_name=Render Opacity
       sections [{type=MovieSceneFloatSection, end_frame=20, start_frame=0, channels [None {type=MovieSceneScriptingFloatChannel, keys [[12,0],[20,1]]}]}]
      }]
    }
    RewardTrack_VBox {
     id=#3
     possessed_class=VerticalBox
     tracks [{
       type=MovieSceneFloatTrack
       display_name=Render Opacity
       sections [{type=MovieSceneFloatSection, end_frame=20, start_frame=0, channels [None {type=MovieSceneScriptingFloatChannel, keys [[12,0],[20,1]]}]}]
      }]
    }
    LegalandActionBar {
     id=#4
     possessed_class=HorizontalBox
     tracks [{
       type=MovieSceneFloatTrack
       display_name=Render Opacity
       sections [{type=MovieSceneFloatSection, end_frame=20, start_frame=0, channels [None {type=MovieSceneScriptingFloatChannel, keys [[12,0],[20,1]]}]}]
      }]
    }
   ]}
 }
 Outro_Default {
  binding_count=4
  length_seconds=1
  object=&self:Outro_Default
  sequence {bindings [
    ChaseWidget_Container {
     id=#1
     possessed_class=Overlay
     tracks [{
       type=MovieSceneFloatTrack
       display_name=Render Opacity
       sections [{type=MovieSceneFloatSection, end_frame=20, start_frame=0, channels [None {type=MovieSceneScriptingFloatChannel, keys [[0,1],[20,0]]}]}]
      }]
    }
    WBP_EventPassHeader {
     id=#2
     possessed_class=WBP_EventPassHeader_C
     tracks [{
       type=MovieSceneFloatTrack
       display_name=Render Opacity
       sections [{type=MovieSceneFloatSection, end_frame=20, start_frame=0, channels [None {type=MovieSceneScriptingFloatChannel, keys [[0,1],[20,0]]}]}]
      }]
    }
    RewardTrack_VBox {
     id=#3
     possessed_class=VerticalBox
     tracks [{
       type=MovieSceneFloatTrack
       display_name=Render Opacity
       sections [{type=MovieSceneFloatSection, end_frame=20, start_frame=0, channels [None {type=MovieSceneScriptingFloatChannel, keys [[0,1],[20,0]]}]}]
      }]
    }
    LegalandActionBar {
     id=#4
     possessed_class=HorizontalBox
     tracks [{
       type=MovieSceneFloatTrack
       display_name=Render Opacity
       sections [{type=MovieSceneFloatSection, end_frame=20, start_frame=0, channels [None {type=MovieSceneScriptingFloatChannel, keys [[0,1],[20,0]]}]}]
      }]
    }
   ]}
 }
 Anim_Reveal {
  binding_count=5
  length_seconds=0.9
  object=&self:Anim_Reveal
  sequence {bindings [
    RewardTrack_VBox {
     id=#5
     possessed_class=VerticalBox
     tracks [
      {
       type=MovieSceneFloatTrack
       display_name=Render Opacity
       sections [{type=MovieSceneFloatSection, end_frame=18, start_frame=0, channels [None {type=MovieSceneScriptingFloatChannel, keys [[0,0],[18,1]]}]}]
      }
      {
       type=MovieScene2DTransformTrack
       sections [{
         type=MovieScene2DTransformSection
         end_frame=18
         start_frame=0
         channels [
          Translation.X {type=MovieSceneScriptingFloatChannel, const [0,1]}
          Translation.Y {type=MovieSceneScriptingFloatChannel, keys [[0,20],[18,0]]}
          Angle {type=MovieSceneScriptingFloatChannel, const [0,1]}
          Scale.X {type=MovieSceneScriptingFloatChannel, const [1,1]}
          Scale.Y {type=MovieSceneScriptingFloatChannel, const [1,1]}
          Shear.X {type=MovieSceneScriptingFloatChannel, const [0,1]}
          Shear.Y {type=MovieSceneScriptingFloatChannel, const [0,1]}
         ]
        }]
      }
     ]
    }
    KeyArt_Container {
     id=#6
     possessed_class=VerticalBox
     tracks [{
       type=MovieSceneFloatTrack
       display_name=Render Opacity
       sections [{type=MovieSceneFloatSection, end_frame=18, start_frame=0, channels [None {type=MovieSceneScriptingFloatChannel, const [1,2]}]}]
      }]
    }
    WBP_EventPassHeader {
     id=#7
     possessed_class=WBP_EventPassHeader_C
     tracks [
      {
       type=MovieSceneFloatTrack
       display_name=Render Opacity
       sections [{type=MovieSceneFloatSection, end_frame=18, start_frame=0, channels [None {type=MovieSceneScriptingFloatChannel, keys [[0,0],[18,1]]}]}]
      }
      {
       type=MovieScene2DTransformTrack
       sections [{
         type=MovieScene2DTransformSection
         end_frame=18
         start_frame=0
         channels [
          Translation.X {type=MovieSceneScriptingFloatChannel, keys [[0,-20],[18,0]]}
          Translation.Y {type=MovieSceneScriptingFloatChannel, keys [[0,-20],[18,0]]}
          Angle {type=MovieSceneScriptingFloatChannel, const [0,1]}
          Scale.X {type=MovieSceneScriptingFloatChannel, const [1,1]}
          Scale.Y {type=MovieSceneScriptingFloatChannel, const [1,1]}
          Shear.X {type=MovieSceneScriptingFloatChannel, const [0,1]}
          Shear.Y {type=MovieSceneScriptingFloatChannel, const [0,1]}
         ]
        }]
      }
     ]
    }
    LegalandActionBar {
     id=#8
     possessed_class=HorizontalBox
     tracks [
      {
       type=MovieSceneFloatTrack
       display_name=Render Opacity
       sections [{type=MovieSceneFloatSection, end_frame=18, start_frame=0, channels [None {type=MovieSceneScriptingFloatChannel, keys [[0,0],[18,1]]}]}]
      }
      {
       type=MovieScene2DTransformTrack
       sections [{
         type=MovieScene2DTransformSection
         end_frame=18
         start_frame=0
         channels [
          Translation.X {type=MovieSceneScriptingFloatChannel, const [0,1]}
          Translation.Y {type=MovieSceneScriptingFloatChannel, keys [[0,20],[18,0]]}
          Angle {type=MovieSceneScriptingFloatChannel, const [0,1]}
          Scale.X {type=MovieSceneScriptingFloatChannel, const [1,1]}
          Scale.Y {type=MovieSceneScriptingFloatChannel, const [1,1]}
          Shear.X {type=MovieSceneScriptingFloatChannel, const [0,1]}
          Shear.Y {type=MovieSceneScriptingFloatChannel, const [0,1]}
         ]
        }]
      }
     ]
    }
    ChaseWidget_Container {
     id=#9
     possessed_class=VerticalBox
     tracks [
      {
       type=MovieScene2DTransformTrack
       sections [{
         type=MovieScene2DTransformSection
         end_frame=18
         start_frame=0
         channels [
          Translation.X {type=MovieSceneScriptingFloatChannel, keys [[0,20],[18,0]]}
          Translation.Y {type=MovieSceneScriptingFloatChannel, keys [[0,-20],[18,0]]}
          Angle {type=MovieSceneScriptingFloatChannel, const [0,1]}
          Scale.X {type=MovieSceneScriptingFloatChannel, const [1,1]}
          Scale.Y {type=MovieSceneScriptingFloatChannel, const [1,1]}
          Shear.X {type=MovieSceneScriptingFloatChannel, const [0,1]}
          Shear.Y {type=MovieSceneScriptingFloatChannel, const [0,1]}
         ]
        }]
      }
      {
       type=MovieSceneFloatTrack
       display_name=Render Opacity
       sections [{type=MovieSceneFloatSection, end_frame=18, start_frame=0, channels [None {type=MovieSceneScriptingFloatChannel, keys [[0,0],[18,1]]}]}]
      }
     ]
    }
   ]}
 }
 Anim_LoadError {
  binding_count=1
  length_seconds=1
  object=&self:Anim_LoadError
  sequence {bindings [Error_Container {
     id=#10
     possessed_class=Overlay
     tracks [
      {
       type=MovieSceneFloatTrack
       display_name=Render Opacity
       sections [{type=MovieSceneFloatSection, end_frame=20, start_frame=0, channels [None {type=MovieSceneScriptingFloatChannel, keys [[0,0],[20,1]]}]}]
      }
      {
       type=MovieScene2DTransformTrack
       sections [{
         type=MovieScene2DTransformSection
         end_frame=20
         start_frame=0
         channels [
          Translation.X {type=MovieSceneScriptingFloatChannel, keys [[0,-100],[20,0]]}
          Translation.Y {type=MovieSceneScriptingFloatChannel, const [0,1]}
          Angle {type=MovieSceneScriptingFloatChannel, const [0,1]}
          Scale.X {type=MovieSceneScriptingFloatChannel, const [1,1]}
          Scale.Y {type=MovieSceneScriptingFloatChannel, const [1,1]}
          Shear.X {type=MovieSceneScriptingFloatChannel, const [0,1]}
          Shear.Y {type=MovieSceneScriptingFloatChannel, const [0,1]}
         ]
        }]
      }
     ]
    }]}
 }
 Anim_KeyArt_FadeOut {
  binding_count=1
  length_seconds=0.9
  object=&self:Anim_KeyArt_FadeOut
  sequence {bindings [KeyArt_Container {
     id=#11
     possessed_class=VerticalBox
     tracks [{
       type=MovieSceneFloatTrack
       display_name=Render Opacity
       sections [{type=MovieSceneFloatSection, end_frame=10, start_frame=0, channels [None {type=MovieSceneScriptingFloatChannel, keys [[0,1],[10,0]]}]}]
      }]
    }]}
 }
]
graphs [
 ApplyTextModifiers {nodes [
   n1 = ApplyTextModifiers!() -> (then, TextInput:Text) {is=title K2Node_FunctionEntry}
   > Return Node!(
    execute
    Result:Text=Format Text!(
     Format=@n1
     StartDate:Text=Get Var_StartDate() -> (Var_StartDate:Text)
     EndDate:Text=Get Var_EndDate() -> (Var_EndDate:Text)
     CurrencyName:Text=Get Var_CurrencyName() -> (Var_CurrencyName:Text)
     EventName:Text=Get Var_EventName() -> (Var_EventName:Text)
     TokenName:Text=Get Var_TokenName() -> (Var_TokenName:Text)
     QuestBundleName:Text=Get Var_QuestBundleName() -> (Var_QuestBundleName:Text)
     ChaseRewardName:Text=Get Var_ChaseRewardName() -> (Var_ChaseRewardName:Text)
     p:Text=MakeLiteralText(Value="LOCTEXT(\"\\r\\n\\r\\n\")")
    ) -> (Result)
   )
  ]}
 BP_GetDesiredFocusTarget {nodes [
   Get Desired Focus Target()
   > Return Node!(
    execute
    ReturnValue:Widget Object Reference=Get Button_GoToQuests!(self:WBP Event Pass Header Object Reference=Get WBP_EventPassHeader() -> (WBP_EventPassHeader:WBP Event Pass Header Object Reference)) -> (Button_GoToQuests:WBP UIKit Button Loud Object Reference)
   )
  ]}
 BP_GetDesiredInputConfig {nodes [
   Get Desired Input Config() {is=title K2Node_FunctionEntry}
   > Return Node(ReturnValue:UIInput Config Structure=Make UIInput Config!(
     bIgnoreMoveInput:bool=False
     bIgnoreLookInput:bool=False
     InputMode:ECommonInputMode Enum=All
     MouseCaptureMode:EMouseCaptureMode Enum=CaptureDuringMouseDown
     MouseLockMode:EMouseLockMode Enum=DoNotLock
     bHideCursorDuringViewportCapture:bool=True
    ) -> (UIInputConfig:UIInput Config Structure) {is=title K2Node_MakeStruct})
  ]}
 BlockInput {nodes [
   n1 = BlockInput!() -> (then, BlockInput:bool=~) {is=title K2Node_FunctionEntry}
   > SetVisibility(
    self=[
     Get WBP_ActionBarContainer() -> (WBP_ActionBarContainer:WBP Action Bar Container Object Reference)
     Get WBP_EventPassHeader()
     Get RewardTrack_Slot() -> (RewardTrack_Slot:Named Slot Object Reference)
     Get Button_ViewChaseReward() -> (Button_ViewChaseReward:WBP UIKit Button Regular Object Reference)
    ]
    InVisibility=Select(Option 0=SelfHitTestInvisible, Option 1=HitTestInvisible, Index=@n1)
   )
  ]}
 EventGraph {nodes [
   "On Clicked (Button_MoreInfo)"!() -> (OutputDelegate, then, Button:Common Button Base Object Reference) {is=title K2Node_ComponentBoundEvent}
   > n1 = Construct Object from Class(Class=/SeasonPass/MSEE/Data/DialogVM_MSEE_Basic.DialogVM_MSEE_Basic_C) -> (ReturnValue:Dialog VM MSEE Basic Object Reference)
   > n2 = Create Widget!(
    execute:Exec
    Class:User Widget Class Reference=&/WBP_MSEE_AboutDialog.WBP_MSEE_AboutDialog_C
    OwningPlayer:Player Controller Object Reference
    UIKitDialogViewModel:UI Kit Dialog Viewmodel Object Reference=@n1
   ) -> (then:Exec, ReturnValue:WBP MSEE About Dialog Object Reference)
   > PopulateEntries!(
    execute
    self:WBP MSEE About Dialog Object Reference=@n2
    Entries:Array of S MSEE Custom Dialog Entry Structures=Get AboutDialogEntries() -> (AboutDialogEntries:Array of S MSEE Custom Dialog Entry Structures)
   ) -> (then)
   > SetFooterLegalText!(execute, self:WBP MSEE About Dialog Object Reference=@n2, InText:Text=Get AboutModalDisclaimer() -> (AboutModalDisclaimer:Text)) -> (then)
   > SetImageByRef!(
    execute
    self:WBP MSEE About Dialog Object Reference=@n2
    ImageReference:"Fort Image Reference Structure (by ref)"=Make Fort Image Reference!(Image:Soft Object Reference, ImageUrl:String=Get CMS Modal Art URL() -> (CMS Modal Art URL:String), MaterialTextureParamName:Name=None) -> (FortImageReference:Fort Image Reference Structure) {is=title K2Node_MakeStruct}
   ) -> (then)
   > TryPushWidgetAsModalPanel!(execute, PushWidget:User Widget Object Reference=@n2) -> (then)
   "On Clicked (Button_BuyRewards)"!() -> (OutputDelegate, then, Button:Common Button Base Object Reference) {is=title K2Node_ComponentBoundEvent}
   > Set PendingSubMenu(PendingSubMenu:bool=true) -> (Output_Get:bool)
   > n3 = PushPrimaryContentSoftClassAsync!(
    execute:Exec
    SoftWidgetClass:Common Activatable Widget Soft Class Reference=Get BuyScreen() -> (BuyScreen:Common Activatable Widget Soft Class Reference)
   ) -> (then:Exec, OnPrimaryContentPushed:Exec, ActivatableWidget:Common Activatable Widget Object Reference) {is=title K2Node_AsyncAction}
   n4 = Set BuyMenu(execute=@n3.OnPrimaryContentPushed, BuyMenu:Common Activatable Widget Object Reference=@n3) -> (Output_Get:Common Activatable Widget Object Reference)
   n5 = OnBuyMenuClosed() {is=title K2Node_CustomEvent}
   Bind Event to On Widget Deactivated(execute=@n4.then, self:Common Activatable Widget Object Reference=@n3, Delegate=@n5) {is=title K2Node_AddDelegate}
   "On Quest Button Clicked (WBP_EventPassHeader)"() {is=title K2Node_ComponentBoundEvent}
   n6 = GetReward!(self:Season Pass Reward VM Object Reference=Get SelectedReward!(self:Season Pass Selection VM Object Reference=Get SeasonPassSelectionVM() -> (SeasonPassSelectionVM:Season Pass Selection VM Object Reference)) -> (SelectedReward:Season Pass Reward VM Object Reference)) -> (ReturnValue:Athena Season Item Entry Base Object Reference)
   > NavigateToQuestScreen!(
    execute
    Context:Widget Object Reference=Self-Reference()
    StartingCategoryTag:"Gameplay Tag Structure (by ref)"="Get (a ref)"(
     Array:Array of Gameplay Tag Structures=GetQuestCategoryTags!(Context:Widget Object Reference, RewardEntry:Athena Season Item Entry Base Object Reference=@n6) -> (ReturnValue_GameplayTags:Array of Gameplay Tag Structures)
     Dimension 1=0
    ) -> (Output:"Gameplay Tag Structure (by ref)")
   ) -> (then)
   "On Clicked (Button_ViewChaseReward)"!() -> (OutputDelegate, then, Button:Common Button Base Object Reference) {is=title K2Node_ComponentBoundEvent}
   > BlockInput!(execute, self, BlockInput:bool=true) -> (then) {is=title K2Node_CallFunction}
   > Set SelectedReward!(
    execute
    SelectedReward:Season Pass Reward VM Object Reference=Get ChaseRewardItem() -> (ChaseRewardItem:Season Pass Reward VM Object Reference)
    self:Season Pass Selection VM Object Reference=Get SeasonPassSelectionVM()
   ) -> (then, Output_Get:Season Pass Reward VM Object Reference)
   > n7 = Set Show Chase Widget(Show Chase Widget:bool=true) -> (Output_Get:bool)
   Set PendingSubMenu(execute=[@n15.then, @n7.then], PendingSubMenu=true)
   > n8 = PushPrimaryContentSoftClassAsync(SoftWidgetClass=Get InspectScreen() -> (InspectScreen:Common Activatable Widget Soft Class Reference))
   n9 = Cast To WBP_MSEE_InspectScreen!(execute=@n8.OnPrimaryContentPushed, Object=@n8) -> (then, CastFailed, AsWBP MSEE Inspect Screen:WBP MSEE Inspect Screen Object Reference) {is=title K2Node_DynamicCast}
   > Set In Decorator Classes!(
    execute
    In Decorator Classes:Array of Rich Text Block Decorator Class References=Get Decorators() -> (Decorators:Array of Rich Text Block Decorator Class References)
    self:WBP MSEE Inspect Screen Object Reference=@n9
   ) -> (then, Output_Get:Array of Rich Text Block Decorator Class References)
   > UpdateForCustom(self:WBP MSEE Inspect Screen Object Reference=@n9)
   > Set ShowChaseWidget!(execute, ShowChaseWidget:bool=Get Show Chase Widget() -> (Show Chase Widget:bool), self:WBP MSEE Inspect Screen Object Reference=@n9) -> (then, Output_Get:bool)
   > ShouldShowChaseWidget(self:WBP MSEE Inspect Screen Object Reference=@n9)
   n10 = RequestLoadBackground!() -> (OutputDelegate, then, BackgroundMaterial:Soft Object Reference, Direction:EVaultWorldTransitionDirection Enum=Forward) {is=title K2Node_CustomEvent}
   > Set RequestedBackgroundSoftPath(RequestedBackgroundSoftPath:Soft Object Path Structure=To Soft Object Path!(SoftObjectReference:Soft Object Reference=@n10.BackgroundMaterial) -> (ReturnValue:Soft Object Path Structure)) -> (Output_Get:Soft Object Path Structure)
   > n11 = Async Load Asset(Asset=@n10.BackgroundMaterial)
   n12 = GetSoftObjectPath(Object=@n11)
   n13 = Get RequestedBackgroundSoftPath() -> (RequestedBackgroundSoftPath:Soft Object Path Structure)
   Branch(
    execute=@n11.Completed
    Condition="Equal Exactly (String)"(A=BreakSoftObjectPath(InSoftObjectPath=@n12), B=BreakSoftObjectPath(InSoftObjectPath=@n13))
   )
   > Branch(Condition=Get bIsActive() -> (bIsActive:bool))
   > UpdateBackground!(
    execute
    self:Fort Item Preview Background Object Reference=Get FortItemPreviewBackground() -> (FortItemPreviewBackground:Fort Item Preview Background Object Reference)
    LoadedBackground:Object Reference=@n11
    TransitionDirection:EVaultWorldTransitionDirection Enum=@n10.Direction
    PrimaryColor:Linear Color Structure=(R=1,G=1,B=1,A=0)
    SecondaryColor:Linear Color Structure=(R=0,G=0,B=0,A=0)
   ) -> (then)
   PlayIntroAnim() {is=title K2Node_CustomEvent}
   > AfterIntroPlays()
   n14 = LoadInspectionScreen!() -> (OutputDelegate, then, ShowChaseWidget:bool=~, RewardVM:Season Pass Reward VM Object Reference, FortItemVM:Fort Item VM Object Reference) {is=title K2Node_CustomEvent}
   > BlockInput(BlockInput=true)
   > Set SelectedReward(SelectedReward=@n14.RewardVM, self=Get SeasonPassSelectionVM())
   > n15 = Set Show Chase Widget(Show Chase Widget=@n14.ShowChaseWidget)
   AfterIntroPlays() {is=title K2Node_CustomEvent}
   > IsScreenActive!(
    execute
    self:WBP MSEE Chase Reward Progress Object Reference=Get Chase Widget() -> (Chase Widget:WBP MSEE Chase Reward Progress Object Reference)
    IsActive:bool=true
   ) -> (then)
   > BlockInput(BlockInput=false)
   > n16 = StartListeningToEvent!(
    execute:Exec
    Target:Gameplay Event Router Component Object Reference=GetGlobalEventRouter()
    EventType:Script Struct Object Reference=/FortniteCoreUI/UI/Social/Chat/TextChat/BPS_TextChatActiveStateChange
    Context:Object Reference=GetOwningPlayer()
    EventBubblingRule:EEventBubblingRule Enum
   ) -> (
    then:Exec
    AsyncTaskProxy:Async Action Start Listening To Event Object Reference
    OnEventReceived:Exec
    Payload:BPS Text Chat Active State Change Structure
   )
   > Set ChatEventListener(ChatEventListener:Async Action Start Listening To Event Object Reference=@n16) -> (Output_Get:Async Action Start Listening To Event Object Reference)
   SetButtonPromptsVisibility!(
    execute=@n16.OnEventReceived
    self
    ShouldShow:bool=NOT Boolean(A=Break BPS Text Chat Active State Change!(BPS_TextChatActiveStateChange:"BPS Text Chat Active State Change Structure (by ref)"=@n16.Payload) -> (bActive_4_#12:bool=~) {is=title K2Node_BreakStruct})
   ) -> (then)
   HandleErrorLoad() {is=title K2Node_CustomEvent}
   > PlayAnimationForward(InAnimation=Get Anim_LoadError() -> (Anim_LoadError:Widget Animation Object Reference), PlaybackSpeed=float / float(A=1, B=0.25))
   n17 = Event PreConstruct()
   > UpdateSettings()
   > Set with Broadcast IsUsingCompactUI!(execute, IsUsingCompactUI:bool=IsUsingCompactUI()) -> (then, Output_Get:bool)
   > Branch(Condition=@n17.IsDesignTime)
   > SetText(self=Get Text_Header() -> (Text_Header:Common Rich Text Block Object Reference), InText="INVTEXT(\"[PH] Header Text \")")
   > SetText(self=Get Text_Completed() -> (Text_Completed:Common Rich Text Block Object Reference), InText="INVTEXT(\"[PH] Completed Text \")")
   > SetText(self=Get Text_Body() -> (Text_Body:Common Rich Text Block Object Reference), InText="INVTEXT(\"[PH] Body Text\")")
   > "SetText (Text)"(self=Get LegalText() -> (LegalText:Common Text Object Reference), InText="INVTEXT(\"[PH] Legal Text\")")
   > SetText(
    self=Get ChaseClaimed_Text() -> (ChaseClaimed_Text:Common Rich Text Block Object Reference)
    InText="INVTEXT(\"[PH] Chase Reward Claimed Text\")"
   )
   Event On Activated()
   > Set SelectedPass!(execute, SelectedPass:Season Pass VM Object Reference, self:Season Pass Selection VM Object Reference=Get SeasonPassSelectionVM()) -> (then, Output_Get:Season Pass VM Object Reference)
   > n18 = GetCurrentStateWidget!(execute, PlayerController:Player Controller Object Reference=GetOwningPlayer()) -> (then, ReturnValue:Common Activatable Widget Object Reference)
   > n19 = Cast To AthenaUIStateWidget_Frontend!(execute, Object=@n18) -> (then, CastFailed, AsAthena UIState Widget Frontend:Athena UIState Widget Frontend Object Reference) {is=title K2Node_DynamicCast}
   > n20 = SetFrontendVisibilityMode!(execute, self:Athena UIState Widget Frontend Object Reference=@n19, InFrontendVisibilityMode:EFrontendVisibilityMode Enum=OnlyTop) -> (then)
   GetSeasonPassInfo!(
    execute=[@n20.then, @n19.CastFailed]
    self
    Target:Season Pass System VM Object Reference=Get SeasonPassSystemVM() -> (SeasonPassSystemVM:Season Pass System VM Object Reference)
   ) -> (then)
   > ResetToDefaults()
   > PlayIntroAnim() {is=title K2Node_CallFunction}
   Event On Deactivated()
   > IsScreenActive(self=Get Chase Widget(), IsActive=false)
   > n21 = GetCurrentStateWidget(PlayerController=GetOwningPlayer())
   > n22 = Cast To AthenaUIStateWidget_Frontend(Object=@n21)
   > n23 = SetFrontendVisibilityMode(self=@n22, InFrontendVisibilityMode=Normal)
   n24 = Branch(execute=[@n23.then, @n22.CastFailed], Condition=Get PendingSubMenu() -> (PendingSubMenu:bool))
   StopListeningToEvent(
    execute=[@n26.then, @n24.then]
    self:Async Action Start Listening To Event Object Reference=Get ChatEventListener() -> (ChatEventListener:Async Action Start Listening To Event Object Reference)
   )
   > Set ChatEventListener()
   UpdateBackground(
    execute=@n24.else
    self=Get FortItemPreviewBackground()
    TransitionDirection=None
    PrimaryColor=(R=0,G=0,B=0,A=0)
    SecondaryColor=(R=0,G=0,B=0,A=0)
   )
   n25 = Get WBP_ItemPreview() -> (WBP_ItemPreview:WBP Item Preview Object Reference)
   > ClearPreviewItem(self:Fort Item Preview Widget Object Reference=@n25)
   > n26 = DeactivateWidget(self=@n25)
   OnThemeDataLoaded!() -> (OutputDelegate, then, ThemeData:Season Pass Theme Data Structure) {is=title K2Node_CustomEvent}
   Set SkipIntro(execute=@n27.then, SkipIntro:bool=true) -> (Output_Get:bool)
   n27 = Unbind Event from On Widget Deactivated(
    execute=@n5.then
    self:Common Activatable Widget Object Reference=Get BuyMenu() -> (BuyMenu:Common Activatable Widget Object Reference)
    Delegate=@n5
   ) {is=title K2Node_RemoveDelegate}
  ]}
 GetBodyText {nodes [
   n1 = GetBodyText!() -> (
    then
    ThemeData:Season Pass Theme Data Structure
    PreviewedItem:Fort Item VM Object Reference
    Previewed RewardVM:Season Pass Reward VM Object Reference
   ) {is=title K2Node_FunctionEntry}
   > Get!(execute:Exec) -> (then:Exec, else:Exec, PreviewedItem:Fort Item VM Object Reference)
   > Get!(execute) -> (then, else, Previewed RewardVM:Season Pass Reward VM Object Reference)
   n2 = TryLookupText!(
    Table:"Fort Text Reference Table Structure (by ref)"=Break Season Pass Theme Data!(SeasonPassThemeData:"Season Pass Theme Data Structure (by ref)"=@n1) -> (
     TitleColor:Linear Color Structure
     TitleMaterial:Soft Object Reference
     TitleOutline:Font Outline Settings Structure
     SeasonColor:Linear Color Structure
     ProgressionColor:Linear Color Structure
     OverlayColor:Linear Color Structure
     VaultBackgroundMaterial:Soft Object Reference
     GiftboxVaultBackgroundMaterial:Soft Object Reference
     UIBackgroundMaterial:Soft Object Reference
     UpsellBackgroundMaterial:Soft Object Reference
     CrewCategoryBackgroundMaterial:Soft Object Reference
     CustomLogo:Soft Object Reference
     CustomSeasonTitle:Text
     CustomSeasonTitleColor:Linear Color Structure
     Upsell2BackgrounMaterial:Soft Object Reference
     ImageReferenceTable:Fort Image Reference Table Structure
     ColorReferenceTable:Fort Color Reference Table Structure
     TextReferenceTable:Fort Text Reference Table Structure
    ) {is=title K2Node_BreakStruct}.TextReferenceTable
    Name:String=Track_Body
   ) -> (OutTextReference:Text, ReturnValue:bool)
   n3 = GetDisplayName!(self:Item Definition Base Object Reference=Get ItemDefinition!(self:Fort Item VM Object Reference=Get PreviewedItem() -> (PreviewedItem:Fort Item VM Object Reference)) -> (ItemDefinition:Fort Item Definition Object Reference)) -> (ReturnValue:Text)
   n4 = Get Previewed RewardVM() -> (Previewed RewardVM:Season Pass Reward VM Object Reference)
   n5 = Format Text!(
    Format=@n2
    Cost:int=Get ProgressCountForUnlock!(self:Season Pass Reward VM Object Reference=@n4) -> (ProgressCountForUnlock:int=0)
    Item:Text=@n3
   ) -> (Result)
   n6 = Get bIsRewardOwned!(self:Season Pass Reward VM Object Reference=@n4) -> (bIsRewardOwned:bool)
   > Return Node!(execute, Text:Text=SelectText(A=SelectText(A=@n3, B=@n5, bPickA=@n6), bPickA=@n2.ReturnValue))
  ]}
 GetChaseCompletedText {nodes [
   n1 = GetChaseCompletedText!() -> (then, ThemeData:Season Pass Theme Data Structure, chaseRewardItem:Season Pass Reward VM Object Reference) {is=title K2Node_FunctionEntry}
   > Is Valid(InputObject=@n1.chaseRewardItem)
   n2 = TryLookupText(Table=Break Season Pass Theme Data(SeasonPassThemeData=@n1).TextReferenceTable, Name=ChaseReward_Claimed)
   n3 = Get bIsRewardOwned(self=Get chaseRewardItem() -> (chaseRewardItem:Season Pass Reward VM Object Reference))
   > Return Node!(execute, ReturnValue:Text=SelectText(A=@n2, bPickA=AND Boolean(A=@n2.ReturnValue, B=@n3)))
  ]}
 GetCompletedText {nodes [
   GetCompletedText!() -> (then, ThemeData:"Season Pass Theme Data Structure (by ref)", PreviewedSeasonReward:Season Pass Reward VM Object Reference) {is=title K2Node_FunctionEntry}
   n1 = Get RewardTrackWidgets() -> (RewardTrackWidgets:Array of Widget Object References)
   > n2 = Cast To WBP_MSEE_Itemcard!(
    execute
    Object="Get (a copy)"(Array:Array of Widget Object References=@n1, Dimension 1=Last Index!(TargetArray:Array of Widget Object References=@n1) -> (ReturnValue=0)) -> (Output:Widget Object Reference)
   ) -> (then, CastFailed, AsWBP MSEE Itemcard:WBP MSEE Itemcard Object Reference) {is=title K2Node_DynamicCast}
   n3 = TryLookupText(
    Table=Break Season Pass Theme Data(SeasonPassThemeData=Get ThemeData() -> (ThemeData:"Season Pass Theme Data Structure (by ref)")).TextReferenceTable
    Name=Track_Complete
   )
   n4 = Get bIsRewardOwned(self=Get SeasonPassRewardVM!(self:WBP MSEE Itemcard Object Reference=@n2) -> (SeasonPassRewardVM:Season Pass Reward VM Object Reference))
   > Return Node!(execute, Text:Text=SelectText(A=@n3, bPickA=AND Boolean(A=@n3.ReturnValue, B=@n4)))
  ]}
 GetDialogData {nodes [
   GetDialogData!() -> (then, ThemeData:Season Pass Theme Data Structure) {is=title K2Node_FunctionEntry}
   > Clear(TargetArray:Array of S MSEE Custom Dialog Entry Structures=Get AboutDialogEntries())
   > n1 = For Each Loop(Array:Array of S MSEE Custom Dialog Lookup Structures=Get DialogSetup() -> (DialogSetup:Array of S MSEE Custom Dialog Lookup Structures)) -> (Array Element:S MSEE Custom Dialog Lookup Structure)
   n2 = Break S MSEE Custom Dialog Lookup!(S_MSEE_CustomDialogLookup:"S MSEE Custom Dialog Lookup Structure (by ref)"=@n1) -> (
    HeaderLookup_21_#13:String
    BodyLookup_20_#14:String
    ImageLookup_23_#15:String
    ImageSize_26_#16:Vector 2D Structure
   ) {is=title K2Node_BreakStruct}
   n3 = Get!() -> (
    ThemeData_TitleColor:Linear Color Structure
    ThemeData_TitleMaterial:Soft Object Reference
    ThemeData_TitleOutline:Font Outline Settings Structure
    ThemeData_SeasonColor:Linear Color Structure
    ThemeData_ProgressionColor:Linear Color Structure
    ThemeData_OverlayColor:Linear Color Structure
    ThemeData_VaultBackgroundMaterial:Soft Object Reference
    ThemeData_GiftboxVaultBackgroundMaterial:Soft Object Reference
    ThemeData_UIBackgroundMaterial:Soft Object Reference
    ThemeData_UpsellBackgroundMaterial:Soft Object Reference
    ThemeData_CrewCategoryBackgroundMaterial:Soft Object Reference
    ThemeData_CustomLogo:Soft Object Reference
    ThemeData_CustomSeasonTitle:Text
    ThemeData_CustomSeasonTitleColor:Linear Color Structure
    ThemeData_Upsell2BackgrounMaterial:Soft Object Reference
    ThemeData_ImageReferenceTable:Fort Image Reference Table Structure
    ThemeData_ColorReferenceTable:Fort Color Reference Table Structure
    ThemeData_TextReferenceTable:Fort Text Reference Table Structure
   )
   n4 = TryLookupText(Table=@n3.ThemeData_TextReferenceTable, Name=@n2.BodyLookup_20_#14)
   n5 = SelectText(A=@n4, bPickA=@n4.ReturnValue)
   n6 = TryLookupText(Table=@n3.ThemeData_TextReferenceTable, Name=@n2)
   n7 = TryLookupImage!(
    Table:"Fort Image Reference Table Structure (by ref)"=@n3.ThemeData_ImageReferenceTable
    Name:String=@n2.ImageLookup_23_#15
   ) -> (OutImageReference:Fort Image Reference Structure, ReturnValue:bool)
   > Add(
    TargetArray:Array of S MSEE Custom Dialog Entry Structures=Get AboutDialogEntries()
    NewItem:"S MSEE Custom Dialog Entry Structure (by ref)"=Make S MSEE Custom Dialog Entry!(
     Header_2_#17:Text=SelectText(A=@n6, bPickA=@n6.ReturnValue)
     Body_4_#18:Text=ApplyTextModifiers!(TextInput:Text=@n5) -> (Result:Text) {is=title K2Node_CallFunction}
     ImageReference_10_#19:Fort Image Reference Structure=Select(Option 0:Fort Image Reference Structure, Option 1:Fort Image Reference Structure=@n7, Index=@n7.ReturnValue) -> (ReturnValue:Fort Image Reference Structure)
     ImageSize_9_#20:Vector 2D Structure=@n2.ImageSize_26_#16
     DecoratorClasses_14_#13:Array of Rich Text Block Decorator Class References=Get Decorators()
    ) -> (S_MSEE_CustomDialogEntry:S MSEE Custom Dialog Entry Structure) {is=title K2Node_MakeStruct}
   ) {is=title K2Node_CallArrayFunction}
  ]}
 GetHeaderText {nodes [
   n1 = GetHeaderText!() -> (then, Wallet:int, ThemeData:Season Pass Theme Data Structure) {is=title K2Node_FunctionEntry}
   > n2 = Branch(Condition="integer <= integer"(A=@n1))
   n3 = Get()
   n4 = TryLookupText(Table=@n3.ThemeData_TextReferenceTable, Name=Track_Header_FTUE)
   > Return Node!(execute, ReturnValue:Text=SelectText(A=@n4, bPickA=@n4.ReturnValue))
   n5 = TryLookupText(Table=@n3.ThemeData_TextReferenceTable, Name=Track_Header)
   n6 = Get Wallet() -> (Wallet:int=0)
   Return Node(execute=@n2.else, ReturnValue=SelectText(A=Format Text!(Format=@n5, Wallet:int=@n6) -> (Result), bPickA=@n5.ReturnValue))
  ]}
 GetHoveredItemVM {nodes [
   n1 = GetHoveredItemVM!() -> (then, SeasonReward VM:Season Pass Reward VM Object Reference, FortItemVM:Fort Item VM Object Reference, Hovered Item Index:int) {is=title K2Node_FunctionEntry}
   > Set with Broadcast Previewed Season Reward VM!(execute, Previewed Season Reward VM:Season Pass Reward VM Object Reference=@n1) -> (then, Output_Get:Season Pass Reward VM Object Reference)
   > Set PreviewedItemVM(PreviewedItemVM:Fort Item VM Object Reference=Get FortItemVM() -> (FortItemVM:Fort Item VM Object Reference)) -> (Output_Get:Fort Item VM Object Reference)
   > Set Hovered Item Index(Hovered Item Index:int=Get Hovered Item Index() -> (Hovered Item Index:int=0)) -> (Output_Get:int=0)
  ]}
 GetPrimaryContentSetup {nodes [
   GetPrimaryContentSetup() {is=title K2Node_FunctionEntry}
   > Return Node(ReturnValue:Primary Content Setup Structure)
  ]}
 GetSeasonPassInfo {nodes [
   n1 = GetSeasonPassInfo!() -> (then, Target:Season Pass System VM Object Reference) {is=title K2Node_FunctionEntry}
   n2 = FindSeasonPass!(self:Season Pass System VM Object Reference=@n1, TemplateId:String=Get Pass Template Id() -> (Pass Template Id:String)) -> (ReturnValue:Season Pass VM Object Reference)
   > n3 = Is Valid(InputObject=@n2)
   SetActiveWidgetIndex(
    execute=@n3.Is Not Valid
    self:Common Visibility Switcher Object Reference=Get ContentSwitcher() -> (ContentSwitcher:Common Visibility Switcher Object Reference)
    Index=1
   )
   > HandleErrorLoad() {is=title K2Node_CallFunction}
   SetActiveWidgetIndex(execute=@n3.Is Valid, self=Get ContentSwitcher(), Index=0)
   n4 = Get SeasonPassSelectionVM()
   > n5 = Set SelectedPass(SelectedPass=@n2, self=@n4)
   n6 = "Get (a copy)"(Array:Array of Season Pass Category VM Object References=Get Categories!(self:Season Pass VM Object Reference=@n5) -> (Categories:Array of Season Pass Category VM Object References)) -> (Output:Season Pass Category VM Object Reference)
   > Set SelectedPage!(
    execute
    SelectedPage:Season Pass Page VM Object Reference="Get (a copy)"(Array:Array of Season Pass Page VM Object References=Get Pages!(self:Season Pass Category VM Object Reference=@n6) -> (Pages:Array of Season Pass Page VM Object References)) -> (Output:Season Pass Page VM Object Reference)
    self:Season Pass Selection VM Object Reference=@n4
   ) -> (then, Output_Get:Season Pass Page VM Object Reference)
   n7 = Get Pages(self="Get (a copy)"(Array:Array of Season Pass Category VM Object References=Get Categories(self=@n5), Dimension 1=1) -> (Output:Season Pass Category VM Object Reference))
   n8 = "Get (a copy)"(Array:Array of Season Pass Track VM Object References=Get Tracks!(self:Season Pass Page VM Object Reference="Get (a copy)"(Array:Array of Season Pass Page VM Object References=@n7) -> (Output:Season Pass Page VM Object Reference)) -> (Tracks:Array of Season Pass Track VM Object References)) -> (Output:Season Pass Track VM Object Reference)
   > Set with Broadcast ChaseRewardItem!(
    execute
    ChaseRewardItem:Season Pass Reward VM Object Reference="Get (a copy)"(Array:Array of Season Pass Reward VM Object References=Get Rewards!(self:Season Pass Track VM Object Reference=@n8) -> (Rewards:Array of Season Pass Reward VM Object References)) -> (Output:Season Pass Reward VM Object Reference)
   ) -> (then, Output_Get:Season Pass Reward VM Object Reference)
  ]}
 Nav_ToBuyButton {nodes [
   Nav_ToBuyButton!() -> (then, Navigation:EUINavigation Enum) {is=title K2Node_FunctionEntry}
   > Return Node(ReturnValue:Widget Object Reference=Get Button_GoToQuests(self=Get WBP_EventPassHeader()))
  ]}
 Nav_ToLeftRewardItem {nodes [
   Nav_ToLeftRewardItem!() -> (then, Navigation:EUINavigation Enum) {is=title K2Node_FunctionEntry}
   > Return Node(ReturnValue="Get (a copy)"(Array:Array of Widget Object References=Get RewardTrackWidgets()) -> (Output:Widget Object Reference))
  ]}
 Nav_ToRightRewardItem {nodes [
   Nav_ToRightRewardItem!() -> (then, Navigation:EUINavigation Enum) {is=title K2Node_FunctionEntry}
   n1 = Get RewardTrackWidgets()
   > Return Node(ReturnValue="Get (a copy)"(Array=@n1, Dimension 1=Last Index(TargetArray=@n1) -> (ReturnValue=0)))
  ]}
 Nav_UpFromRewardTrack {nodes [
   Nav_UpFromRewardTrack!() -> (then, Navigation:EUINavigation Enum) {is=title K2Node_FunctionEntry}
   n1 = Last Index(TargetArray=Get RewardTrackWidgets()) -> (ReturnValue=0)
   n2 = int / int(A=SelectInt(A=@n1, B=1, bPickA="integer > integer"(A=@n1)), B=2)
   n3 = Get Hovered Item Index() -> (Hovered Item Index=0)
   n4 = Get WBP_EventPassHeader()
   > Return Node(ReturnValue=Select(
     Option 0:Widget Object Reference=Get Button_GoToQuests(self=@n4)
     Option 1:Widget Object Reference=Get Button_ViewChaseReward()
     Index="integer > integer"(A=@n3, B=@n2)
    ) -> (ReturnValue:Widget Object Reference))
  ]}
 ResetToDefaults {nodes [
   ResetToDefaults() {is=title K2Node_FunctionEntry}
   n1 = Get WBP_ItemPreview()
   > ActivateWidget(self=@n1)
   > SetCurrentPreviewItem!(execute, self:Fort Item Preview Widget Object Reference=@n1, ItemVM:Fort Item VM Object Reference) -> (then)
   > n2 = Branch(Condition=Get PendingSubMenu())
   > Set PendingSubMenu(PendingSubMenu=false)
   UpdateBackground(
    execute=@n2.else
    self=Get FortItemPreviewBackground()
    TransitionDirection=None
    PrimaryColor=(R=0,G=0,B=0,A=0)
    SecondaryColor=(R=0,G=0,B=0,A=0)
   )
  ]}
 SetBackgroundURL {nodes [
   n1 = SetBackgroundURL!() -> (then, MediaUrl:String) {is=title K2Node_FunctionEntry}
   > SetMediaURL!(
    execute
    self:Epic CMSImage Object Reference=Get EpicCMSImage_BackgroundLoaderHelper() -> (EpicCMSImage_BackgroundLoaderHelper:Epic CMSImage Object Reference)
    MediaUrl:String=@n1
   ) -> (then)
  ]}
 SetButtonPromptsVisibility {nodes [
   n1 = SetButtonPromptsVisibility!() -> (then, ShouldShow:bool=~) {is=title K2Node_FunctionEntry}
   > SetVisibility(
    self=Get WBP_ActionBarContainer()
    InVisibility=Select(Option 0:ESlateVisibility Enum=Collapsed, Option 1:ESlateVisibility Enum=SelfHitTestInvisible, Index=@n1) -> (ReturnValue:ESlateVisibility Enum)
   )
  ]}
 SetChaseWidget {nodes [
   n1 = SetChaseWidget!() -> (then, ChaseWidget:WBP MSEE Chase Reward Progress Object Reference) {is=title K2Node_FunctionEntry}
   > Set Chase Widget(Chase Widget:WBP MSEE Chase Reward Progress Object Reference=@n1) -> (Output_Get:WBP MSEE Chase Reward Progress Object Reference)
  ]}
 SetKeyArt {nodes [
   n1 = SetKeyArt!() -> (then, CMSKeyArt:Fort Lazy Image Object Reference) {is=title K2Node_FunctionEntry}
   > Set CMSKey Art(CMSKey Art:Fort Lazy Image Object Reference=@n1) -> (Output_Get:Fort Lazy Image Object Reference)
  ]}
 SetRewardTrack {nodes [
   n1 = SetRewardTrack!() -> (then, RewardTrack:Dynamic Entry Box Object Reference) {is=title K2Node_FunctionEntry}
   > n2 = Set Reward Track(Reward Track:Dynamic Entry Box Object Reference=@n1) -> (Output_Get:Dynamic Entry Box Object Reference)
   > n3 = Set RewardTrackWidgets(RewardTrackWidgets:Array of Widget Object References=GetAllEntries!(self:Dynamic Entry Box Base Object Reference=@n2) -> (ReturnValue:Array of User Widget Object References)) -> (Output_Get:Array of Widget Object References)
   > n4 = Cast To WBP_MSEE_Itemcard(Object="Get (a copy)"(Array=@n3))
   > Set PreviewedItemVM(PreviewedItemVM=Get CurrentItemVM!(self:WBP MSEE Itemcard Object Reference=@n4) -> (CurrentItemVM:Fort Item VM Object Reference))
   > Set with Broadcast Previewed Season Reward VM(Previewed Season Reward VM=Get SeasonPassRewardVM(self=@n4))
  ]}
 SetTextModifiers {nodes [
   n1 = SetTextModifiers!() -> (
    then
    Var_StartDate:Text
    Var_EndDate:Text
    Var_CurrencyName:Text
    Var_EventName:Text
    Var_TokenName:Text
    Var_QuestBundleName:Text
    Var_ChaseRewardName:Text
   ) {is=title K2Node_FunctionEntry}
   > n2 = Sequence!(execute) -> (then_0, then_1, then_2:Exec, then_3:Exec, then_4:Exec, then_5:Exec, then_6:Exec)
   > Set Var_StartDate(Var_StartDate:Text=@n1) -> (Output_Get:Text)
   Set Var_EndDate(execute=@n2.then_1, Var_EndDate:Text=@n1.Var_EndDate) -> (Output_Get:Text)
   Set Var_CurrencyName(execute=@n2.then_2, Var_CurrencyName:Text=@n1.Var_CurrencyName) -> (Output_Get:Text)
   Set Var_EventName(execute=@n2.then_3, Var_EventName:Text=@n1.Var_EventName) -> (Output_Get:Text)
   Set Var_TokenName(execute=@n2.then_4, Var_TokenName:Text=@n1.Var_TokenName) -> (Output_Get:Text)
   Set Var_QuestBundleName(execute=@n2.then_5, Var_QuestBundleName:Text=@n1.Var_QuestBundleName) -> (Output_Get:Text)
   Set Var_ChaseRewardName(execute=@n2.then_6, Var_ChaseRewardName:Text=@n1.Var_ChaseRewardName) -> (Output_Get:Text)
  ]}
 UpdateDialogData {nodes [
   UpdateDialogData() {is=title K2Node_FunctionEntry}
   > n1 = For Each Loop(Array:Array of S MSEE Custom Dialog Entry Structures=Get AboutDialogEntries()) -> (Array Element:S MSEE Custom Dialog Entry Structure)
   Clear(execute=@n1.Completed, TargetArray=Get AboutDialogEntries())
   > Set AboutDialogEntries(AboutDialogEntries:Array of S MSEE Custom Dialog Entry Structures=Get UpdatedDialogEntries() -> (UpdatedDialogEntries:Array of S MSEE Custom Dialog Entry Structures)) -> (Output_Get:Array of S MSEE Custom Dialog Entry Structures)
   n2 = Break S MSEE Custom Dialog Entry!(S_MSEE_CustomDialogEntry:"S MSEE Custom Dialog Entry Structure (by ref)"=@n1) -> (
    Header_2_#17:Text
    Body_4_#18:Text
    ImageReference_10_#19:Fort Image Reference Structure
    ImageSize_9_#20:Vector 2D Structure
    DecoratorClasses_14_#13:Array of Rich Text Block Decorator Class References
   ) {is=title K2Node_BreakStruct}
   Add(
    execute=@n1.LoopBody
    TargetArray=Get UpdatedDialogEntries()
    NewItem=Make S MSEE Custom Dialog Entry(
     Header_2_#17=@n2
     Body_4_#18=ApplyTextModifiers(TextInput=@n2.Body_4_#18)
     ImageReference_10_#19=@n2.ImageReference_10_#19
     ImageSize_9_#20=@n2.ImageSize_9_#20
     DecoratorClasses_14_#13=@n2.DecoratorClasses_14_#13
    )
   )
  ]}
 UpdateRichTextStyles {nodes [
   n1 = UpdateRichTextStyles!() -> (then, Rich Text Decorators:Array of Rich Text Block Decorator Class References) {is=title K2Node_FunctionEntry}
   > UpdateDecorators!(
    execute
    self:WBP Event Pass Header Object Reference=Get WBP_EventPassHeader()
    InDecoratorClasses:Array of Rich Text Block Decorator Class References=@n1
   ) -> (then)
   > SetDecorators!(
    execute
    self:Rich Text Block Object Reference=[Get Text_Body(), Get Text_Completed(), Get Text_Header()]
    InDecoratorClasses:Array of Rich Text Block Decorator Class References=Get Rich Text Decorators() -> (Rich Text Decorators:Array of Rich Text Block Decorator Class References)
   ) -> (then)
  ]}
 UpdateSettings {nodes [
   UpdateSettings() {is=title K2Node_FunctionEntry}
   > UpdateRichTextStyles!(execute, self, Rich Text Decorators:Array of Rich Text Block Decorator Class References=Get Decorators()) -> (then) {is=title K2Node_CallFunction}
  ]}
 Update_BuyRewards_Visibility {nodes [
   n1 = Update_BuyRewards_Visibility!() -> (then, SelectedPass:Season Pass VM Object Reference) {is=title K2Node_FunctionEntry}
   > n2 = Is Valid(InputObject=@n1)
   Set Visibility(execute=@n2.Is Not Valid, Visibility=Collapsed, self=Get Button_BuyRewards() -> (Button_BuyRewards:WBP UIKit Button Quiet Object Reference))
   n3 = GetNumberOfIncompleteQuestEntries!(self:Season Pass Category VM Object Reference="Get (a copy)"(Array:Array of Season Pass Category VM Object References=Get Categories(self=@n1)) -> (Output:Season Pass Category VM Object Reference)) -> (ReturnValue:int=0)
   Set Visibility(execute=@n2.Is Valid, Visibility=Select(Option 0=Visible, Option 1=Collapsed, Index="Equal (Integer)"(A=@n3)), self=Get Button_BuyRewards())
  ]}
 VM_GetCMSData {nodes [
   n1 = VM_GetCMSData!() -> (
    then
    CMSData_PurchaseBackgroundURL:String
    CMSData_PurchaseConfirmBackgroundURL:String
    CMSData_PurchaseDisclaimer:Text
    CMSData_PurchaseDescription:Text
    CMSData_PurchaseConfirmDescription:Text
    CMSData_LandingPageDisclaimer:Text
    CMSData_LevelPurchaseDisclaimer:Text
    CMSData_SeasonEndDateText:Text
    CMSData_SeasonEndDateTime:Date Time Structure
   ) {is=title K2Node_FunctionEntry}
   > Set AboutModalDisclaimer(AboutModalDisclaimer:Text=@n1.CMSData_PurchaseDescription) -> (Output_Get:Text)
   > Set Var_EndDate(Var_EndDate=@n1.CMSData_SeasonEndDateText)
   > SetBrushFromImageReference!(
    execute
    self:Fort Lazy Image Object Reference=Get CMSKey Art() -> (CMSKey Art:Fort Lazy Image Object Reference)
    ImageReference:"Fort Image Reference Structure (by ref)"=Make Fort Image Reference(ImageUrl=@n1, MaterialTextureParamName=None)
    bMatchSize:bool=false
    DownloadPriority:ECMSDownloadPriority Enum=Normal
   ) -> (then)
   > Set CMS Modal Art URL(CMS Modal Art URL:String=@n1.CMSData_PurchaseConfirmBackgroundURL) -> (Output_Get:String)
   > UpdateDialogData() {is=title K2Node_CallFunction}
  ]}
 VM_GetThemeData {nodes [
   n1 = VM_GetThemeData!() -> (then, ThemeData:"Season Pass Theme Data Structure (by ref)") {is=title K2Node_FunctionEntry}
   > n2 = Set Theme Data(Theme Data:Season Pass Theme Data Structure=@n1) -> (Output_Get:Season Pass Theme Data Structure)
   > OnThemeDataLoaded!(execute, self, ThemeData:Season Pass Theme Data Structure=@n2) -> (then) {is=title K2Node_CallFunction}
   n3 = Break Season Pass Theme Data(SeasonPassThemeData=@n1)
   > RequestLoadBackground!(execute, self, BackgroundMaterial:Soft Object Reference=@n3.VaultBackgroundMaterial, Direction:EVaultWorldTransitionDirection Enum=None) -> (then) {is=title K2Node_CallFunction}
   > GetDialogData!(execute, self, ThemeData:Season Pass Theme Data Structure=Get ThemeData()) -> (then) {is=title K2Node_CallFunction}
   > SetText!(
    execute
    self:Fort CTAButton Object Reference=Get Button_ViewChaseReward()
    ButtonText:"Text (by ref)"=TryLookupText(Table=@n3.TextReferenceTable, Name=Button_Chase)
   ) -> (then)
  ]}
]
variables [
 AboutDialogEntries {type="array<S_MSEE_CustomDialogEntry>", default []}
 AboutModalDisclaimer {type=string, default=""}
 BuyMenu {type=CommonActivatableWidget, default=null}
 BuyScreen {type=Class, default=&/WBP_MSEE_BuyRewards.WBP_MSEE_BuyRewards_C}
 Chase Widget {type=WBP_MSEE_ChaseReward_Progress_C, default=null}
 ChaseRewardItem {type=SeasonPassRewardVM, default=null}
 ChaseRewardPage {type=SeasonPassPageVM, default=null}
 ChatEventListener {type=AsyncAction_StartListeningToEvent, default=null}
 CMS Modal Art URL {type=string, default=""}
 CMSKey Art {type=FortLazyImage, default=null}
 CompactUIHeaderPadding {type=Margin, default=<ltrb 0 60 0 0>}
 Decorators {type="array<Class>", default []}
 DefaultHeaderPadding {type=Margin, default=<ltrb 0 32 0 0>}
 DialogSetup {type="array<S_MSEE_CustomDialogLookup>", default []}
 ErrorLoadingPass {type=boolean}
 Hovered Item Index {type=integer}
 HoveredItemVM {type=SeasonPassRewardVM, default=null}
 InspectScreen {type=Class, default=&/WBP_MSEE_InspectScreen.WBP_MSEE_InspectScreen_C}
 IsUsingCompactUI {type=boolean}
 Pass Template Id {type=string, default=AthenaSeason:safflower_seasondata_seasonasset}
 PendingSubMenu {type=boolean}
 Previewed Season Reward VM {type=SeasonPassRewardVM, default=null}
 RequestedBackgroundSoftPath {type=SoftObjectPath, default=""}
 Reward Track {type=DynamicEntryBox, default=null}
 RewardTrackWidgets {type="array<Widget>", default []}
 Show Chase Widget {type=boolean}
 SkipIntro {type=boolean}
 Theme Data {
  type=SeasonPassThemeData
  default {
   crewCategoryBackgroundMaterial=null
   customLogo=null
   customSeasonTitle=""
   giftboxVaultBackgroundMaterial=null
   primaryAssetId=""
   titleMaterial=null
   uIBackgroundMaterial=null
   upsell2BackgrounMaterial=null
   upsellBackgroundMaterial=null
   vaultBackgroundMaterial=null
   benefitsByType []
   colorReferenceTable {namedColors []}
   customSeasonTitleColor=<rgba 1 1 1 1>
   imageReferenceTable {namedImages []}
   overlayColor=<rgba 0 0 0 1>
   progressionColor=<rgba 1 1 1 1>
   seasonColor=<rgba 1 1 1 1>
   textReferenceTable {namedTexts []}
   titleColor=<rgba 1 1 1 1>
   titleOutline {
    bApplyOutlineToDropShadows=false
    bMiteredCorners=false
    bSeparateFillAlpha=false
    outlineMaterial=null
    outlineSize=0
    outlineColor=<rgba 0 0 0 1>
   }
  }
 }
 Var_ChaseRewardName {type=string, default=""}
 Var_CurrencyName {type=string, default=""}
 Var_EndDate {type=string, default=""}
 Var_EventName {type=string, default=""}
 Var_QuestBundleName {type=string, default=""}
 Var_StartDate {type=string, default=""}
 Var_TokenName {type=string, default=""}
]
widgets {
 info {
  inheritedWidgetCount=0
  namedSlotCount=1
  parentClass=&CommonActivatableWidget
  rootWidgetClass=&Overlay
  widgetCount=35
 }
 widgets [
  Overlay_1 {widgetClassPath=&Overlay, widget=Overlay_1}
  EpicCMSImage_BackgroundLoaderHelper {
   widgetClassPath=/Script/EpicCMSUIFramework.EpicCMSImage
   bIsVariable=true
   parent=Overlay_1
   widget=EpicCMSImage_BackgroundLoaderHelper
   properties {visibility=Collapsed}
   slot {class=OverlaySlot}
  }
  FortItemPreviewBackground {
   widgetClassPath=/Script/FortniteUI.FortItemPreviewBackground
   bIsVariable=true
   parent=Overlay_1
   widget=FortItemPreviewBackground
   properties {previewSceneChangerTag {tagName=UI.ItemPreview.SceneChanger.MPItemShop}}
   slot {class=OverlaySlot, properties {horizontalAlignment=HAlign_Fill, verticalAlignment=VAlign_Fill}}
  }
  WBP_CaptureForPostBufferUpdate {
   widgetClassPath=&WBP_CaptureForPostBufferUpdate
   bIsVariable=true
   parent=Overlay_1
   slot {class=OverlaySlot, properties {horizontalAlignment=HAlign_Fill, verticalAlignment=VAlign_Fill}}
  }
  Background_CustomSlot {
   widgetClassPath=&NamedSlot
   bIsVariable=true
   parent=Overlay_1
   widget=Background_CustomSlot
   slot {class=OverlaySlot, properties {horizontalAlignment=HAlign_Fill, verticalAlignment=VAlign_Fill}}
  }
  ContentSwitcher {
   widgetClassPath=&CommonVisibilitySwitcher
   bIsVariable=true
   parent=Overlay_1
   widget=ContentSwitcher
   slot {class=OverlaySlot, properties {horizontalAlignment=HAlign_Fill, verticalAlignment=VAlign_Fill}}
  }
  MainContent {widgetClassPath=&Overlay, parent=ContentSwitcher, widget=MainContent, slot {class=CommonVisibilitySwitcherSlot}}
  SafeZone {
   widgetClassPath=&SafeZone
   bIsVariable=true
   parent=MainContent
   widget=SafeZone
   slot {class=OverlaySlot, properties {horizontalAlignment=HAlign_Fill, verticalAlignment=VAlign_Fill}}
  }
  Screen_SizeBox {
   widgetClassPath=&SizeBox
   parent=SafeZone
   widget=Screen_SizeBox
   properties {maxAspectRatio=3.9, maxDesiredWidth=4212, minAspectRatio=1.7778, minDesiredHeight=1080}
   slot {class=SafeZoneSlot}
  }
  MainContent_Grid {
   widgetClassPath=&GridPanel
   parent=Screen_SizeBox
   widget=MainContent_Grid
   properties {columnFill [0,1,0], rowFill [0,1,0]}
   slot {class=SizeBoxSlot, properties {padding=<ltrb 80 100 80 40>}}
  }
  WBP_ItemPreview {
   widgetClassPath=/FortniteCoreUI/Frontend/ItemPreview/WBP_ItemPreview
   bIsVariable=true
   parent=MainContent_Grid
   widget=WBP_ItemPreview
   properties {
    bShowFloor=false
    bSupportsActivationFocus=false
    itemViewObject=&self:WidgetTree.WBP_ItemPreview.ItemViewObject
    visibility=Collapsed
    previewSceneChangerTag {tagName=UI.ItemPreview.SceneChanger.MPItemShop}
   }
   slot {class=GridSlot, properties {columnSpan=3, rowSpan=3}}
  }
  KeyArt_Container {
   widgetClassPath=&VerticalBox
   parent=MainContent_Grid
   widget=KeyArt_Container
   slot {class=GridSlot, properties {columnSpan=3, layer=-1, rowSpan=3}}
  }
  KeyArt_Slot {
   widgetClassPath=&NamedSlot
   bIsVariable=true
   parent=KeyArt_Container
   widget=KeyArt_Slot
   slot {class=VerticalBoxSlot, properties {size {sizeRule=Fill, value=1}}}
  }
  WBP_EventPassHeader {
   widgetClassPath=&/WBP_EventPassHeader
   bIsVariable=true
   parent=MainContent_Grid
   widget=WBP_EventPassHeader
   properties {navigation=&self:WidgetTree.WBP_EventPassHeader.WidgetNavigation_0}
   slot {class=GridSlot, properties {horizontalAlignment=HAlign_Left, verticalAlignment=VAlign_Top, padding=<ltrb 0 32 0 0>}}
  }
  RewardTrack_VBox {
   widgetClassPath=&VerticalBox
   parent=MainContent_Grid
   widget=RewardTrack_VBox
   slot {class=GridSlot, properties {columnSpan=3, row=2, padding=<ltrb 0 0 0 48>}}
  }
  Text_Header {
   widgetClassPath=&CommonRichTextBlock
   bIsVariable=true
   parent=RewardTrack_VBox
   widget=Text_Header
   properties {
    applyLineHeightToBottomLine=false
    autoWrapText=true
    bAutoCollapseWithEmptyText=true
    bOverrideDefaultStyle=true
    lineHeightPercentage=1.15
    mobileTextBlockScale=1.2
    textTransformPolicy=ToUpper
    defaultTextStyleOverride {
     overflowPolicy=Clip
     transformPolicy=ToUpper
     colorAndOpacity {colorUseRule=UseColor_Specified, specifiedColor=<rgba 1 1 1 1>}
     font {
      bForceMonospaced=false
      bMaterialIsStencil=false
      fontMaterial=null
      fontObject=&HeadingNow
      letterSpacing=0
      monospacedWidth=1
      size=30
      skewAmount=0
      typefaceFontName=76 Bold Italic
      outlineSettings {
       bApplyOutlineToDropShadows=false
       bMiteredCorners=false
       bSeparateFillAlpha=false
       outlineMaterial=null
       outlineSize=0
       outlineColor=<rgba 0 0 0 1>
      }
     }
     highlightColor {colorUseRule=UseColor_Specified, specifiedColor=<rgba 0 0 0 0>}
     highlightShape {
      bIsDynamicallyLoaded=false
      drawAs=Image
      imageType=NoImage
      mirroring=NoMirror
      resourceName=None
      resourceObject=null
      tiling=NoTile
      imageSize=<32 32>
      margin=<ltrb 0 0 0 0>
      outlineSettings {
       bUseBrushTransparency=false
       roundingType=HalfHeightRadius
       width=0
       color {colorUseRule=UseColor_Specified, specifiedColor=<rgba 0 0 0 0>}
       cornerRadii=<0 0 0 0>
      }
      tintColor {colorUseRule=UseColor_Specified, specifiedColor=<rgba 1 1 1 1>}
      uVRegion {bIsValid=false, max=<0 0>, min=<0 0>}
     }
     selectedBackgroundColor {colorUseRule=UseColor_Foreground, specifiedColor=<rgba 1 0 1 1>}
     shadowColorAndOpacity=<rgba 0 0 0 1>
     shadowOffset=<0 0>
     strikeBrush {
      bIsDynamicallyLoaded=false
      drawAs=Image
      imageType=NoImage
      mirroring=NoMirror
      resourceName=None
      resourceObject=null
      tiling=NoTile
      imageSize=<32 32>
      margin=<ltrb 0 0 0 0>
      outlineSettings {
       bUseBrushTransparency=false
       roundingType=HalfHeightRadius
       width=0
       color {colorUseRule=UseColor_Specified, specifiedColor=<rgba 0 0 0 0>}
       cornerRadii=<0 0 0 0>
      }
      tintColor {colorUseRule=UseColor_Specified, specifiedColor=<rgba 1 1 1 1>}
      uVRegion {bIsValid=false, max=<0 0>, min=<0 0>}
     }
     underlineBrush {
      bIsDynamicallyLoaded=false
      drawAs=Image
      imageType=NoImage
      mirroring=NoMirror
      resourceName=None
      resourceObject=null
      tiling=NoTile
      imageSize=<32 32>
      margin=<ltrb 0 0 0 0>
      outlineSettings {
       bUseBrushTransparency=false
       roundingType=HalfHeightRadius
       width=0
       color {colorUseRule=UseColor_Specified, specifiedColor=<rgba 0 0 0 0>}
       cornerRadii=<0 0 0 0>
      }
      tintColor {colorUseRule=UseColor_Specified, specifiedColor=<rgba 1 1 1 1>}
      uVRegion {bIsValid=false, max=<0 0>, min=<0 0>}
     }
    }
   }
   slot {class=VerticalBoxSlot}
  }
  Text_Completed {
   widgetClassPath=&CommonRichTextBlock
   bIsVariable=true
   parent=RewardTrack_VBox
   widget=Text_Completed
   properties {
    applyLineHeightToBottomLine=false
    autoWrapText=true
    bAutoCollapseWithEmptyText=true
    bOverrideDefaultStyle=true
    lineHeightPercentage=1.15
    mobileTextBlockScale=1.2
    textTransformPolicy=ToUpper
    defaultTextStyleOverride {
     overflowPolicy=Clip
     transformPolicy=ToUpper
     colorAndOpacity {colorUseRule=UseColor_Specified, specifiedColor=<rgba 1 1 1 1>}
     font {
      bForceMonospaced=false
      bMaterialIsStencil=false
      fontMaterial=null
      fontObject=&HeadingNow
      letterSpacing=0
      monospacedWidth=1
      size=22.5
      skewAmount=0
      typefaceFontName=75 Medium Italic
      outlineSettings {
       bApplyOutlineToDropShadows=true
       bMiteredCorners=false
       bSeparateFillAlpha=false
       outlineMaterial=null
       outlineSize=0
       outlineColor=<rgba 0 0 0 1>
      }
     }
     highlightColor {colorUseRule=UseColor_Specified, specifiedColor=<rgba 0 0 0 0>}
     highlightShape {
      bIsDynamicallyLoaded=false
      drawAs=Image
      imageType=NoImage
      mirroring=NoMirror
      resourceName=None
      resourceObject=null
      tiling=NoTile
      imageSize=<32 32>
      margin=<ltrb 0 0 0 0>
      outlineSettings {
       bUseBrushTransparency=false
       roundingType=HalfHeightRadius
       width=0
       color {colorUseRule=UseColor_Specified, specifiedColor=<rgba 0 0 0 0>}
       cornerRadii=<0 0 0 0>
      }
      tintColor {colorUseRule=UseColor_Specified, specifiedColor=<rgba 1 1 1 1>}
      uVRegion {bIsValid=false, max=<0 0>, min=<0 0>}
     }
     selectedBackgroundColor {colorUseRule=UseColor_Foreground, specifiedColor=<rgba 1 0 1 1>}
     shadowColorAndOpacity=<rgba 0 0 0 1>
     shadowOffset=<0 0>
     strikeBrush {
      bIsDynamicallyLoaded=false
      drawAs=Image
      imageType=NoImage
      mirroring=NoMirror
      resourceName=None
      resourceObject=null
      tiling=NoTile
      imageSize=<32 32>
      margin=<ltrb 0 0 0 0>
      outlineSettings {
       bUseBrushTransparency=false
       roundingType=HalfHeightRadius
       width=0
       color {colorUseRule=UseColor_Specified, specifiedColor=<rgba 0 0 0 0>}
       cornerRadii=<0 0 0 0>
      }
      tintColor {colorUseRule=UseColor_Specified, specifiedColor=<rgba 1 1 1 1>}
      uVRegion {bIsValid=false, max=<0 0>, min=<0 0>}
     }
     underlineBrush {
      bIsDynamicallyLoaded=false
      drawAs=Image
      imageType=NoImage
      mirroring=NoMirror
      resourceName=None
      resourceObject=null
      tiling=NoTile
      imageSize=<32 32>
      margin=<ltrb 0 0 0 0>
      outlineSettings {
       bUseBrushTransparency=false
       roundingType=HalfHeightRadius
       width=0
       color {colorUseRule=UseColor_Specified, specifiedColor=<rgba 0 0 0 0>}
       cornerRadii=<0 0 0 0>
      }
      tintColor {colorUseRule=UseColor_Specified, specifiedColor=<rgba 1 1 1 1>}
      uVRegion {bIsValid=false, max=<0 0>, min=<0 0>}
     }
    }
   }
   slot {class=VerticalBoxSlot, properties {padding=<ltrb 0 8 0 0>}}
  }
  SizeBox_2 {
   widgetClassPath=&SizeBox
   parent=RewardTrack_VBox
   widget=SizeBox_2
   properties {bOverride_MinDesiredHeight=true, minDesiredHeight=42}
   slot {class=VerticalBoxSlot, properties {padding=<ltrb 0 0 0 16>}}
  }
  Text_Body {
   widgetClassPath=&CommonRichTextBlock
   bIsVariable=true
   parent=SizeBox_2
   widget=Text_Body
   properties {
    autoWrapText=true
    bAutoCollapseWithEmptyText=true
    bOverrideDefaultStyle=true
    lineHeightPercentage=1.15
    mobileTextBlockScale=1.25
    defaultTextStyleOverride {
     overflowPolicy=Clip
     transformPolicy=None
     colorAndOpacity {colorUseRule=UseColor_Specified, specifiedColor=<rgba 1 1 1 1>}
     font {
      bForceMonospaced=false
      bMaterialIsStencil=false
      fontMaterial=null
      fontObject=&HeadingNow
      letterSpacing=0
      monospacedWidth=1
      size=18
      skewAmount=0
      typefaceFontName=75 Medium
      outlineSettings {
       bApplyOutlineToDropShadows=true
       bMiteredCorners=false
       bSeparateFillAlpha=false
       outlineMaterial=null
       outlineSize=0
       outlineColor=<rgba 0 0 0 1>
      }
     }
     highlightColor {colorUseRule=UseColor_Specified, specifiedColor=<rgba 0 0 0 0>}
     highlightShape {
      bIsDynamicallyLoaded=false
      drawAs=Image
      imageType=NoImage
      mirroring=NoMirror
      resourceName=None
      resourceObject=null
      tiling=NoTile
      imageSize=<32 32>
      margin=<ltrb 0 0 0 0>
      outlineSettings {
       bUseBrushTransparency=false
       roundingType=HalfHeightRadius
       width=0
       color {colorUseRule=UseColor_Specified, specifiedColor=<rgba 0 0 0 0>}
       cornerRadii=<0 0 0 0>
      }
      tintColor {colorUseRule=UseColor_Specified, specifiedColor=<rgba 1 1 1 1>}
      uVRegion {bIsValid=false, max=<0 0>, min=<0 0>}
     }
     selectedBackgroundColor {colorUseRule=UseColor_Foreground, specifiedColor=<rgba 1 0 1 1>}
     shadowColorAndOpacity=<rgba 0 0 0 1>
     shadowOffset=<0 0>
     strikeBrush {
      bIsDynamicallyLoaded=false
      drawAs=Image
      imageType=NoImage
      mirroring=NoMirror
      resourceName=None
      resourceObject=null
      tiling=NoTile
      imageSize=<32 32>
      margin=<ltrb 0 0 0 0>
      outlineSettings {
       bUseBrushTransparency=false
       roundingType=HalfHeightRadius
       width=0
       color {colorUseRule=UseColor_Specified, specifiedColor=<rgba 0 0 0 0>}
       cornerRadii=<0 0 0 0>
      }
      tintColor {colorUseRule=UseColor_Specified, specifiedColor=<rgba 1 1 1 1>}
      uVRegion {bIsValid=false, max=<0 0>, min=<0 0>}
     }
     underlineBrush {
      bIsDynamicallyLoaded=false
      drawAs=Image
      imageType=NoImage
      mirroring=NoMirror
      resourceName=None
      resourceObject=null
      tiling=NoTile
      imageSize=<32 32>
      margin=<ltrb 0 0 0 0>
      outlineSettings {
       bUseBrushTransparency=false
       roundingType=HalfHeightRadius
       width=0
       color {colorUseRule=UseColor_Specified, specifiedColor=<rgba 0 0 0 0>}
       cornerRadii=<0 0 0 0>
      }
      tintColor {colorUseRule=UseColor_Specified, specifiedColor=<rgba 1 1 1 1>}
      uVRegion {bIsValid=false, max=<0 0>, min=<0 0>}
     }
    }
   }
   slot {class=SizeBoxSlot, properties {verticalAlignment=VAlign_Bottom}}
  }
  RewardTrack_Slot {
   widgetClassPath=&NamedSlot
   bIsVariable=true
   parent=RewardTrack_VBox
   widget=RewardTrack_Slot
   properties {navigation=&self:WidgetTree.RewardTrack_Slot.WidgetNavigation_0}
   slot {class=VerticalBoxSlot, properties {horizontalAlignment=HAlign_Center, padding=<ltrb 0 8 0 32>}}
  }
  ChaseWidget_Container {
   widgetClassPath=&VerticalBox
   parent=MainContent_Grid
   widget=ChaseWidget_Container
   properties {bOverride_Cursor=true}
   slot {class=GridSlot, properties {column=2, horizontalAlignment=HAlign_Right, verticalAlignment=VAlign_Top, padding=<ltrb 0 32 0 0>}}
  }
  ChaseWidget_Slot {widgetClassPath=&NamedSlot, bIsVariable=true, parent=ChaseWidget_Container, widget=ChaseWidget_Slot, slot {class=VerticalBoxSlot}}
  Button_ViewChaseReward {
   widgetClassPath=&WBP_UIKit_Button_Regular
   bIsVariable=true
   parent=ChaseWidget_Container
   widget=Button_ViewChaseReward
   properties {navigation=&self:WidgetTree.Button_ViewChaseReward.WidgetNavigation_0, text=View Item, visibility=Visible}
   slot {class=VerticalBoxSlot, properties {horizontalAlignment=HAlign_Center}}
  }
  ChaseClaimed_Text {
   widgetClassPath=&CommonRichTextBlock
   bIsVariable=true
   parent=ChaseWidget_Container
   widget=ChaseClaimed_Text
   properties {
    autoWrapText=true
    bOverrideDefaultStyle=true
    justification=Center
    lineHeightPercentage=1.15
    mobileTextBlockScale=1.2
    textStyleSet=/Game/Athena/Items/Cosmetics/Metadata/ItemDescriptionTextStyle
    wrapTextAt=700
    defaultTextStyleOverride {
     overflowPolicy=Clip
     transformPolicy=None
     colorAndOpacity {colorUseRule=UseColor_Specified, specifiedColor=<rgba 0.0437 1 0.2307 1>}
     font {
      bForceMonospaced=false
      bMaterialIsStencil=false
      fontMaterial=null
      fontObject=&HeadingNow
      letterSpacing=0
      monospacedWidth=1
      size=19.5
      skewAmount=0
      typefaceFontName=75 Medium
      outlineSettings {
       bApplyOutlineToDropShadows=true
       bMiteredCorners=false
       bSeparateFillAlpha=false
       outlineMaterial=null
       outlineSize=0
       outlineColor=<rgba 0 0 0 1>
      }
     }
     highlightColor {colorUseRule=UseColor_Specified, specifiedColor=<rgba 0 0 0 0>}
     highlightShape {
      bIsDynamicallyLoaded=false
      drawAs=Image
      imageType=NoImage
      mirroring=NoMirror
      resourceName=None
      resourceObject=null
      tiling=NoTile
      imageSize=<32 32>
      margin=<ltrb 0 0 0 0>
      outlineSettings {
       bUseBrushTransparency=false
       roundingType=HalfHeightRadius
       width=0
       color {colorUseRule=UseColor_Specified, specifiedColor=<rgba 0 0 0 0>}
       cornerRadii=<0 0 0 0>
      }
      tintColor {colorUseRule=UseColor_Specified, specifiedColor=<rgba 1 1 1 1>}
      uVRegion {bIsValid=false, max=<0 0>, min=<0 0>}
     }
     selectedBackgroundColor {colorUseRule=UseColor_Foreground, specifiedColor=<rgba 1 0 1 1>}
     shadowColorAndOpacity=<rgba 0 0 0 1>
     shadowOffset=<0 0>
     strikeBrush {
      bIsDynamicallyLoaded=false
      drawAs=Image
      imageType=NoImage
      mirroring=NoMirror
      resourceName=None
      resourceObject=null
      tiling=NoTile
      imageSize=<32 32>
      margin=<ltrb 0 0 0 0>
      outlineSettings {
       bUseBrushTransparency=false
       roundingType=HalfHeightRadius
       width=0
       color {colorUseRule=UseColor_Specified, specifiedColor=<rgba 0 0 0 0>}
       cornerRadii=<0 0 0 0>
      }
      tintColor {colorUseRule=UseColor_Specified, specifiedColor=<rgba 1 1 1 1>}
      uVRegion {bIsValid=false, max=<0 0>, min=<0 0>}
     }
     underlineBrush {
      bIsDynamicallyLoaded=false
      drawAs=Image
      imageType=NoImage
      mirroring=NoMirror
      resourceName=None
      resourceObject=null
      tiling=NoTile
      imageSize=<32 32>
      margin=<ltrb 0 0 0 0>
      outlineSettings {
       bUseBrushTransparency=false
       roundingType=HalfHeightRadius
       width=0
       color {colorUseRule=UseColor_Specified, specifiedColor=<rgba 0 0 0 0>}
       cornerRadii=<0 0 0 0>
      }
      tintColor {colorUseRule=UseColor_Specified, specifiedColor=<rgba 1 1 1 1>}
      uVRegion {bIsValid=false, max=<0 0>, min=<0 0>}
     }
    }
   }
   slot {class=VerticalBoxSlot, properties {padding=<ltrb 0 8 0 0>}}
  }
  LegalandActionBar {
   widgetClassPath=&HorizontalBox
   bIsVariable=true
   parent=MainContent_Grid
   widget=LegalandActionBar
   properties {bOverride_Cursor=true}
   slot {class=GridSlot, properties {columnSpan=3, row=3, verticalAlignment=VAlign_Bottom}}
  }
  LegalText {
   widgetClassPath=&CommonTextBlock
   bIsVariable=true
   parent=LegalandActionBar
   widget=LegalText
   properties {
    bIsScrollingEnabled=false
    mobileFontSizeMultiplier=1.4
    wrapTextAt=720
    colorAndOpacity {colorUseRule=UseColor_Specified, specifiedColor=<rgba 1 1 1 0.6>}
    font {
     bForceMonospaced=false
     bMaterialIsStencil=false
     fontMaterial=null
     fontObject=&HeadingNow
     letterSpacing=10
     monospacedWidth=1
     size=11.25
     skewAmount=0
     typefaceFontName=74 Regular
     outlineSettings {
      bApplyOutlineToDropShadows=false
      bMiteredCorners=false
      bSeparateFillAlpha=false
      outlineMaterial=null
      outlineSize=0
      outlineColor=<rgba 0 0 0 1>
     }
    }
    shadowOffset=<0 0>
   }
   slot {class=HorizontalBoxSlot, properties {verticalAlignment=VAlign_Center, size {sizeRule=Fill, value=1}}}
  }
  Spacer_329 {
   widgetClassPath=&Spacer
   parent=LegalandActionBar
   widget=Spacer_329
   slot {class=HorizontalBoxSlot, properties {verticalAlignment=VAlign_Center, size {sizeRule=Fill, value=1}}}
  }
  WBP_ActionBarContainer {
   widgetClassPath=/Game/UI/Frontend/BottomBar/WBP_ActionBarContainer
   bIsVariable=true
   parent=LegalandActionBar
   widget=WBP_ActionBarContainer
   slot {class=HorizontalBoxSlot, properties {horizontalAlignment=HAlign_Right, verticalAlignment=VAlign_Center}}
  }
  Button_Bar {widgetClassPath=&HorizontalBox, namedSlotHost=WBP_ActionBarContainer, widget=Button_Bar}
  Button_BuyRewards {
   widgetClassPath=/FortUILibrary/UIKit/Buttons/WBP_UIKit_Button_Quiet
   bIsVariable=true
   parent=Button_Bar
   widget=Button_BuyRewards
   properties {
    bIsFocusable=false
    text=Buy Rewards
    visibility=Visible
    triggeringInputAction {dataTable=&InputActionDataTable, rowName=GenericFaceButtonLeft}
   }
   slot {class=HorizontalBoxSlot, properties {verticalAlignment=VAlign_Center, padding=<ltrb 0 0 16 0>}}
  }
  Button_MoreInfo {
   widgetClassPath=&WBP_UIKit_Button_Quiet
   bIsVariable=true
   parent=Button_Bar
   widget=Button_MoreInfo
   properties {
    bIsFocusable=false
    text=More Info
    visibility=Visible
    triggeringInputAction {dataTable=&InputActionDataTable, rowName=GenericRightStickClick}
   }
   slot {class=HorizontalBoxSlot, properties {verticalAlignment=VAlign_Center, padding=<ltrb 0 0 16 0>}}
  }
  WBP_AthenaBottomBar_WithNotifications {
   widgetClassPath=/Game/UI/Frontend/BottomBar/WBP_AthenaBottomBar_WithNotifications
   bIsVariable=true
   parent=Button_Bar
   widget=WBP_AthenaBottomBar_WithNotifications
   slot {class=HorizontalBoxSlot}
  }
  Foreground_CustomSlot {
   widgetClassPath=&NamedSlot
   bIsVariable=true
   parent=MainContent
   widget=Foreground_CustomSlot
   properties {visibility=HitTestInvisible}
   slot {class=OverlaySlot, properties {horizontalAlignment=HAlign_Fill, verticalAlignment=VAlign_Fill}}
  }
  Error_Container {widgetClassPath=&Overlay, parent=ContentSwitcher, widget=Error_Container, slot {class=CommonVisibilitySwitcherSlot}}
  ErrorMessage_Text {
   widgetClassPath=&CommonRichTextBlock
   bIsVariable=true
   parent=Error_Container
   widget=ErrorMessage_Text
   properties {
    autoWrapText=true
    bOverrideDefaultStyle=true
    justification=Center
    lineHeightPercentage=1.15
    mobileTextBlockScale=1.2
    text=Error loading data
    textTransformPolicy=ToUpper
    defaultTextStyleOverride {
     overflowPolicy=Clip
     transformPolicy=None
     colorAndOpacity {colorUseRule=UseColor_Specified, specifiedColor=<rgba 1 1 1 1>}
     font {
      bForceMonospaced=false
      bMaterialIsStencil=false
      fontMaterial=null
      fontObject=&HeadingNow
      letterSpacing=0
      monospacedWidth=1
      size=32.25
      skewAmount=0
      typefaceFontName=HeadingNow-86BoldItalic
      outlineSettings {
       bApplyOutlineToDropShadows=false
       bMiteredCorners=false
       bSeparateFillAlpha=false
       outlineMaterial=null
       outlineSize=0
       outlineColor=<rgba 0 0 0 1>
      }
     }
     highlightColor {colorUseRule=UseColor_Specified, specifiedColor=<rgba 0 0 0 0>}
     highlightShape {
      bIsDynamicallyLoaded=false
      drawAs=Image
      imageType=NoImage
      mirroring=NoMirror
      resourceName=None
      resourceObject=null
      tiling=NoTile
      imageSize=<32 32>
      margin=<ltrb 0 0 0 0>
      outlineSettings {
       bUseBrushTransparency=false
       roundingType=HalfHeightRadius
       width=0
       color {colorUseRule=UseColor_Specified, specifiedColor=<rgba 0 0 0 0>}
       cornerRadii=<0 0 0 0>
      }
      tintColor {colorUseRule=UseColor_Specified, specifiedColor=<rgba 1 1 1 1>}
      uVRegion {bIsValid=false, max=<0 0>, min=<0 0>}
     }
     selectedBackgroundColor {colorUseRule=UseColor_Foreground, specifiedColor=<rgba 1 0 1 1>}
     shadowColorAndOpacity=<rgba 0 0 0 1>
     shadowOffset=<0 0>
     strikeBrush {
      bIsDynamicallyLoaded=false
      drawAs=Image
      imageType=NoImage
      mirroring=NoMirror
      resourceName=None
      resourceObject=null
      tiling=NoTile
      imageSize=<32 32>
      margin=<ltrb 0 0 0 0>
      outlineSettings {
       bUseBrushTransparency=false
       roundingType=HalfHeightRadius
       width=0
       color {colorUseRule=UseColor_Specified, specifiedColor=<rgba 0 0 0 0>}
       cornerRadii=<0 0 0 0>
      }
      tintColor {colorUseRule=UseColor_Specified, specifiedColor=<rgba 1 1 1 1>}
      uVRegion {bIsValid=false, max=<0 0>, min=<0 0>}
     }
     underlineBrush {
      bIsDynamicallyLoaded=false
      drawAs=Image
      imageType=NoImage
      mirroring=NoMirror
      resourceName=None
      resourceObject=null
      tiling=NoTile
      imageSize=<32 32>
      margin=<ltrb 0 0 0 0>
      outlineSettings {
       bUseBrushTransparency=false
       roundingType=HalfHeightRadius
       width=0
       color {colorUseRule=UseColor_Specified, specifiedColor=<rgba 0 0 0 0>}
       cornerRadii=<0 0 0 0>
      }
      tintColor {colorUseRule=UseColor_Specified, specifiedColor=<rgba 1 1 1 1>}
      uVRegion {bIsValid=false, max=<0 0>, min=<0 0>}
     }
    }
   }
   slot {class=OverlaySlot, properties {horizontalAlignment=HAlign_Fill, verticalAlignment=VAlign_Center, padding=<ltrb 0 0 0 16>}}
  }
 ]
}
