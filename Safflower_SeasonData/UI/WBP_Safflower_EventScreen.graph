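# Safflower season event screen, a child of WBP_MSEE_Screen that fills it with key art, a reward track and a chase widget.
# It hands these widgets to the parent, routes track clicks to inspection and hovers to the parent, and feeds theme images into the key art.
# The intro plays Anim_Reveal, character reveal and a background Material Parameter Collection animation together, then calls AfterIntroPlays.
# On activation, key art top padding adapts to tablet aspect ratio.
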WidgetBlueprint /Safflower_SeasonData/UI/WBP_Safflower_EventScreen
animations [Intro {
  binding_count=0
  length_seconds=0.5
  object=&self:Intro
  sequence {
   bindings []
   tracks [{
     type=MovieSceneMaterialParameterCollectionTrack
     display_name=MPC_Safflower_BackgroundStates
     sections [{
       type=MovieSceneParameterSection
       channels [
        Anim_CloudHeight {type=MovieSceneScriptingFloatChannel, keys [[0,1],[8,0]]}
        Anim_GroundHeight {type=MovieSceneScriptingFloatChannel, keys [[0,1],[10,0]]}
        Anim_HideGround {type=MovieSceneScriptingFloatChannel, keys [[0,1],[10,0]]}
       ]
      }]
    }]
  }
 }]
graphs [
 BP_GetDesiredFocusTarget {nodes [
   Get Desired Focus Target()
   > n1 = Parent: Get Desired Focus Target!(execute) -> (then, ReturnValue:Widget Object Reference) {is=title K2Node_CallParentFunction}
   > Return Node!(execute, ReturnValue:Widget Object Reference=@n1)
  ]}
 BP_GetDesiredInputConfig {nodes [
   Get Desired Input Config() {is=title K2Node_FunctionEntry}
   > n1 = Parent: Get Desired Input Config!(execute) -> (then, ReturnValue:UIInput Config Structure) {is=title K2Node_CallParentFunction}
   > Return Node(ReturnValue:UIInput Config Structure=@n1)
  ]}
 EventGraph {nodes [
   n1 = "On Reward Track Populated (RewardTrack)"!() -> (OutputDelegate, then, RewardTrack:Dynamic Entry Box Object Reference) {is=title K2Node_ComponentBoundEvent}
   > SetRewardTrack!(execute, self:WBP MSEE Screen Object Reference, RewardTrack:Dynamic Entry Box Object Reference=@n1.RewardTrack) -> (then)
   n2 = "On Item Clicked (RewardTrack)"!() -> (OutputDelegate, then, RewardVM:Season Pass Reward VM Object Reference, ItemVM:Fort Item VM Object Reference) {is=title K2Node_ComponentBoundEvent}
   > LoadInspectionScreen!(
    execute
    self:WBP MSEE Screen Object Reference
    ShowChaseWidget:bool=false
    RewardVM:Season Pass Reward VM Object Reference=@n2.RewardVM
    FortItemVM:Fort Item VM Object Reference=@n2.ItemVM
   ) -> (then)
   n3 = "On Item Hovered (RewardTrack)"!() -> (OutputDelegate, then, RewardVM:Season Pass Reward VM Object Reference, FortItemVM:Fort Item VM Object Reference, ItemIndex:int=0) {is=title K2Node_ComponentBoundEvent}
   > GetHoveredItemVM!(
    execute
    self:WBP MSEE Screen Object Reference
    SeasonReward VM:Season Pass Reward VM Object Reference=@n3.RewardVM
    FortItemVM:Fort Item VM Object Reference=@n3.FortItemVM
    Hovered Item Index:int=@n3.ItemIndex
   ) -> (then)
   n4 = Event PreConstruct()
   > Parent: PreConstruct(IsDesignTime=@n4.IsDesignTime)
   > SetChaseWidget!(
    execute
    self:WBP MSEE Screen Object Reference
    ChaseWidget:WBP MSEE Chase Reward Progress Object Reference=Get ChaseWidget() -> (ChaseWidget:WBP Safflower Chase Widget Object Reference)
   ) -> (then)
   > SetKeyArt!(
    execute
    self:WBP MSEE Screen Object Reference
    CMSKeyArt:Fort Lazy Image Object Reference=Get CMS_KeyArt_Main!(self:WBP Safflower Key Art Object Reference=Get WBP_Safflower_KeyArt() -> (WBP_Safflower_KeyArt:WBP Safflower Key Art Object Reference)) -> (CMS_KeyArt_Main:Fort Lazy Image Object Reference)
   ) -> (then)
   Event PlayIntroAnim() {is=title K2Node_Event}
   > n5 = Play Animation with Finished event!(
    execute:Exec
    Widget:User Widget Object Reference=Self-Reference()
    InAnimation:Widget Animation Object Reference=Get Anim_Reveal() -> (Anim_Reveal:Widget Animation Object Reference)
    StartAtTime:float=0
    NumLoopsToPlay:int=1
    PlayMode:EUMGSequencePlayMode Enum
    PlaybackSpeed:float=2
   ) -> (then:Exec, Result:Widget Animation Handle Structure, Finished:Exec) {is=title K2Node_PlayAnimation2}
   > PlayAnimationForward(InAnimation=Get Intro())
   n6 = Get WBP_Safflower_KeyArt()
   > PlayAnimationForward(
    self=@n6
    InAnimation=Get CharacterReveal!(self:WBP Safflower Key Art Object Reference=@n6) -> (CharacterReveal:Widget Animation Object Reference)
   )
   AfterIntroPlays(execute=@n5.Finished, self:WBP MSEE Screen Object Reference)
   Event On Activated()
   > Parent: On Activated() {is=title K2Node_CallParentFunction}
   > SetPadding!(
    execute
    self:Overlay Slot Object Reference=SlotAsOverlaySlot(Widget=Get MainKeyArtScaleBox() -> (MainKeyArtScaleBox:Scale Box Object Reference))
    InPadding_Left:float=0
    InPadding_Top:float=Select(Option 0:float=170, Option 1:float=350, Index=IsTabletAspectRatio!() -> (ReturnValue:bool)) -> (ReturnValue:float)
    InPadding_Right:float=0
    InPadding_Bottom:float=0
   ) -> (then)
   n7 = Event OnThemeDataLoaded!() -> (OutputDelegate, then, ThemeData:Season Pass Theme Data Structure) {is=title K2Node_Event}
   > UpdateImageRefs!(
    execute
    self:WBP Safflower Key Art Object Reference=Get WBP_Safflower_KeyArt()
    ImageTable:"Fort Image Reference Table Structure (by ref)"=Break Season Pass Theme Data!(SeasonPassThemeData:"Season Pass Theme Data Structure (by ref)"=@n7.ThemeData) -> (
     TitleColor:Linear Color Structure
     TitleMaterial:Soft Object Reference
     TitleOutline:Font Outline Settings Structure
     SeasonColor:Linear Color Structure
     ProgressionColor:Linear Color Structure
     OverlayColor:Linear Color Structure
     VaultBackgroundMaterial:Soft Object Reference
     GiftboxVaultBackgroundMaterial:Soft Object Reference
     UIBackgroundMaterial:Soft Object Reference
     UpsellBackgroundMaterial:Soft Object Reference
     CrewCategoryBackgroundMaterial:Soft Object Reference
     CustomLogo:Soft Object Reference
     CustomSeasonTitle:Text
     CustomSeasonTitleColor:Linear Color Structure
     Upsell2BackgrounMaterial:Soft Object Reference
     ImageReferenceTable:Fort Image Reference Table Structure
     ColorReferenceTable:Fort Color Reference Table Structure
     TextReferenceTable:Fort Text Reference Table Structure
    ) {is=title K2Node_BreakStruct}.ImageReferenceTable
   ) -> (then)
  ]}
 GetPrimaryContentSetup {nodes [
   GetPrimaryContentSetup() {is=title K2Node_FunctionEntry}
   > Return Node!(
    execute
    ReturnValue_bShowBottomBar:bool=True
    ReturnValue_bShowChatWidget:bool=True
    ReturnValue_bEnableNewTextChat:bool=False
    ReturnValue_bShowTopBar:bool=True
    ReturnValue_bTabSupportsInterface:bool=False
   )
  ]}
]
variables []
widgets {
 info {
  inheritedWidgetCount=0
  namedSlotCount=5
  parentClass=/SeasonPass/MSEE/WBP_MSEE_Screen.WBP_MSEE_Screen_C
  rootWidgetClass=null
  widgetCount=10
 }
 widgets [
  WBP_Safflower_RewardTrack {widgetClassPath=&/WBP_Safflower_RewardTrack, bIsVariable=true, widget=WBP_Safflower_RewardTrack}
  GridPanel_1 {widgetClassPath=&GridPanel, widget=GridPanel_1}
  WBP_Safflower_ChaseWidget {
   widgetClassPath=&/WBP_Safflower_ChaseWidget
   bIsVariable=true
   parent=GridPanel_1
   widget=WBP_Safflower_ChaseWidget
   slot {class=GridSlot}
  }
  PH_LazyImage {
   widgetClassPath=&FortLazyImage
   bIsVariable=true
   parent=GridPanel_1
   widget=PH_LazyImage
   properties {
    brush {
     bIsDynamicallyLoaded=false
     drawAs=Image
     imageType=FullColor
     mirroring=NoMirror
     resourceName=None
     resourceObject=/FortUILibrary/Textures/MaterialInstances/MI_UIKit_Icon_Gift
     tiling=NoTile
     imageSize=<200 200>
     margin=<ltrb 0 0 0 0>
     outlineSettings {
      bUseBrushTransparency=false
      roundingType=HalfHeightRadius
      width=0
      color {colorUseRule=UseColor_Specified, specifiedColor=<rgba 0 0 0 0>}
      cornerRadii=<0 0 0 0>
     }
     tintColor {colorUseRule=UseColor_Specified, specifiedColor=<rgba 1 1 1 1>}
     uVRegion {bIsValid=false, max=<0 0>, min=<0 0>}
    }
    loadingBackgroundBrush {
     bIsDynamicallyLoaded=false
     drawAs=Image
     imageType=FullColor
     mirroring=NoMirror
     resourceName=None
     resourceObject=null
     tiling=NoTile
     imageSize=<32 32>
     margin=<ltrb 0 0 0 0>
     outlineSettings {
      bUseBrushTransparency=false
      roundingType=HalfHeightRadius
      width=0
      color {colorUseRule=UseColor_Specified, specifiedColor=<rgba 0 0 0 0>}
      cornerRadii=<0 0 0 0>
     }
     tintColor {colorUseRule=UseColor_Specified, specifiedColor=<rgba 1 1 1 1>}
     uVRegion {bIsValid=false, max=<0 0>, min=<0 0>}
    }
   }
   slot {class=GridSlot, properties {horizontalAlignment=HAlign_Center, row=1, verticalAlignment=VAlign_Center}}
  }
  RewardTrack {widgetClassPath=&WBP_Safflower_RewardTrack, bIsVariable=true, widget=RewardTrack, properties {progressPreview=0.1}}
  ChaseWidget_Grid {widgetClassPath=&GridPanel, widget=ChaseWidget_Grid, properties {bOverride_Cursor=true}}
  ChaseWidget {
   widgetClassPath=&WBP_Safflower_ChaseWidget
   bIsVariable=true
   parent=ChaseWidget_Grid
   widget=ChaseWidget
   slot {class=GridSlot, properties {horizontalAlignment=HAlign_Center, verticalAlignment=VAlign_Center, padding=<ltrb 0 0 16 0>}}
  }
  Overlay_0 {widgetClassPath=&Overlay, widget=Overlay_0}
  MainKeyArtScaleBox {
   widgetClassPath=&ScaleBox
   bIsVariable=true
   parent=Overlay_0
   widget=MainKeyArtScaleBox
   slot {class=OverlaySlot, properties {horizontalAlignment=HAlign_Fill, verticalAlignment=VAlign_Fill, padding=<ltrb 0 200 0 0>}}
  }
  WBP_Safflower_KeyArt {
   widgetClassPath=&/WBP_Safflower_KeyArt
   bIsVariable=true
   parent=MainKeyArtScaleBox
   widget=WBP_Safflower_KeyArt
   properties {renderTransform {angle=0, scale=<1 1>, shear=<0 0>, translation=<-60 0>}}
   slot {class=ScaleBoxSlot, properties {verticalAlignment=VAlign_Bottom}}
  }
 ]
}
