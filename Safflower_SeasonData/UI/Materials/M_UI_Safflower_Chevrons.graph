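# Translucent UI material for Safflower season content that draws a chevron arrow with a foil finish and drop shadow.
# Screen-space scale-to-fit UVs sample a chevron distance field, anti-aliased with adjustable softness into the arrow mask.
# A Y-offset second sample forms the shadow, tinted by a color parameter and faintly added to the combined shape.
# A center-scaled foil texture, softened by a lerp, supplies surface detail, with named reroutes linking the parts.
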Material /Safflower_SeasonData/UI/Materials/M_UI_Safflower_Chevrons
graph {
 properties {
  bAutomaticallySetUsageInEditor=false
  bUsedWithStaticMesh=false
  blendMode=BLEND_Translucent
  materialDomain=MD_UI
  usedShadingModels=Default Lit
 }
 nodes [
  LinearInterpolate(B=LinearInterpolate(A=NamedRerouteUsage(), B=NamedRerouteUsage(), Alpha=NamedRerouteUsage()), Alpha=NamedRerouteUsage())
  n1 = NamedRerouteDeclaration(None=MaterialFunctionCall!(
    ScreenUV=MaterialFunctionCall() {materialFunction=&GetUserInterfaceUV}.Normalized UV
    TextureSize
    ScalePivotPoint
   ) {materialFunction=/FortUILibrary/MaterialFunctions/Basics/MF_UI_ScreenSpaceUVScaling}.Scaled to Fit) {name=Arrow UVs, nodeColor=<rgba 0 0 0 1>}
  NamedRerouteDeclaration(None=MaterialFunctionCall(
    DistanceField=TextureSample(UVs=@n1) {texture=/Safflower_SeasonData/UI/Textures/T_UI_Safflower_Chevron}.A
    Threshold=Constant() {r=0.5}
    Smoothing=ScalarParameter() {group=Shape, parameterName=ArrowSoftness, sliderMax=1}
   ) {materialFunction=&MF_UI_AntiAliasing}.Outer) {name=Arrow Mask, nodeColor=<rgba 0 0 0 1>}
  n2 = MaterialFunctionCall(UVs=TextureCoordinate(), Texture Scale=Constant() {r=3}) {materialFunction=&ScaleUVsByCenter}
  NamedRerouteDeclaration(None=LinearInterpolate(A=TextureSample(UVs=@n2.UVs) {texture=/Safflower_SeasonData/UI/Textures/T_UI_Safflower_FoilTexture}.RGB) {constAlpha=0.2}) {name=Foil Texture, nodeColor=<rgba 0 0 0 1>}
  n3 = NamedRerouteUsage()
  n4 = AppendVector(A=Constant(), B=ScalarParameter() {defaultValue=-0.045, parameterName=ShadowY})
  n5 = MaterialFunctionCall(
   DistanceField=TextureSample(UVs=Add(A=@n3, B=@n4)) {texture=&T_UI_Safflower_Chevron}.A
   Threshold=Constant() {r=0.5}
   Smoothing=ScalarParameter() {group=Shape, parameterName=ArrowSoftness, sliderMax=1}
  ) {materialFunction=&MF_UI_AntiAliasing}
  n6 = NamedRerouteUsage()
  NamedRerouteDeclaration(None=Saturate(None=Subtract(A=@n5.Outer, B=@n6))) {name=Arrow Shadow, nodeColor=<rgba 0 0 0 1>}
  n7 = NamedRerouteUsage()
  n8 = LinearInterpolate(B=@n5.Outer) {constAlpha=0.15}
  NamedRerouteDeclaration(None=Saturate(None=Add(A=@n7, B=@n8))) {name=Combined Arrow Shape, nodeColor=<rgba 0 0 0 1>}
  NamedRerouteDeclaration(None=VectorParameter() {group=Color, parameterName=ShadowColor}.RGB) {name=Shadow Color, nodeColor=<rgba 0 0 0 1>}
 ]
}
