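# Translucent UI material that draws a procedural speech bubble shape for the Safflower season UI.
# A rounded-rectangle SDF with parameterized corner radii is unioned with a translated, rescaled triangle SDF that forms the tail.
# A narrow SmoothStep and OneMinus turn the merged distance field into a crisp filled mask.
# A vertex transform scales by BubbleScale and publishes the resulting pixel size through the NewPixelSize named reroute.
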Material /Safflower_SeasonData/UI/Materials/M_UI_Safflower_SpeechBubble
graph {
 properties {
  bUsedWithStaticMesh=false
  blendMode=BLEND_Translucent
  materialDomain=MD_UI
  pixelDepthOffsetMode=PDOM_Legacy
  usedShadingModels=Default Lit
 }
 nodes [
  Constant3Vector() {constant=<rgba 1 1 1 1>}
  n1 = NamedRerouteUsage()
  MaterialFunctionCall!(Value PX=@n1, UV Pixel Size, X or Y) {materialFunction=/FortUILibrary/MaterialFunctions/Basics/MF_UI_PixelsToNormalized}
  NamedRerouteDeclaration(None=MaterialFunctionCall!(Anchor, Translation, Padding, Scale=VectorParameter() {parameterName=BubbleScale, defaultValue=<rgba 3 3 0 1>}.RGB) {materialFunction=/FortUILibrary/MaterialFunctions/Basics/MF_UI_VertexTransform}.New Pixel Size) {name=NewPixelSize, nodeColor=<rgba 0 1 0.0391 1>}
  Normalize(VectorInput=@n1)
  n2 = VectorParameter() {group=Corners, parameterName=Corner Radius, sortPriority=2, defaultValue=<rgba 0.25 0.25 0.25 0.25>}
  n3 = MaterialFunctionCall(
   Size=Constant2Vector() {g=0.33, r=0.33}
   Corner Radius=AppendVector(A=@n2.RGB, B=@n2.RGB)
   Radius in Pixels=ScalarParameter() {group=Corners, parameterName=Radius in Pixels, sliderMax=1, sortPriority=0}
  ) {materialFunction=&MF_UI_CoreUI_RoundedMask}
  n4 = MaterialFunctionCall!(Input UVs, Translate=Constant2Vector() {g=-0.216, r=0.07}) {materialFunction=/FortUILibrary/MaterialFunctions/Basics/MF_UI_Translate}
  n5 = Constant2Vector() {g=0.15, r=0.15}
  n6 = MaterialFunctionCall!(
   UVs=MaterialFunctionCall(
    UV=MaterialFunctionCall(UVs=@n4.UVs raw, Texture Scale=@n5) {materialFunction=&ScaleUVsByCenter}.UVs
    ScalePivotPoint=Constant2Vector() {r=1.049}
   ) {materialFunction=&MF_UI_UVScaling}.Scaled to Fit
   Corner 01=Constant2Vector() {g=-1, r=-1}
   Corner 02=Constant2Vector() {g=-1, r=1}
   Corner 03=Constant2Vector() {g=1, r=1}
   Glow Min
   Glow Max
   Outline Thickness
  ) {materialFunction=/FortUILibrary/MaterialFunctions/Shapes/MF_UI_WIP_Triangle}
  OneMinus(None=SmoothStep(Value=MaterialFunctionCall!(DistanceFieldA=@n3.SDF, DistanceFieldB=@n6.SDF, BlendAmount) {materialFunction=/FortUILibrary/MaterialFunctions/Shapes/Modifiers/MF_UI_SDF_Union}.Exact) {constMax=0.005})
 ]
}
