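# Translucent UI material that draws the triangular arrow of a speech bubble, judging by its name, for the Safflower season UI.
# Normalized UI UVs are shifted one pixel along V and fed to a triangle SDF function with fixed, asymmetric corners, giving a skewed pointer.
# An anti-aliasing function smooths the distance field into a shape mask, and a FillColour parameter supplies the tint, white by default.
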Material /Safflower_SeasonData/UI/Materials/M_UI_Safflower_SpeechBubbleArrow
graph {
 properties {
  bAutomaticallySetUsageInEditor=false
  bUsedWithStaticMesh=false
  blendMode=BLEND_Translucent
  materialDomain=MD_UI
  usedShadingModels=Default Lit
 }
 nodes [
  n1 = MaterialFunctionCall() {materialFunction=&GetUserInterfaceUV}
  n2 = AppendVector(
   A=Constant()
   B=MaterialFunctionCall!(Value PX=Constant() {r=1}, UV Pixel Size, X or Y) {materialFunction=/FortUILibrary/MaterialFunctions/Basics/MF_UI_PixelsToNormalized}."Normalized (>1px)"
  )
  MaterialFunctionCall(DistanceField=MaterialFunctionCall!(
    UVs=Subtract(A=@n1.Normalized UV, B=@n2)
    Corner 01=Constant2Vector() {g=1, r=-0.65}
    Corner 02=Constant2Vector() {g=-1, r=-1}
    Corner 03=Constant2Vector() {g=1, r=1}
    Glow Min
    Glow Max
    Outline Thickness
   ) {materialFunction=/FortUILibrary/MaterialFunctions/Shapes/MF_UI_WIP_Triangle}.SDF) {materialFunction=&MF_UI_AntiAliasing}
  VectorParameter() {parameterName=FillColour, defaultValue=<rgba 1 1 1 1>}
 ]
}
