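# Speech bubble widget that shows timed dialog lines from primary and secondary text pools.
# TriggerDialog picks an indexed line, then chained timers open, close and finish the bubble, delaying secondary lines by the speech duration and broadcasting SlotIsSpeaking and FinishedCycle.
# Empty lines wait out the cycle; StartRandomLineLoop shows random primary lines, and StopSpeech clears all timers.
# AdjustArrow flips the bubble through the backplate material's BubbleScale parameter.
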WidgetBlueprint /Safflower_SeasonData/UI/WBP_SpeechBubble
graphs [
 AdjustArrow {nodes [
   n1 = AdjustArrow!() -> (then, FlipArrow_H:bool=~, FlipSpeechBubble_V:bool=~) {is=title K2Node_FunctionEntry}
   > n2 = GetDynamicMaterial(self=Get Backplate() -> (Backplate:Image Object Reference))
   > SetVectorParameterValue!(
    execute
    self=@n2
    ParameterName=BubbleScale
    Value_R:float=SelectFloat(A=-3, B=3, bPickA=@n1)
    Value_G:float=SelectFloat(A=-3, B=3, bPickA=@n1.FlipSpeechBubble_V)
    Value_B:float=0
    Value_A:float=0
   ) -> (then)
  ]}
 EventGraph {nodes [
   n1 = OpenSpeechBubble() {is=title K2Node_CustomEvent}
   > ShowSpeechBubble!(execute, self, ShowSpeechBubble:bool=true) -> (then)
   > n2 = Call Slot Is Speaking!(execute:Exec, self:WBP Speech Bubble Object Reference, IsSpeaking:bool=true) -> (then:Exec) {is=title K2Node_CallDelegate}
   n3 = Get SpeechDuration() -> (SpeechDuration:double)
   n4 = float - float(A=@n3, B=1.5)
   n5 = Get IsPrimary() -> (IsPrimary:bool)
   n6 = CloseSpeechBubble() {is=title K2Node_CustomEvent}
   n7 = Set Timer by Event(execute=@n2.then, Delegate=@n6, Time="Clamp (Float)"(Value=SelectFloat(A=@n3, B=@n4, bPickA=@n5), Min=2, Max=10), bLooping=false)
   > Set SpeechBubbleActive(SpeechBubbleActive:Timer Handle Structure=@n7) -> (Output_Get:Timer Handle Structure)
   DialogSpeechCycle() {is=title K2Node_CustomEvent}
   n8 = Get SpeechDuration()
   n9 = Get IsPrimary()
   n10 = Get DelayToStart() -> (DelayToStart:double)
   > n11 = Set Timer by Event(Delegate=@n1, Time=float + float(A=SelectFloat(A=@n10, B=@n8, bPickA=@n9), B=0.1), bLooping=false)
   > Set PreSpeechBubbleTimer(PreSpeechBubbleTimer:Timer Handle Structure=@n11) -> (Output_Get:Timer Handle Structure)
   n12 = FinishCycle() {is=title K2Node_CustomEvent}
   > Call Finished Cycle!(execute:Exec, self:WBP Speech Bubble Object Reference) -> (then:Exec) {is=title K2Node_CallDelegate}
   SkipLine() {is=title K2Node_CustomEvent}
   > n13 = Set Timer by Event(
    Delegate=Create Event()
    Time=float + float!(A=Get DelayToStart(), B=Get SpeechDuration(), C:double=Get DelayToNextLoop() -> (DelayToNextLoop:double)) -> (ReturnValue)
    bLooping=false
   )
   > Set PostSpeechBubbleTimer(PostSpeechBubbleTimer:Timer Handle Structure=@n13) -> (Output_Get:Timer Handle Structure)
   n14 = Event PreConstruct()
   > Set WrapTextAt!(
    execute
    WrapTextAt:float=Get WrapTextAt() -> (WrapTextAt:float)
    self:Text Layout Widget Object Reference=Get Text!() -> (Text:Common Rich Text Block Object Reference)
   ) -> (then, Output_Get:float)
   > AdjustArrow!(
    execute
    self
    FlipArrow_H:bool=Get Flip Arrow Horz() -> (Flip Arrow Horz:bool)
    FlipSpeechBubble_V:bool=Get Flip Speech Bubble Vert() -> (Flip Speech Bubble Vert:bool)
   ) -> (then) {is=title K2Node_CallFunction}
   > n15 = Branch(Condition=@n14.IsDesignTime)
   > ShowSpeechBubble(ShowSpeechBubble=true)
   > SetText(self=Get Text(), InText="INVTEXT(\"[PH] Test Text\")")
   ShowSpeechBubble(execute=@n15.else, ShowSpeechBubble=false)
   n16 = Set Timer by Event(execute=@n17.then, Delegate=@n12, Time=Get DelayToNextLoop(), bLooping=false)
   > Set PostSpeechBubbleTimer(PostSpeechBubbleTimer=@n16)
   n17 = Call Slot Is Speaking(execute=@n18.then)
   n18 = ShowSpeechBubble(execute=@n6.then, ShowSpeechBubble=false)
  ]}
 FinishedCycle {nodes [FinishedCycle() {is=title K2Node_FunctionEntry}]}
 SetSpeechTexts {nodes [
   n1 = SetSpeechTexts!() -> (then, Primary Strings:Array of Texts, Secondary Strings:Array of Texts) {is=title K2Node_FunctionEntry}
   > Set PrimaryStringPool(PrimaryStringPool:Array of Texts=@n1) -> (Output_Get:Array of Texts)
   > Set SecondaryStringPool(SecondaryStringPool:Array of Texts=@n1.Secondary Strings) -> (Output_Get:Array of Texts)
  ]}
 SetSpeechTimeOptions {nodes [
   n1 = SetSpeechTimeOptions!() -> (then, DelayToStart:double=~, SpeechDuration:double=~, DelayToNextLoop:double=~) {is=title K2Node_FunctionEntry}
   > Set DelayToStart(DelayToStart:double=@n1) -> (Output_Get:double)
   > Set SpeechDuration(SpeechDuration:double=@n1.SpeechDuration) -> (Output_Get:double)
   > Set DelayToNextLoop(DelayToNextLoop:double=@n1.DelayToNextLoop) -> (Output_Get:double)
  ]}
 ShowSpeechBubble {nodes [
   n1 = ShowSpeechBubble!() -> (then, ShowSpeechBubble:bool=~) {is=title K2Node_FunctionEntry}
   > SetVisibility(self=Get Content() -> (Content:Grid Panel Object Reference), InVisibility=Select(Option 0=Hidden, Option 1=HitTestInvisible, Index=@n1))
  ]}
 SlotIsSpeaking {nodes [SlotIsSpeaking!() -> (then, IsSpeaking:bool=~) {is=title K2Node_FunctionEntry}]}
 StartRandomLineLoop {nodes [
   StartRandomLineLoop() {is=title K2Node_FunctionEntry}
   n1 = Get PrimaryStringPool() -> (PrimaryStringPool:Array of Texts)
   n2 = int + int(A=Last Index!(TargetArray:Array of Texts=@n1) -> (ReturnValue=0))
   > SetText(
    self=Get Text()
    InText="Get (a copy)"(Array:Array of Texts=@n1, Dimension 1=RandomInteger!(Max:int=@n2) -> (ReturnValue:int=0)) -> (Output:Text)
   )
   > ShowSpeechBubble(ShowSpeechBubble=true) {is=title K2Node_CallFunction}
   > n3 = Set Timer by Event(
    Delegate=Create Event()
    Time=float + float!(A=Get DelayToStart(), B=Get SpeechDuration(), C=Get DelayToNextLoop(), D:double=1) -> (ReturnValue)
    bLooping=true
    bMaxOncePerFrame=true
   )
   > Set SingleLoop(SingleLoop:Timer Handle Structure=@n3) -> (Output_Get:Timer Handle Structure)
  ]}
 StopSpeech {nodes [
   StopSpeech() {is=title K2Node_FunctionEntry}
   > n1 = For Each Loop(Array:Array of Timer Handle Structures=Make Array!(
     "[0]":Timer Handle Structure=Get SingleLoop() -> (SingleLoop:Timer Handle Structure)
     "[1]":Timer Handle Structure=Get PreSpeechBubbleTimer() -> (PreSpeechBubbleTimer:Timer Handle Structure)
     "[2]":Timer Handle Structure=Get SpeechBubbleActive() -> (SpeechBubbleActive:Timer Handle Structure)
     "[3]":Timer Handle Structure=Get PostSpeechBubbleTimer() -> (PostSpeechBubbleTimer:Timer Handle Structure)
    ) -> (Array:Array of Timer Handle Structures)) -> (Array Element:Timer Handle Structure)
   ShowSpeechBubble(execute=@n1.Completed, ShowSpeechBubble=false)
   > Call Slot Is Speaking()
   Pause Timer by Handle!(execute=@n1.LoopBody, Handle:Timer Handle Structure=@n1) -> (then)
   > Clear and Invalidate Timer by Handle(Handle=@n1)
  ]}
 TriggerDialog {nodes [
   n1 = TriggerDialog!() -> (then, Index:int, IsPrimary:bool=~) {is=title K2Node_FunctionEntry}
   > Set IsPrimary(IsPrimary:bool=@n1.IsPrimary) -> (Output_Get:bool)
   n2 = "Get (a copy)"(
    Array=Select(
     Option 0:Array of Texts=Get SecondaryStringPool() -> (SecondaryStringPool:Array of Texts)
     Option 1:Array of Texts=Get PrimaryStringPool()
     Index=Get IsPrimary()
    ) -> (ReturnValue:Array of Texts)
    Dimension 1=Get Index() -> (Index:int=0)
   )
   > n3 = Branch(Condition=NOT Boolean(A=TextIsEmpty(InText=@n2)))
   > SetText(self=Get Text(), InText=@n2)
   > DialogSpeechCycle() {is=title K2Node_CallFunction}
   SkipLine(execute=@n3.else) {is=title K2Node_CallFunction}
  ]}
]
variables [
 DelayToNextLoop {type=number, default=30}
 DelayToStart {type=number, default=1}
 FinishedCycle {}
 Flip Arrow Horz {type=boolean}
 Flip Speech Bubble Vert {type=boolean}
 IsPrimary {type=boolean}
 PostSpeechBubbleTimer {type=TimerHandle}
 PreSpeechBubbleTimer {type=TimerHandle}
 PrimaryStringPool {type="array<string>", default []}
 SecondaryStringPool {type="array<string>", default []}
 SingleLoop {type=TimerHandle}
 SlotIsSpeaking {}
 SpeechBubbleActive {type=TimerHandle}
 SpeechDuration {type=number, default=8}
 WrapTextAt {type=number, default=400}
]
widgets {
 info {inheritedWidgetCount=0, parentClass=&CommonActivatableWidget, rootWidgetClass=&GridPanel, widgetCount=7}
 widgets [
  Content {
   widgetClassPath=&GridPanel
   bIsVariable=true
   widget=Content
   properties {columnFill [1], renderTransformPivot=<0.95 1.5>, rowFill [1]}
  }
  Overlay_1 {widgetClassPath=&Overlay, parent=Content, widget=Overlay_1, slot {class=GridSlot}}
  Backplate {
   widgetClassPath=&Image
   bIsVariable=true
   parent=Overlay_1
   widget=Backplate
   properties {
    visibility=HitTestInvisible
    brush {
     bIsDynamicallyLoaded=false
     drawAs=Image
     imageType=FullColor
     mirroring=NoMirror
     resourceName=None
     resourceObject=/Safflower_SeasonData/UI/Materials/M_UI_Safflower_SpeechBubble
     tiling=NoTile
     imageSize=<8 8>
     margin=<ltrb 0 0 0 0>
     outlineSettings {
      bUseBrushTransparency=false
      roundingType=HalfHeightRadius
      width=0
      color {colorUseRule=UseColor_Specified, specifiedColor=<rgba 0 0 0 0>}
      cornerRadii=<0 0 0 0>
     }
     tintColor {colorUseRule=UseColor_Specified, specifiedColor=<rgba 1 1 1 1>}
     uVRegion {bIsValid=false, max=<0 0>, min=<0 0>}
    }
    colorAndOpacity=<rgba 0.0021 0.0144 0.0262 0.65>
   }
   slot {class=OverlaySlot, properties {horizontalAlignment=HAlign_Fill, verticalAlignment=VAlign_Fill}}
  }
  ArrowContainerDisabled {
   widgetClassPath=&Overlay
   bIsVariable=true
   parent=Overlay_1
   widget=ArrowContainerDisabled
   properties {visibility=Collapsed}
   slot {class=OverlaySlot, properties {horizontalAlignment=HAlign_Fill, verticalAlignment=VAlign_Fill, padding=<ltrb 50 0 50 2.5>}}
  }
  ArrowVAttachment {
   widgetClassPath=&FortVisualAttachment
   bIsVariable=true
   parent=ArrowContainerDisabled
   widget=ArrowVAttachment
   properties {visibility=Collapsed, contentAnchor=<0.5 0>}
   slot {class=OverlaySlot, properties {horizontalAlignment=HAlign_Fill, verticalAlignment=VAlign_Bottom}}
  }
  Image_Arrow {
   widgetClassPath=&Image
   bIsVariable=true
   parent=ArrowVAttachment
   widget=Image_Arrow
   properties {
    brush {
     bIsDynamicallyLoaded=false
     drawAs=Image
     imageType=FullColor
     mirroring=NoMirror
     resourceName=None
     resourceObject=/Safflower_SeasonData/UI/Materials/M_UI_Safflower_SpeechBubbleArrow
     tiling=NoTile
     imageSize=<48 40>
     margin=<ltrb 0 0 0 0>
     outlineSettings {
      bUseBrushTransparency=false
      roundingType=HalfHeightRadius
      width=0
      color {colorUseRule=UseColor_Specified, specifiedColor=<rgba 0 0 0 0>}
      cornerRadii=<0 0 0 0>
     }
     tintColor {colorUseRule=UseColor_Specified, specifiedColor=<rgba 1 1 1 1>}
     uVRegion {bIsValid=false, max=<0 0>, min=<0 0>}
    }
    colorAndOpacity=<rgba 0.0021 0.0144 0.0262 0.65>
    renderTransform {angle=180, scale=<1 1>, shear=<0 0>, translation=<0 0>}
   }
   slot {class=SizeBoxSlot, properties {horizontalAlignment=HAlign_Center, verticalAlignment=VAlign_Center}}
  }
  Text {
   widgetClassPath=&CommonRichTextBlock
   bIsVariable=true
   parent=Content
   widget=Text
   properties {
    bOverrideDefaultStyle=true
    mobileTextBlockScale=1.3
    text="[PH] Test Text [PH] Test Text [PH] Test Text [PH] Test Text [PH] Test Text [PH] Test Text [PH] Test Text [PH] Test Text [PH] Test Text "
    wrapTextAt=600
    decoratorClasses [/Game/Athena/UI/Frontend/CosmeticGating/Data/RichTextDecorator_CosmeticGate.RichTextDecorator_CosmeticGate_C]
    defaultTextStyleOverride {
     overflowPolicy=Clip
     transformPolicy=None
     colorAndOpacity {colorUseRule=UseColor_Specified, specifiedColor=<rgba 1 1 1 1>}
     font {
      bForceMonospaced=false
      bMaterialIsStencil=false
      fontMaterial=null
      fontObject=&HeadingNow
      letterSpacing=10
      monospacedWidth=1
      size=14.25
      skewAmount=0
      typefaceFontName=74 Regular
      outlineSettings {
       bApplyOutlineToDropShadows=false
       bMiteredCorners=false
       bSeparateFillAlpha=false
       outlineMaterial=null
       outlineSize=0
       outlineColor=<rgba 0 0 0 1>
      }
     }
     highlightColor {colorUseRule=UseColor_Specified, specifiedColor=<rgba 0 0 0 1>}
     highlightShape {
      bIsDynamicallyLoaded=false
      drawAs=Image
      imageType=NoImage
      mirroring=NoMirror
      resourceName=None
      resourceObject=null
      tiling=NoTile
      imageSize=<32 32>
      margin=<ltrb 0 0 0 0>
      outlineSettings {
       bUseBrushTransparency=false
       roundingType=HalfHeightRadius
       width=0
       color {colorUseRule=UseColor_Specified, specifiedColor=<rgba 0 0 0 0>}
       cornerRadii=<0 0 0 0>
      }
      tintColor {colorUseRule=UseColor_Specified, specifiedColor=<rgba 1 1 1 1>}
      uVRegion {bIsValid=false, max=<0 0>, min=<0 0>}
     }
     selectedBackgroundColor {colorUseRule=UseColor_Foreground, specifiedColor=<rgba 1 0 1 1>}
     shadowColorAndOpacity=<rgba 0 0 0 1>
     shadowOffset=<0 0>
     strikeBrush {
      bIsDynamicallyLoaded=false
      drawAs=Image
      imageType=NoImage
      mirroring=NoMirror
      resourceName=None
      resourceObject=null
      tiling=NoTile
      imageSize=<32 32>
      margin=<ltrb 0 0 0 0>
      outlineSettings {
       bUseBrushTransparency=false
       roundingType=HalfHeightRadius
       width=0
       color {colorUseRule=UseColor_Specified, specifiedColor=<rgba 0 0 0 0>}
       cornerRadii=<0 0 0 0>
      }
      tintColor {colorUseRule=UseColor_Specified, specifiedColor=<rgba 1 1 1 1>}
      uVRegion {bIsValid=false, max=<0 0>, min=<0 0>}
     }
     underlineBrush {
      bIsDynamicallyLoaded=false
      drawAs=Image
      imageType=NoImage
      mirroring=NoMirror
      resourceName=None
      resourceObject=null
      tiling=NoTile
      imageSize=<32 32>
      margin=<ltrb 0 0 0 0>
      outlineSettings {
       bUseBrushTransparency=false
       roundingType=HalfHeightRadius
       width=0
       color {colorUseRule=UseColor_Specified, specifiedColor=<rgba 0 0 0 0>}
       cornerRadii=<0 0 0 0>
      }
      tintColor {colorUseRule=UseColor_Specified, specifiedColor=<rgba 1 1 1 1>}
      uVRegion {bIsValid=false, max=<0 0>, min=<0 0>}
     }
    }
   }
   slot {class=GridSlot, properties {verticalAlignment=VAlign_Center, padding=<ltrb 18 18 18 18>}}
  }
 ]
}
