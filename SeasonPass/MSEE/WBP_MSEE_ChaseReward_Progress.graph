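# Base widget for the Season Pass MSEE chase reward progress display, showing a row of reward slots around a named content slot.
# Initialize gathers slot children from a given panel into RewardSlots.
# Theme data supplies an image table that resolves each slot's empty and full images, falling back to an error icon, then OnThemeDataLoaded fires.
# Each slot's state comes from comparing its index with the wallet view model's second currency.
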WidgetBlueprint /SeasonPass/MSEE/WBP_MSEE_ChaseReward_Progress
graphs [
 EventGraph {nodes [OnThemeDataLoaded!() -> (OutputDelegate, then, ThemeData:Season Pass Theme Data Structure) {is=title K2Node_CustomEvent}]}
 FakeSlotDataPreview {nodes [
   n1 = FakeSlotDataPreview!() -> (then, Slot_Empty:Fort Image Reference Structure, Slot_Full:Fort Image Reference Structure) {is=title K2Node_FunctionEntry}
   > n2 = For Each Loop(Array:Array of WBP MSEE Chase Reward Slot Object References=Get RewardSlots() -> (RewardSlots:Array of WBP MSEE Chase Reward Slot Object References)) -> (Array Element:WBP MSEE Chase Reward Slot Object Reference)
   > GetImages!(
    execute
    self:WBP MSEE Chase Reward Slot Object Reference=@n2
    Slot_Empty:Fort Image Reference Structure=@n1
    Slot_Full:Fort Image Reference Structure=@n1.Slot_Full
    ActiveScalarName:Name=None
   ) -> (then)
  ]}
 GetSlotImageData {nodes [
   GetSlotImageData!() -> (then, ImageTableRef:"Fort Image Reference Table Structure (by ref)") {is=title K2Node_FunctionEntry}
   > n1 = For Each Loop(Array=Get RewardSlots())
   n2 = Get ImageTableRef() -> (ImageTableRef:"Fort Image Reference Table Structure (by ref)")
   n3 = TryLookupImage!(
    Table:"Fort Image Reference Table Structure (by ref)"=@n2
    Name:String=Get EmptyImageRef!(self:WBP MSEE Chase Reward Slot Object Reference=@n1) -> (EmptyImageRef:String)
   ) -> (OutImageReference:Fort Image Reference Structure, ReturnValue:bool)
   n4 = Make Fort Image Reference!(
    Image:Soft Object Reference=/FortUILibrary/Textures/MaterialInstances/MI_UIKit_Icon_ImageError
    ImageUrl:String
    MaterialTextureParamName:Name=None
   ) -> (FortImageReference:Fort Image Reference Structure) {is=title K2Node_MakeStruct}
   > Set SlotEmpty(SlotEmpty:Fort Image Reference Structure=Select(Option 0:Fort Image Reference Structure=@n4, Option 1:Fort Image Reference Structure=@n3, Index=@n3.ReturnValue) -> (ReturnValue:Fort Image Reference Structure)) -> (Output_Get:Fort Image Reference Structure)
   n5 = TryLookupImage(Table=@n2, Name=Get FullImageRef!(self:WBP MSEE Chase Reward Slot Object Reference=@n1) -> (FullImageRef:String))
   > Set SlotFull(SlotFull:Fort Image Reference Structure=Select(Option 0=@n4, Option 1=@n5, Index=@n5.ReturnValue)) -> (Output_Get:Fort Image Reference Structure)
   > GetImages(
    self=@n1
    Slot_Empty=Get SlotEmpty() -> (SlotEmpty:Fort Image Reference Structure)
    Slot_Full=Get SlotFull() -> (SlotFull:Fort Image Reference Structure)
    ActiveScalarName=Get ActiveScalarName!(self:WBP MSEE Chase Reward Slot Object Reference=@n1) -> (ActiveScalarName:Name=None)
   )
   n6 = int + int(A=@n1.Array Index)
   n7 = Get SecondCurrentCurrency!(self:MSEEWallet VM Object Reference=Get MSEEWalletVM() -> (MSEEWalletVM:MSEEWallet VM Object Reference)) -> (SecondCurrentCurrency:int=0)
   > SetSlotState!(
    execute
    self:WBP MSEE Chase Reward Slot Object Reference=@n1
    SlotState:E_MSEE_ChaseSlot_State Enum=Select(
     Option 0:E_MSEE_ChaseSlot_State Enum=NewEnumerator0
     Option 1:E_MSEE_ChaseSlot_State Enum=NewEnumerator1
     Index="integer <= integer"(A=@n6, B=@n7)
    ) -> (ReturnValue:E_MSEE_ChaseSlot_State Enum)
   ) -> (then)
  ]}
 Initialize {nodes [
   n1 = Initialize!() -> (then, ContentGrid:Panel Widget Object Reference) {is=title K2Node_FunctionEntry}
   > Clear(TargetArray:Array of WBP MSEE Chase Reward Slot Object References=Get RewardSlots())
   > n2 = For Each Loop(Array:Array of Widget Object References=GetAllChildren(self=@n1)) -> (Array Element:Widget Object Reference)
   > n3 = Cast To WBP_MSEE_ChaseReward_Slot!(execute, Object=@n2) -> (then, CastFailed, AsWBP MSEE Chase Reward Slot:WBP MSEE Chase Reward Slot Object Reference) {is=title K2Node_DynamicCast}
   > Add(
    TargetArray:Array of WBP MSEE Chase Reward Slot Object References=Get RewardSlots()
    NewItem:"WBP MSEE Chase Reward Slot Object Reference (by ref)"=@n3
   ) {is=title K2Node_CallArrayFunction}
  ]}
 IsScreenActive {nodes [IsScreenActive!() -> (then, IsActive:bool=~) {is=title K2Node_FunctionEntry}]}
 SetImage {nodes [
   n1 = SetImage!() -> (
    then
    ImageObject:Fort Lazy Image Object Reference
    LookupName:String
    ImageTable:"Fort Image Reference Table Structure (by ref)"
    Image Size:Vector 2D Structure
   ) {is=title K2Node_FunctionEntry}
   n2 = TryLookupImage(Table=@n1.ImageTable, Name=@n1.LookupName)
   > SetBrushFromImageReference!(
    execute
    self:Fort Lazy Image Object Reference=@n1
    ImageReference:"Fort Image Reference Structure (by ref)"=Select(
     Option 0:"Fort Image Reference Structure (by ref)"=Make Fort Image Reference(Image=&MI_UIKit_Icon_ImageError, MaterialTextureParamName=None)
     Option 1:"Fort Image Reference Structure (by ref)"=@n2
     Index=@n2.ReturnValue
    ) -> (ReturnValue:"Fort Image Reference Structure (by ref)")
    bMatchSize:bool=false
    DownloadPriority:ECMSDownloadPriority Enum=Normal
   ) -> (then)
  ]}
 VM_GetThemeData {nodes [
   n1 = VM_GetThemeData!() -> (then, ThemeData:Season Pass Theme Data Structure) {is=title K2Node_FunctionEntry}
   > GetSlotImageData!(
    execute
    self
    ImageTableRef:"Fort Image Reference Table Structure (by ref)"=Break Season Pass Theme Data!(SeasonPassThemeData:"Season Pass Theme Data Structure (by ref)"=@n1) -> (
     TitleColor:Linear Color Structure
     TitleMaterial:Soft Object Reference
     TitleOutline:Font Outline Settings Structure
     SeasonColor:Linear Color Structure
     ProgressionColor:Linear Color Structure
     OverlayColor:Linear Color Structure
     VaultBackgroundMaterial:Soft Object Reference
     GiftboxVaultBackgroundMaterial:Soft Object Reference
     UIBackgroundMaterial:Soft Object Reference
     UpsellBackgroundMaterial:Soft Object Reference
     CrewCategoryBackgroundMaterial:Soft Object Reference
     CustomLogo:Soft Object Reference
     CustomSeasonTitle:Text
     CustomSeasonTitleColor:Linear Color Structure
     Upsell2BackgrounMaterial:Soft Object Reference
     ImageReferenceTable:Fort Image Reference Table Structure
     ColorReferenceTable:Fort Color Reference Table Structure
     TextReferenceTable:Fort Text Reference Table Structure
    ) {is=title K2Node_BreakStruct}.ImageReferenceTable
   ) -> (then) {is=title K2Node_CallFunction}
   > OnThemeDataLoaded!(execute, self, ThemeData:Season Pass Theme Data Structure=Get ThemeData() -> (ThemeData:Season Pass Theme Data Structure)) -> (then) {is=title K2Node_CallFunction}
  ]}
]
variables [
 ContentGrid {type=PanelWidget, default=null}
 ImageRefs {type="array<S_MSEE_ImageLookup>", default []}
 RewardSlots {type="array<WBP_MSEE_ChaseReward_Slot_C>", default []}
]
widgets {
 info {inheritedWidgetCount=0, parentClass=&UserWidget, rootWidgetClass=&Overlay, widgetCount=2}
 widgets [
  Overlay_0 {widgetClassPath=&Overlay, widget=Overlay_0}
  Main_Contents {widgetClassPath=&NamedSlot, bIsVariable=true, parent=Overlay_0, widget=Main_Contents, slot {class=OverlaySlot}}
 ]
}
