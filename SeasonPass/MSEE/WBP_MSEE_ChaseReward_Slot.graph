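# Season pass MSEE chase reward slot widget that shows an empty or full image between background and foreground named slots.
# GetImages, despite its name, stores both image references and a material scalar name, then shows the empty image.
# SetSlotState records the state and lazy-loads its image at ImageSize, with a design-time preview in PreConstruct.
# SetFullActionState toggles that scalar on the image's dynamic material, only when the slot is not empty.
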WidgetBlueprint /SeasonPass/MSEE/WBP_MSEE_ChaseReward_Slot
graphs [
 EventGraph {nodes [
   n1 = Event PreConstruct()
   > Branch(Condition=@n1.IsDesignTime)
   > SetSlotState!(execute, self, SlotState:E_MSEE_ChaseSlot_State Enum=Get PreviewSlotState() -> (PreviewSlotState:E_MSEE_ChaseSlot_State Enum=NewEnumerator0)) -> (then)
  ]}
 GetImages {nodes [
   n1 = GetImages!() -> (then, Slot_Empty:Fort Image Reference Structure, Slot_Full:Fort Image Reference Structure, ActiveScalarName:Name) {is=title K2Node_FunctionEntry}
   > Set Slot Empty(Slot Empty:Fort Image Reference Structure=@n1) -> (Output_Get:Fort Image Reference Structure)
   > Set Slot Full(Slot Full:Fort Image Reference Structure=@n1.Slot_Full) -> (Output_Get:Fort Image Reference Structure)
   > Set ActiveScalarName(ActiveScalarName:Name=@n1.ActiveScalarName) -> (Output_Get:Name=None)
   > SetImage!(execute, self, ImageReference:"Fort Image Reference Structure (by ref)"=Get Slot_Empty() -> (Slot_Empty:Fort Image Reference Structure)) -> (then)
  ]}
 SetFullActionState {nodes [
   SetFullActionState!() -> (then, IsFullStateActive:bool=~) {is=title K2Node_FunctionEntry}
   > Branch(Condition="Not Equal (Enum)"!(
     A:E_MSEE_ChaseSlot_State Enum=Get ActiveState() -> (ActiveState:E_MSEE_ChaseSlot_State Enum=NewEnumerator0)
     B:E_MSEE_ChaseSlot_State Enum=NewEnumerator0
    ) -> (ReturnValue:bool))
   > n1 = GetDynamicMaterial(self=Get LazyImage() -> (LazyImage:Fort Lazy Image Object Reference))
   > SetScalarParameterValue(
    self=@n1
    ParameterName=Get ActiveScalarName() -> (ActiveScalarName:Name=None)
    Value=SelectFloat(A=1, B=0, bPickA=Get IsFullStateActive() -> (IsFullStateActive:bool))
   )
  ]}
 SetImage {nodes [
   SetImage!() -> (then, ImageReference:"Fort Image Reference Structure (by ref)") {is=title K2Node_FunctionEntry}
   n1 = Get LazyImage()
   n2 = Get!(self:Image Object Reference=@n1) -> (
    Brush_TintColor:Slate Color Structure
    Brush_DrawAs:ESlateBrushDrawType Enum
    Brush_Tiling:ESlateBrushTileType Enum
    Brush_Mirroring:ESlateBrushMirrorType Enum
    Brush_ImageSize:Vector 2D Structure
    Brush_Margin:Margin Structure
    Brush_ResourceObject:Object Reference
    Brush_OutlineSettings:Slate Brush Outline Settings Structure
   )
   > SetBrush!(
    execute
    self=@n1
    InBrush_TintColor:Slate Color Structure=@n2
    InBrush_DrawAs:ESlateBrushDrawType Enum=@n2.Brush_DrawAs
    InBrush_Tiling:ESlateBrushTileType Enum=@n2.Brush_Tiling
    InBrush_Mirroring:ESlateBrushMirrorType Enum=@n2.Brush_Mirroring
    InBrush_ImageSize:Vector 2D Structure=Get ImageSize() -> (ImageSize:Vector 2D Structure)
    InBrush_Margin:Margin Structure=@n2.Brush_Margin
    InBrush_ResourceObject:Object Reference=@n2.Brush_ResourceObject
    InBrush_OutlineSettings:Slate Brush Outline Settings Structure=@n2.Brush_OutlineSettings
   ) -> (then)
   > SetBrushFromImageReference!(
    execute
    self:Fort Lazy Image Object Reference=@n1
    ImageReference:"Fort Image Reference Structure (by ref)"=Get ImageReference() -> (ImageReference:"Fort Image Reference Structure (by ref)")
    bMatchSize:bool=false
    DownloadPriority:ECMSDownloadPriority Enum=Normal
   ) -> (then)
  ]}
 SetSlotState {nodes [
   n1 = SetSlotState!() -> (then, SlotState:E_MSEE_ChaseSlot_State Enum=NewEnumerator0) {is=title K2Node_FunctionEntry}
   > SetImage(ImageReference=Select!(
     NewEnumerator0:Fort Image Reference Structure=Get Slot Empty() -> (Slot Empty:Fort Image Reference Structure)
     NewEnumerator1:Fort Image Reference Structure=Get Slot Full() -> (Slot Full:Fort Image Reference Structure)
     Index:E_MSEE_ChaseSlot_State Enum=@n1
    ) -> (ReturnValue:Fort Image Reference Structure)) {is=title K2Node_CallFunction}
   > Set ActiveState(ActiveState:E_MSEE_ChaseSlot_State Enum=Get SlotState() -> (SlotState:E_MSEE_ChaseSlot_State Enum=NewEnumerator0)) -> (Output_Get:E_MSEE_ChaseSlot_State Enum=NewEnumerator0)
  ]}
]
variables [
 ActiveScalarName {type=string, default=None}
 ActiveState {type=E_MSEE_ChaseSlot_State, default=Empty}
 EmptyImageRef {type=string, default=""}
 FullImageRef {type=string, default=""}
 ImageSize {type=Vector2D, default=<64 64>}
 PreviewSlotState {type=E_MSEE_ChaseSlot_State, default=Empty}
 Slot Empty {type=FortImageReference, default {image=null, imageUrl="", materialTextureParamName=None}}
 Slot Full {type=FortImageReference, default {image=null, imageUrl="", materialTextureParamName=None}}
]
widgets {
 info {inheritedWidgetCount=0, parentClass=&UserWidget, rootWidgetClass=&Overlay, widgetCount=4}
 widgets [
  Overlay_0 {widgetClassPath=&Overlay, widget=Overlay_0}
  BG_Slot {
   widgetClassPath=&NamedSlot
   bIsVariable=true
   parent=Overlay_0
   widget=BG_Slot
   slot {class=OverlaySlot, properties {horizontalAlignment=HAlign_Fill, verticalAlignment=VAlign_Fill}}
  }
  LazyImage {
   widgetClassPath=&FortLazyImage
   bIsVariable=true
   parent=Overlay_0
   widget=LazyImage
   properties {brush {
     bIsDynamicallyLoaded=false
     drawAs=Image
     imageType=FullColor
     mirroring=NoMirror
     resourceName=None
     resourceObject=/FortUILibrary/Textures/MaterialInstances/MI_UIKit_Icon_Idle
     tiling=NoTile
     imageSize=<64 64>
     margin=<ltrb 0 0 0 0>
     outlineSettings {
      bUseBrushTransparency=false
      roundingType=HalfHeightRadius
      width=0
      color {colorUseRule=UseColor_Specified, specifiedColor=<rgba 0 0 0 0>}
      cornerRadii=<0 0 0 0>
     }
     tintColor {colorUseRule=UseColor_Specified, specifiedColor=<rgba 1 1 1 1>}
     uVRegion {bIsValid=false, max=<0 0>, min=<0 0>}
    }}
   slot {class=OverlaySlot}
  }
  FG_Slot {
   widgetClassPath=&NamedSlot
   bIsVariable=true
   parent=Overlay_0
   widget=FG_Slot
   slot {class=OverlaySlot, properties {horizontalAlignment=HAlign_Fill, verticalAlignment=VAlign_Fill}}
  }
 ]
}
