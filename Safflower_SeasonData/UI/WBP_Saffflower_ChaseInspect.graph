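# Safflower season pass chase reward progress widget: six slots on a vine, each with a speech bubble.
# When the screen is active and second currency exceeds 1, it fills bubbles from theme text and starts a timer.
# RunSpeechCycle picks random primary and secondary speakers from the leading run of active slots, sharing indices with bubbles.
# Bubble events drive slot talking states and decrement a pending count that restarts the cycle at zero.
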WidgetBlueprint /Safflower_SeasonData/UI/WBP_Saffflower_ChaseInspect
graphs [
 AllCSlots {nodes [
   Inputs!()
   Outputs!(Array:Array of WBP MSEE Chase Reward Slot Object References=Make Array!(
     "[0]":WBP MSEE Chase Reward Slot Object Reference=Get CSlot_1() -> (CSlot_1:WBP MSEE Chase Reward Slot Object Reference)
     "[1]":WBP MSEE Chase Reward Slot Object Reference=Get CSlot_2() -> (CSlot_2:WBP MSEE Chase Reward Slot Object Reference)
     "[2]":WBP MSEE Chase Reward Slot Object Reference=Get CSlot_3() -> (CSlot_3:WBP MSEE Chase Reward Slot Object Reference)
     "[3]":WBP MSEE Chase Reward Slot Object Reference=Get CSlot_4() -> (CSlot_4:WBP MSEE Chase Reward Slot Object Reference)
     "[4]":WBP MSEE Chase Reward Slot Object Reference=Get CSlot_5() -> (CSlot_5:WBP MSEE Chase Reward Slot Object Reference)
     "[5]":WBP MSEE Chase Reward Slot Object Reference=Get CSlot_6() -> (CSlot_6:WBP MSEE Chase Reward Slot Object Reference)
    ) -> (Array:Array of WBP MSEE Chase Reward Slot Object References))
  ]}
 AllSpeechBubbles {nodes [
   Inputs()
   Outputs(Array:Array of WBP Speech Bubble Object References=Make Array(
     "[0]":WBP Speech Bubble Object Reference=Get Speech_1() -> (Speech_1:WBP Speech Bubble Object Reference)
     "[1]":WBP Speech Bubble Object Reference=Get Speech_2() -> (Speech_2:WBP Speech Bubble Object Reference)
     "[2]":WBP Speech Bubble Object Reference=Get Speech_3() -> (Speech_3:WBP Speech Bubble Object Reference)
     "[3]":WBP Speech Bubble Object Reference=Get Speech_4() -> (Speech_4:WBP Speech Bubble Object Reference)
     "[4]":WBP Speech Bubble Object Reference=Get Speech_5() -> (Speech_5:WBP Speech Bubble Object Reference)
     "[5]":WBP Speech Bubble Object Reference=Get Speech_6() -> (Speech_6:WBP Speech Bubble Object Reference)
    ) -> (Array:Array of WBP Speech Bubble Object References))
  ]}
 EventGraph {nodes [
   n1 = "Slot Is Speaking (Speech_2)"!() -> (OutputDelegate, then, IsSpeaking:bool) {is=title K2Node_ComponentBoundEvent}
   > SetFullActionState!(execute, self:WBP MSEE Chase Reward Slot Object Reference=Get CSlot_2(), IsFullStateActive:bool=@n1.IsSpeaking) -> (then)
   n2 = "Finished Cycle (Speech_6)"() {is=title K2Node_ComponentBoundEvent}
   Branch(
    execute=[@n2.then, @n5.then, @n6.then, @n9.then, @n10.then, @n13.then]
    Condition="integer > integer"(A=Get PendingCycles() -> (PendingCycles:int=0))
   )
   > n3 = Decrement Int(Value=Get PendingCycles() -> (PendingCycles=0))
   > Branch(Condition="integer <= integer"(A=@n3))
   > RunSpeechCycle()
   n4 = "Slot Is Speaking (Speech_6)"!() -> (OutputDelegate, then, IsSpeaking:bool) {is=title K2Node_ComponentBoundEvent}
   > SetFullActionState(self=Get CSlot_6(), IsFullStateActive=@n4.IsSpeaking)
   n5 = "Finished Cycle (Speech_1)"() {is=title K2Node_ComponentBoundEvent}
   n6 = "Finished Cycle (Speech_2)"() {is=title K2Node_ComponentBoundEvent}
   n7 = "Slot Is Speaking (Speech_1)"!() -> (OutputDelegate, then, IsSpeaking:bool) {is=title K2Node_ComponentBoundEvent}
   > SetFullActionState(self=Get CSlot_1(), IsFullStateActive=@n7.IsSpeaking)
   n8 = "Slot Is Speaking (Speech_3)"!() -> (OutputDelegate, then, IsSpeaking:bool) {is=title K2Node_ComponentBoundEvent}
   > SetFullActionState(self=Get CSlot_3(), IsFullStateActive=@n8.IsSpeaking)
   n9 = "Finished Cycle (Speech_3)"() {is=title K2Node_ComponentBoundEvent}
   n10 = "Finished Cycle (Speech_4)"() {is=title K2Node_ComponentBoundEvent}
   n11 = "Slot Is Speaking (Speech_4)"!() -> (OutputDelegate, then, IsSpeaking:bool) {is=title K2Node_ComponentBoundEvent}
   > SetFullActionState(self=Get CSlot_4(), IsFullStateActive=@n11.IsSpeaking)
   n12 = "Slot Is Speaking (Speech_5)"!() -> (OutputDelegate, then, IsSpeaking:bool) {is=title K2Node_ComponentBoundEvent}
   > SetFullActionState(self=Get CSlot_5(), IsFullStateActive=@n12.IsSpeaking)
   n13 = "Finished Cycle (Speech_5)"() {is=title K2Node_ComponentBoundEvent}
   n14 = SpeechCheck() {is=title K2Node_CustomEvent}
   Branch(execute=[@n20.then, @n14.then], Condition=Get AllowSpeech() -> (AllowSpeech:bool))
   > SetupSpeechTexts!(execute, self, Theme Data:Season Pass Theme Data Structure=Get Theme Data() -> (Theme Data:Season Pass Theme Data Structure)) -> (then)
   > GetAllActiveCSlots()
   > Set PendingCycles(PendingCycles:int=0) -> (Output_Get:int=0)
   > n15 = Set Timer by Function Name!(
    execute
    Object:Object Reference
    FunctionName:String=RunSpeechCycle
    Time:float=4
    bLooping:bool=false
    bMaxOncePerFrame:bool=false
    InitialStartDelay:float=0
    InitialStartDelayVariance:float=0
   ) -> (then, ReturnValue:Timer Handle Structure) {is=title K2Node_CallFunction}
   > Set StartDialogLoop(StartDialogLoop:Timer Handle Structure=@n15) -> (Output_Get:Timer Handle Structure)
   n16 = Event PreConstruct()
   > Parent: PreConstruct(IsDesignTime=@n16.IsDesignTime)
   > Initialize!(
    execute
    self:WBP MSEE Chase Reward Progress Object Reference
    ContentGrid:Panel Widget Object Reference=Get Slots_Grid() -> (Slots_Grid:Grid Panel Object Reference)
   ) -> (then)
   > Branch(Condition=@n16.IsDesignTime)
   > n17 = For Each Loop(Array:Array of WBP MSEE Chase Reward Slot Object References=Get RewardSlots() -> (RewardSlots:Array of WBP MSEE Chase Reward Slot Object References)) -> (Array Element:WBP MSEE Chase Reward Slot Object Reference)
   > GetImages!(
    execute
    self:WBP MSEE Chase Reward Slot Object Reference=@n17
    Slot_Empty:Fort Image Reference Structure=Get Slot Empty!(self:WBP MSEE Chase Reward Slot Object Reference=@n17) -> (Slot Empty:Fort Image Reference Structure)
    Slot_Full:Fort Image Reference Structure=Get Slot Full!(self:WBP MSEE Chase Reward Slot Object Reference=@n17) -> (Slot Full:Fort Image Reference Structure)
    ActiveScalarName:Name=Get ActiveScalarName!(self:WBP MSEE Chase Reward Slot Object Reference=@n17) -> (ActiveScalarName:Name=None)
   ) -> (then)
   > SetSlotState!(
    execute
    self:WBP MSEE Chase Reward Slot Object Reference=@n17
    SlotState:E_MSEE_ChaseSlot_State Enum=Get PreviewSlotState!(self:WBP MSEE Chase Reward Slot Object Reference=@n17) -> (PreviewSlotState:E_MSEE_ChaseSlot_State Enum=NewEnumerator0)
   ) -> (then)
   n18 = Event OnThemeDataLoaded!() -> (OutputDelegate, then, ThemeData:Season Pass Theme Data Structure) {is=title K2Node_Event}
   > n19 = Set Theme Data(Theme Data:Season Pass Theme Data Structure=@n18.ThemeData) -> (Output_Get:Season Pass Theme Data Structure)
   > n20 = Parent: OnThemeDataLoaded!(execute, ThemeData:Season Pass Theme Data Structure=@n19) -> (then) {is=title K2Node_CallParentFunction}
  ]}
 GetAllActiveCSlots {nodes [
   GetAllActiveCSlots() {is=title K2Node_FunctionEntry}
   > Clear(TargetArray:Array of WBP MSEE Chase Reward Slot Object References=Get ActiveCSlots() -> (ActiveCSlots:Array of WBP MSEE Chase Reward Slot Object References))
   > n1 = For Each Loop with Break!(
    Exec:Exec
    Array:Array of WBP MSEE Chase Reward Slot Object References=All CSlots!() -> (Array:Array of WBP MSEE Chase Reward Slot Object References) {is=title K2Node_MacroInstance}
    Break:Exec=@n2.else
   ) -> (LoopBody:Exec, Array Element:WBP MSEE Chase Reward Slot Object Reference, Array Index:int, Completed:Exec)
   > n2 = Branch(Condition="Equal (Enum)"(
     A:E_MSEE_ChaseSlot_State Enum=Get ActiveState!(self:WBP MSEE Chase Reward Slot Object Reference=@n1) -> (ActiveState:E_MSEE_ChaseSlot_State Enum=NewEnumerator0)
     B:E_MSEE_ChaseSlot_State Enum=NewEnumerator1
    ))
   > Add(
    TargetArray:Array of WBP MSEE Chase Reward Slot Object References=Get ActiveCSlots()
    NewItem:"WBP MSEE Chase Reward Slot Object Reference (by ref)"=@n1
   ) {is=title K2Node_CallArrayFunction}
  ]}
 GetRandomFromSpeechRange {nodes [
   GetRandomFromSpeechRange() {is=title K2Node_FunctionEntry}
   n1 = Get SecondarySpeechStringRefs() -> (SecondarySpeechStringRefs:Array of Strings)
   n2 = Get PrimarySpeechStringRefs() -> (PrimarySpeechStringRefs:Array of Strings)
   > Return Node!(
    execute
    Primary Speech:int=RandomInteger!(Max:int=Length(TargetArray:Array of Strings=@n2)) -> (ReturnValue:int=0)
    Secondary Speech:int=RandomInteger(Max=Length(TargetArray=@n1)) -> (ReturnValue=0)
   )
  ]}
 IsScreenActive {nodes [
   n1 = IsScreenActive!() -> (then, IsActive:bool) {is=title K2Node_FunctionEntry}
   > Parent: IsScreenActive!(execute, IsActive:bool=@n1) -> (then) {is=title K2Node_CallParentFunction}
   n2 = Get SecondCurrentCurrency!(self:MSEEWallet VM Object Reference=Get MSEEWalletVM() -> (MSEEWalletVM:MSEEWallet VM Object Reference)) -> (SecondCurrentCurrency:int=0)
   > Set AllowSpeech(AllowSpeech:bool=AND Boolean(A=Get IsActive() -> (IsActive:bool), B="integer > integer"(A=@n2, B=1))) -> (Output_Get:bool)
   > n3 = Branch(Condition=Get IsActive())
   SpeechCheck(execute=[@n4.then, @n3.then]) {is=title K2Node_CallFunction}
   StopSpeech(
    execute=@n3.else
    self:WBP Speech Bubble Object Reference=[Get Speech_1(), Get Speech_4(), Get Speech_5(), Get Speech_6(), Get Speech_2(), Get Speech_3()]
   )
   > n4 = Clear and Invalidate Timer by Handle(Handle=Get StartDialogLoop() -> (StartDialogLoop:Timer Handle Structure))
  ]}
 RunSpeechCycle {nodes [
   RunSpeechCycle() {is=title K2Node_FunctionEntry}
   n1 = Get ActiveCSlots()
   > Set RandomPrimarySpeaker(RandomPrimarySpeaker:int=RandomInteger(Max=Length(TargetArray:Array of WBP MSEE Chase Reward Slot Object References=@n1)) -> (ReturnValue=0)) -> (Output_Get:int=0)
   n2 = "Get (a copy)"(Array:Array of WBP MSEE Chase Reward Slot Object References=@n1, Dimension 1=Get RandomPrimarySpeaker() -> (RandomPrimarySpeaker:int=0)) -> (Output:WBP MSEE Chase Reward Slot Object Reference)
   > Branch(Condition="Equal (Enum)"(A=Get ActiveState(self=@n2) -> (ActiveState=NewEnumerator0), B=NewEnumerator1))
   > n3 = Set PendingCycles(PendingCycles=1) -> (Output_Get=0)
   TriggerDialog!(
    execute=[@n13."  ", @n3.then]
    self:WBP Speech Bubble Object Reference="Get (a copy)"(
     Array:Array of WBP Speech Bubble Object References=All Speech Bubbles!() -> (Array:Array of WBP Speech Bubble Object References) {is=title K2Node_MacroInstance}
     Dimension 1=Get RandomPrimarySpeaker() -> (RandomPrimarySpeaker=0)
    ) -> (Output:WBP Speech Bubble Object Reference)
    Index:int=GetRandomFromSpeechRange!() -> (Primary Speech:int=0, Secondary Speech:int=0) {is=title K2Node_CallFunction}
    IsPrimary:bool=true
   ) -> (then)
   n4 = Get ActiveCSlots()
   n5 = RandomInteger(Max=Length(TargetArray=@n4)) -> (ReturnValue=0)
   n6 = Last Index!(TargetArray:Array of WBP MSEE Chase Reward Slot Object References=@n4) -> (ReturnValue=0)
   n7 = Get RandomPrimarySpeaker() -> (RandomPrimarySpeaker=0)
   n8 = int + int(A=@n7)
   > Set RandomSecondarySpeaker(RandomSecondarySpeaker:int=SelectInt(A="Wrap (Integer)"!(Value:int=@n8, Min:int=0, Max:int=@n6) -> (ReturnValue:int=0), B=@n5, bPickA="Equal (Integer)"(A=@n7, B=@n5))) -> (Output_Get:int=0)
   n9 = Get ActiveCSlots()
   n10 = Get RandomSecondarySpeaker() -> (RandomSecondarySpeaker:int=0)
   > Branch(Condition=Is Valid Index!(TargetArray:Array of WBP MSEE Chase Reward Slot Object References=@n9, IndexToTest:int=@n10) -> (ReturnValue:bool))
   n11 = "Get (a copy)"(Array:Array of WBP MSEE Chase Reward Slot Object References=@n9, Dimension 1=@n10) -> (Output:WBP MSEE Chase Reward Slot Object Reference)
   > Branch(Condition="Equal (Enum)"(A=Get ActiveState(self=@n11) -> (ActiveState=NewEnumerator0), B=NewEnumerator1))
   > n12 = Set PendingCycles(PendingCycles=2) -> (Output_Get=0)
   TriggerDialog(
    execute=[@n14."  ", @n12.then]
    self="Get (a copy)"(
     Array:Array of WBP Speech Bubble Object References=All Speech Bubbles()
     Dimension 1=Get RandomSecondarySpeaker() -> (RandomSecondarySpeaker=0)
    ) -> (Output:WBP Speech Bubble Object Reference)
    Index=GetRandomFromSpeechRange() -> (Primary Speech=0, Secondary Speech=0).Secondary Speech
    IsPrimary=false
   )
   n13 = Increment Int(Value=Get PendingCycles() -> (PendingCycles=0))
   n14 = Increment Int(Value=Get PendingCycles() -> (PendingCycles=0))
  ]}
 SetupSpeechTexts {nodes [
   n1 = SetupSpeechTexts!() -> (then, Theme Data:Season Pass Theme Data Structure) {is=title K2Node_FunctionEntry}
   > n2 = For Each Loop(Array:Array of Strings=Get PrimarySpeechStringRefs()) -> (Array Element:String)
   n3 = For Each Loop(Exec=@n2.Completed, Array=Get SecondarySpeechStringRefs())
   SetSpeechTexts!(
    execute=@n3.Completed
    self:WBP Speech Bubble Object Reference=[Get Speech_3(), Get Speech_2(), Get Speech_5(), Get Speech_6(), Get Speech_1(), Get Speech_4()]
    Primary Strings:Array of Texts=Get PrimarySpeechTexts() -> (PrimarySpeechTexts:Array of Texts)
    Secondary Strings:Array of Texts=Get SecondarySpeechTexts() -> (SecondarySpeechTexts:Array of Texts)
   ) -> (then)
   n4 = Break Season Pass Theme Data!(SeasonPassThemeData:"Season Pass Theme Data Structure (by ref)"=@n1) -> (
    TitleColor:Linear Color Structure
    TitleMaterial:Soft Object Reference
    TitleOutline:Font Outline Settings Structure
    SeasonColor:Linear Color Structure
    ProgressionColor:Linear Color Structure
    OverlayColor:Linear Color Structure
    VaultBackgroundMaterial:Soft Object Reference
    GiftboxVaultBackgroundMaterial:Soft Object Reference
    UIBackgroundMaterial:Soft Object Reference
    UpsellBackgroundMaterial:Soft Object Reference
    CrewCategoryBackgroundMaterial:Soft Object Reference
    CustomLogo:Soft Object Reference
    CustomSeasonTitle:Text
    CustomSeasonTitleColor:Linear Color Structure
    Upsell2BackgrounMaterial:Soft Object Reference
    ImageReferenceTable:Fort Image Reference Table Structure
    ColorReferenceTable:Fort Color Reference Table Structure
    TextReferenceTable:Fort Text Reference Table Structure
   ) {is=title K2Node_BreakStruct}
   n5 = TryLookupText!(Table:"Fort Text Reference Table Structure (by ref)"=@n4.TextReferenceTable, Name:String=@n3) -> (OutTextReference:Text, ReturnValue:bool)
   Add(
    execute=@n3.LoopBody
    TargetArray:Array of Texts=Get SecondarySpeechTexts()
    NewItem:"Text (by ref)"=SelectText(A=@n5, bPickA=@n5.ReturnValue)
   )
   n6 = TryLookupText(Table=@n4.TextReferenceTable, Name=@n2)
   Add(execute=@n2.LoopBody, TargetArray=Get PrimarySpeechTexts(), NewItem=SelectText(A=@n6, bPickA=@n6.ReturnValue))
  ]}
]
variables [
 ActiveCSlots {type="array<WBP_MSEE_ChaseReward_Slot_C>", default []}
 AllowSpeech {type=boolean}
 PendingCycles {type=integer}
 PreviewSlotEmpty {
  type=FortImageReference
  default {image=/Safflower_SeasonData/UI/MaterialInstances/MI_UI_ChaseReward_00, imageUrl="", materialTextureParamName=None}
 }
 PreviewSlotFull {
  type=FortImageReference
  default {image=/Safflower_SeasonData/UI/MaterialInstances/MI_UI_ChaseReward_01, imageUrl="", materialTextureParamName=None}
 }
 PreviewSlotFullActive {
  type=FortImageReference
  default {image=/FortUILibrary/Textures/MaterialInstances/MI_UIKit_Icon_Notification_ON, imageUrl="", materialTextureParamName=None}
 }
 PrimaryFinished {type=boolean}
 PrimarySpeechStringRefs {
  type="array<string>"
  default [
   Speech_Primary_1
   Speech_Primary_2
   Speech_Primary_3
   Speech_Primary_4
   Speech_Primary_5
   Speech_Primary_6
   Speech_Primary_7
   Speech_Primary_8
   Speech_Primary_9
   Speech_Primary_10
   Speech_Primary_11
   Speech_Primary_12
   Speech_Primary_13
   Speech_Primary_14
   Speech_Primary_15
   Speech_Primary_16
  ]
 }
 SecondaryFinished {type=boolean}
 SecondarySpeechStringRefs {
  type="array<string>"
  default [Speech_Secondary_1, Speech_Secondary_2, Speech_Secondary_3, Speech_Secondary_4, Speech_Secondary_5, Speech_Secondary_6]
 }
 StartDialogLoop {type=TimerHandle}
 Theme Data {
  type=SeasonPassThemeData
  default {
   crewCategoryBackgroundMaterial=null
   customLogo=null
   customSeasonTitle=""
   giftboxVaultBackgroundMaterial=null
   primaryAssetId=""
   titleMaterial=null
   uIBackgroundMaterial=null
   upsell2BackgrounMaterial=null
   upsellBackgroundMaterial=null
   vaultBackgroundMaterial=null
   benefitsByType []
   colorReferenceTable {namedColors []}
   customSeasonTitleColor=<rgba 1 1 1 1>
   imageReferenceTable {namedImages []}
   overlayColor=<rgba 0 0 0 1>
   progressionColor=<rgba 1 1 1 1>
   seasonColor=<rgba 1 1 1 1>
   textReferenceTable {namedTexts []}
   titleColor=<rgba 1 1 1 1>
   titleOutline {
    bApplyOutlineToDropShadows=false
    bMiteredCorners=false
    bSeparateFillAlpha=false
    outlineMaterial=null
    outlineSize=0
    outlineColor=<rgba 0 0 0 1>
   }
  }
 }
]
widgets {
 info {
  inheritedWidgetCount=0
  namedSlotCount=7
  parentClass=/SeasonPass/MSEE/WBP_MSEE_ChaseReward_Progress.WBP_MSEE_ChaseReward_Progress_C
  rootWidgetClass=null
  widgetCount=25
 }
 widgets [
  Overlay_1 {widgetClassPath=&Overlay, widget=Overlay_1}
  Image_42 {
   widgetClassPath=&Image
   bIsVariable=true
   parent=Overlay_1
   widget=Image_42
   properties {brush {
     bIsDynamicallyLoaded=false
     drawAs=Image
     imageType=FullColor
     mirroring=NoMirror
     resourceName=None
     resourceObject=/Safflower_SeasonData/UI/Textures/T_UI_Safflower_VineStraight
     tiling=NoTile
     imageSize=<256 32>
     margin=<ltrb 0 0 0 0>
     outlineSettings {
      bUseBrushTransparency=false
      roundingType=HalfHeightRadius
      width=0
      color {colorUseRule=UseColor_Specified, specifiedColor=<rgba 0 0 0 0>}
      cornerRadii=<0 0 0 0>
     }
     tintColor {colorUseRule=UseColor_Specified, specifiedColor=<rgba 1 1 1 1>}
     uVRegion {bIsValid=false, max=<0 0>, min=<0 0>}
    }}
   slot {class=OverlaySlot, properties {horizontalAlignment=HAlign_Fill, verticalAlignment=VAlign_Fill, padding=<ltrb 24 36 24 36>}}
  }
  Slots_Grid {widgetClassPath=&GridPanel, bIsVariable=true, parent=Overlay_1, widget=Slots_Grid, slot {class=OverlaySlot}}
  CSlot_1 {
   widgetClassPath=/SeasonPass/MSEE/WBP_MSEE_ChaseReward_Slot
   bIsVariable=true
   parent=Slots_Grid
   widget=CSlot_1
   properties {
    activeScalarName=IsTalking
    emptyImageRef=ChaseSlot_Empty
    fullImageRef=ChaseSlot_Full_1
    imageSize=<72 72>
    slot Empty {image=&MI_UI_ChaseReward_00, imageUrl="", materialTextureParamName=None}
    slot Full {image=&MI_UI_ChaseReward_01, imageUrl="", materialTextureParamName=None}
   }
   slot {class=GridSlot, properties {horizontalAlignment=HAlign_Left, verticalAlignment=VAlign_Top, padding=<ltrb 10 20 10 10>}}
  }
  FortVisualAttachment_0 {
   widgetClassPath=&FortVisualAttachment
   namedSlotHost=CSlot_1
   widget=FortVisualAttachment_0
   properties {contentAnchor=<1 1>}
   slot=~
  }
  Speech_1 {
   widgetClassPath=&/WBP_SpeechBubble
   bIsVariable=true
   parent=FortVisualAttachment_0
   widget=Speech_1
   properties {delayToNextLoop=10, delayToStart=0, speechDuration=4, renderTransform {angle=0, scale=<1 1>, shear=<0 0>, translation=<44 -29>}}
   slot {class=SizeBoxSlot, properties {horizontalAlignment=HAlign_Right, verticalAlignment=VAlign_Top}}
  }
  CSlot_2 {
   widgetClassPath=&WBP_MSEE_ChaseReward_Slot
   bIsVariable=true
   parent=Slots_Grid
   widget=CSlot_2
   properties {
    activeScalarName=IsTalking
    emptyImageRef=ChaseSlot_Empty
    fullImageRef=ChaseSlot_Full_2
    imageSize=<72 72>
    slot Empty {image=&MI_UI_ChaseReward_00, imageUrl="", materialTextureParamName=None}
    slot Full {image=/Safflower_SeasonData/UI/MaterialInstances/MI_UI_ChaseReward_02, imageUrl="", materialTextureParamName=None}
   }
   slot {class=GridSlot, properties {column=1, horizontalAlignment=HAlign_Left, verticalAlignment=VAlign_Top, padding=<ltrb 10 4 10 10>}}
  }
  Overlay_96 {widgetClassPath=&Overlay, namedSlotHost=CSlot_2, widget=Overlay_96}
  FortVisualAttachment_1 {
   widgetClassPath=&FortVisualAttachment
   parent=Overlay_96
   widget=FortVisualAttachment_1
   properties {contentAnchor=<1 1>}
   slot {class=OverlaySlot}
  }
  Speech_2 {
   widgetClassPath=&WBP_SpeechBubble
   bIsVariable=true
   parent=FortVisualAttachment_1
   widget=Speech_2
   properties {delayToNextLoop=10, delayToStart=0, speechDuration=4, renderTransform {angle=0, scale=<1 1>, shear=<0 0>, translation=<42 -27>}}
   slot {class=SizeBoxSlot, properties {horizontalAlignment=HAlign_Right}}
  }
  CSlot_3 {
   widgetClassPath=&WBP_MSEE_ChaseReward_Slot
   bIsVariable=true
   parent=Slots_Grid
   widget=CSlot_3
   properties {
    activeScalarName=IsTalking
    emptyImageRef=ChaseSlot_Empty
    fullImageRef=ChaseSlot_Full_3
    imageSize=<72 72>
    slot Empty {image=&MI_UI_ChaseReward_00, imageUrl="", materialTextureParamName=None}
    slot Full {image=/Safflower_SeasonData/UI/MaterialInstances/MI_UI_ChaseReward_03, imageUrl="", materialTextureParamName=None}
   }
   slot {class=GridSlot, properties {column=2, horizontalAlignment=HAlign_Left, verticalAlignment=VAlign_Top, padding=<ltrb 10 10 10 10>}}
  }
  Overlay_378 {widgetClassPath=&Overlay, namedSlotHost=CSlot_3, widget=Overlay_378}
  FortVisualAttachment {
   widgetClassPath=&FortVisualAttachment
   parent=Overlay_378
   widget=FortVisualAttachment
   properties {contentAnchor=<1 1>}
   slot {class=OverlaySlot, properties {horizontalAlignment=HAlign_Center}}
  }
  Speech_3 {
   widgetClassPath=&WBP_SpeechBubble
   bIsVariable=true
   parent=FortVisualAttachment
   widget=Speech_3
   properties {delayToNextLoop=10, delayToStart=0, speechDuration=4, renderTransform {angle=0, scale=<1 1>, shear=<0 0>, translation=<3 -35>}}
   slot {class=SizeBoxSlot}
  }
  CSlot_4 {
   widgetClassPath=&WBP_MSEE_ChaseReward_Slot
   bIsVariable=true
   parent=Slots_Grid
   widget=CSlot_4
   properties {
    activeScalarName=IsTalking
    emptyImageRef=ChaseSlot_Empty
    fullImageRef=ChaseSlot_Full_4
    imageSize=<72 72>
    slot Empty {image=&MI_UI_ChaseReward_00, imageUrl="", materialTextureParamName=None}
    slot Full {image=/Safflower_SeasonData/UI/MaterialInstances/MI_UI_ChaseReward_04, imageUrl="", materialTextureParamName=None}
   }
   slot {class=GridSlot, properties {column=3, horizontalAlignment=HAlign_Left, verticalAlignment=VAlign_Top, padding=<ltrb 10 24 10 10>}}
  }
  Overlay_280 {widgetClassPath=&Overlay, namedSlotHost=CSlot_4, widget=Overlay_280}
  FortVisualAttachment_2 {
   widgetClassPath=&FortVisualAttachment
   parent=Overlay_280
   widget=FortVisualAttachment_2
   properties {contentAnchor=<0 1>, renderTransform {angle=0, scale=<1 1>, shear=<0 0>, translation=<0 -62>}}
   slot {class=OverlaySlot, properties {horizontalAlignment=HAlign_Center, verticalAlignment=VAlign_Center}}
  }
  Speech_4 {
   widgetClassPath=&WBP_SpeechBubble
   bIsVariable=true
   parent=FortVisualAttachment_2
   widget=Speech_4
   properties {
    delayToNextLoop=10
    delayToStart=0
    flip Arrow Horz=true
    speechDuration=4
    renderTransform {angle=0, scale=<1 1>, shear=<0 0>, translation=<-12 -5>}
   }
   slot {class=SizeBoxSlot, properties {horizontalAlignment=HAlign_Left, verticalAlignment=VAlign_Top}}
  }
  CSlot_5 {
   widgetClassPath=&WBP_MSEE_ChaseReward_Slot
   bIsVariable=true
   parent=Slots_Grid
   widget=CSlot_5
   properties {
    activeScalarName=IsTalking
    emptyImageRef=ChaseSlot_Empty
    fullImageRef=ChaseSlot_Full_5
    imageSize=<72 72>
    slot Empty {image=&MI_UI_ChaseReward_00, imageUrl="", materialTextureParamName=None}
    slot Full {image=/Safflower_SeasonData/UI/MaterialInstances/MI_UI_ChaseReward_05, imageUrl="", materialTextureParamName=None}
   }
   slot {class=GridSlot, properties {column=4, horizontalAlignment=HAlign_Left, verticalAlignment=VAlign_Top, padding=<ltrb 10 8 10 10>}}
  }
  FortVisualAttachment_3 {
   widgetClassPath=&FortVisualAttachment
   namedSlotHost=CSlot_5
   widget=FortVisualAttachment_3
   properties {contentAnchor=<0 1>}
   slot=~
  }
  Speech_5 {
   widgetClassPath=&WBP_SpeechBubble
   bIsVariable=true
   parent=FortVisualAttachment_3
   widget=Speech_5
   properties {
    delayToNextLoop=10
    delayToStart=0
    flip Arrow Horz=true
    speechDuration=4
    renderTransform {angle=0, scale=<1 1>, shear=<0 0>, translation=<24 -36>}
   }
   slot {class=SizeBoxSlot, properties {horizontalAlignment=HAlign_Left, verticalAlignment=VAlign_Top}}
  }
  CSlot_6 {
   widgetClassPath=&WBP_MSEE_ChaseReward_Slot
   bIsVariable=true
   parent=Slots_Grid
   widget=CSlot_6
   properties {
    activeScalarName=IsTalking
    emptyImageRef=ChaseSlot_Empty
    fullImageRef=ChaseSlot_Full_6
    imageSize=<72 72>
    slot Empty {image=&MI_UI_ChaseReward_00, imageUrl="", materialTextureParamName=None}
    slot Full {image=/Safflower_SeasonData/UI/MaterialInstances/MI_UI_ChaseReward_06, imageUrl="", materialTextureParamName=None}
   }
   slot {class=GridSlot, properties {column=5, horizontalAlignment=HAlign_Left, verticalAlignment=VAlign_Top, padding=<ltrb 10 16 10 10>}}
  }
  Overlay_181 {widgetClassPath=&Overlay, namedSlotHost=CSlot_6, widget=Overlay_181}
  FortVisualAttachment_4 {
   widgetClassPath=&FortVisualAttachment
   parent=Overlay_181
   widget=FortVisualAttachment_4
   properties {contentAnchor=<0 1>}
   slot {class=OverlaySlot, properties {horizontalAlignment=HAlign_Center}}
  }
  Speech_6 {
   widgetClassPath=&WBP_SpeechBubble
   bIsVariable=true
   parent=FortVisualAttachment_4
   widget=Speech_6
   properties {
    delayToNextLoop=10
    delayToStart=0
    flip Arrow Horz=true
    speechDuration=4
    renderTransform {angle=0, scale=<1 1>, shear=<0 0>, translation=<0 -25>}
   }
   slot {class=SizeBoxSlot, properties {horizontalAlignment=HAlign_Left, verticalAlignment=VAlign_Top}}
  }
 ]
}
