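# Safflower season pass inspect screen, a child of WBP_MSEE_InspectScreen that supplies the season's chase reward widget and chevron arrows.
# PreConstruct registers the chase widget with the parent, then SetMobileStyle enlarges and spreads both arrows for compact UI, building their brush from the left arrow's.
# Two intro events play half-second animations, normal or from the buy menu, then notify the parent on finish.
# The animations drive cloud and ground values in a background Material Parameter Collection rather than widget properties.
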WidgetBlueprint /Safflower_SeasonData/UI/WBP_Safflower_InspectScreen
animations [
 Intro_Inspect {
  binding_count=0
  length_seconds=0.5
  object=&self:Intro_Inspect
  sequence {
   bindings []
   tracks [{
     type=MovieSceneMaterialParameterCollectionTrack
     display_name=MPC_Safflower_BackgroundStates
     sections [{
       type=MovieSceneParameterSection
       channels [
        Anim_CloudHeight {type=MovieSceneScriptingFloatChannel, keys [[0,0],[8,1]]}
        Anim_HideGround {type=MovieSceneScriptingFloatChannel, keys [[0,0],[10,1]]}
        Anim_GroundHeight {type=MovieSceneScriptingFloatChannel, keys [[0,0],[10,1]]}
       ]
      }]
    }]
  }
 }
 Intro_FromBuyMenu {
  binding_count=0
  length_seconds=0.5
  object=&self:Intro_FromBuyMenu
  sequence {
   bindings []
   tracks [{
     type=MovieSceneMaterialParameterCollectionTrack
     display_name=MPC_Safflower_BackgroundStates
     sections [{
       type=MovieSceneParameterSection
       channels [
        Anim_CloudHeight {type=MovieSceneScriptingFloatChannel, const [1,2]}
        Anim_HideGround {type=MovieSceneScriptingFloatChannel, const [1,2]}
        Anim_GroundHeight {type=MovieSceneScriptingFloatChannel, keys [[0,0],[10,1]]}
       ]
      }]
    }]
  }
 }
]
graphs [
 EventGraph {nodes [
   n1 = Event PreConstruct()
   > SetChaseInspectWidget!(
    execute
    self:WBP MSEE Inspect Screen Object Reference
    Chase Widget:WBP MSEE Chase Reward Progress Object Reference=Get WBP_Saffflower_ChaseInspect() -> (WBP_Saffflower_ChaseInspect:WBP Saffflower Chase Inspect Object Reference)
   ) -> (then)
   > Parent: PreConstruct(IsDesignTime=@n1.IsDesignTime)
   > SetMobileStyle() {is=title K2Node_CallFunction}
   Event PlayIntroAnim() {is=title K2Node_Event}
   > n2 = Play Animation with Finished event!(
    execute:Exec
    Widget:User Widget Object Reference=Self-Reference()
    InAnimation:Widget Animation Object Reference=Get Intro_Inspect() -> (Intro_Inspect:Widget Animation Object Reference)
    StartAtTime:float=0
    NumLoopsToPlay:int=1
    PlayMode:EUMGSequencePlayMode Enum
    PlaybackSpeed:float=1
   ) -> (then:Exec, Result:Widget Animation Handle Structure, Finished:Exec) {is=title K2Node_PlayAnimation2}
   IntroAnimFinished(execute=@n2.Finished, self:WBP MSEE Inspect Screen Object Reference)
   Event FromBuyMenuIntro() {is=title K2Node_Event}
   > n3 = Play Animation with Finished event(
    Widget=Self-Reference()
    InAnimation=Get Intro_FromBuyMenu() -> (Intro_FromBuyMenu:Widget Animation Object Reference)
    StartAtTime=0
    NumLoopsToPlay=1
    PlaybackSpeed=1
   )
   FromBuyIntroFinished(execute=@n3.Finished, self:WBP MSEE Inspect Screen Object Reference)
  ]}
 SetMobileStyle {nodes [
   SetMobileStyle() {is=title K2Node_FunctionEntry}
   > Branch(Condition=IsUsingCompactUI())
   n1 = Get LightningArrow_L() -> (LightningArrow_L:Image Object Reference)
   n2 = Get!(self:Image Object Reference=@n1) -> (
    Brush_TintColor:Slate Color Structure
    Brush_DrawAs:ESlateBrushDrawType Enum
    Brush_Tiling:ESlateBrushTileType Enum
    Brush_Mirroring:ESlateBrushMirrorType Enum
    Brush_ImageSize:Vector 2D Structure
    Brush_Margin:Margin Structure
    Brush_ResourceObject:Object Reference
    Brush_OutlineSettings:Slate Brush Outline Settings Structure
   )
   > SetBrush(
    self=[Get LightningArrow_R() -> (LightningArrow_R:Image Object Reference), @n1]
    InBrush=Make Slate Brush!(
     TintColor:Slate Color Structure=@n2
     DrawAs:ESlateBrushDrawType Enum=@n2.Brush_DrawAs
     Tiling:ESlateBrushTileType Enum=@n2.Brush_Tiling
     Mirroring:ESlateBrushMirrorType Enum=@n2.Brush_Mirroring
     ImageSize_X:double=152
     ImageSize_Y:double=184
     Margin:Margin Structure=@n2.Brush_Margin
     ResourceObject:Object Reference=@n2.Brush_ResourceObject
     OutlineSettings:Slate Brush Outline Settings Structure=@n2.Brush_OutlineSettings
    ) -> (SlateBrush:Slate Brush Structure) {is=title K2Node_MakeStruct}
   )
   > SetRenderTranslation(self=Get LightningArrow_L(), Translation=(X=-40,Y=0))
   > SetRenderTranslation(self=Get LightningArrow_R(), Translation=(X=40,Y=0))
  ]}
]
variables []
widgets {
 info {
  inheritedWidgetCount=0
  namedSlotCount=3
  parentClass=/SeasonPass/MSEE/WBP_MSEE_InspectScreen.WBP_MSEE_InspectScreen_C
  rootWidgetClass=null
  widgetCount=3
 }
 widgets [
  WBP_Saffflower_ChaseInspect {widgetClassPath=&/WBP_Saffflower_ChaseInspect, bIsVariable=true, widget=WBP_Saffflower_ChaseInspect}
  LightningArrow_L {
   widgetClassPath=&Image
   bIsVariable=true
   widget=LightningArrow_L
   properties {
    brush {
     bIsDynamicallyLoaded=false
     drawAs=Image
     imageType=FullColor
     mirroring=NoMirror
     resourceName=None
     resourceObject=/Safflower_SeasonData/UI/Materials/M_UI_Safflower_Chevrons
     tiling=NoTile
     imageSize=<100 120>
     margin=<ltrb 0 0 0 0>
     outlineSettings {
      bUseBrushTransparency=false
      roundingType=HalfHeightRadius
      width=0
      color {colorUseRule=UseColor_Specified, specifiedColor=<rgba 0 0 0 0>}
      cornerRadii=<0 0 0 0>
     }
     tintColor {colorUseRule=UseColor_Specified, specifiedColor=<rgba 1 1 1 1>}
     uVRegion {bIsValid=false, max=<0 0>, min=<0 0>}
    }
    renderTransform {angle=0, scale=<-1 1>, shear=<0 0>, translation=<-20 0>}
   }
  }
  LightningArrow_R {
   widgetClassPath=&Image
   bIsVariable=true
   widget=LightningArrow_R
   properties {
    brush {
     bIsDynamicallyLoaded=false
     drawAs=Image
     imageType=FullColor
     mirroring=NoMirror
     resourceName=None
     resourceObject=&M_UI_Safflower_Chevrons
     tiling=NoTile
     imageSize=<100 120>
     margin=<ltrb 0 0 0 0>
     outlineSettings {
      bUseBrushTransparency=false
      roundingType=HalfHeightRadius
      width=0
      color {colorUseRule=UseColor_Specified, specifiedColor=<rgba 0 0 0 0>}
      cornerRadii=<0 0 0 0>
     }
     tintColor {colorUseRule=UseColor_Specified, specifiedColor=<rgba 1 1 1 1>}
     uVRegion {bIsValid=false, max=<0 0>, min=<0 0>}
    }
    renderTransform {angle=0, scale=<1 1>, shear=<0 0>, translation=<20 0>}
   }
  }
 ]
}
