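# Translucent UI material for the Safflower season that draws an animated blue-to-cyan glow, named for the top navigation.
# A radial GlowMask fades out from the UV center.
# CloudGlow sums two upward- or downward-panning blurred noise layers, one in scaled viewport space, on a shared CloudTime clock.
# An aspect-corrected, panning DistortionTexture parameter adds further noise, and named reroutes carry these parts into the color blend and saturated mask math.
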Material /Safflower_SeasonData/UI/Materials/M_UI_Safflower_TopNavGlow
graph {
 properties {
  bAutomaticallySetUsageInEditor=false
  bUsedWithStaticMesh=false
  blendMode=BLEND_Translucent
  materialDomain=MD_UI
  usedShadingModels=Default Lit
 }
 nodes [
  LinearInterpolate(
   A=Constant3Vector() {constant=<rgba 0.0037 0.5884 1 1>}
   B=Constant3Vector() {constant=<rgba 0 1 0.9018 1>}
   Alpha=SmoothStep(Value=NamedRerouteUsage()) {constMax=0.8, constMin=0.35}
  )
  n1 = TextureCoordinate()
  n2 = Multiply(A=Distance(A=ComponentMask(None=@n1) {g=true, r=true}, B=Constant() {r=0.5})) {constB=2}
  NamedRerouteDeclaration(None=SmoothStep(Value=OneMinus(None=@n2)) {constMax=0.9}) {name=GlowMask, nodeColor=<rgba 1 0 0.4922 1>}
  n3 = Time()
  n4 = Constant()
  NamedRerouteDeclaration(None=Multiply(A=Add(A=@n3, B=@n4), B=Constant() {r=1.5})) {name=CloudTime, nodeColor=<rgba 0 0.8125 1 1>}
  n5 = NamedRerouteUsage()
  n6 = MaterialFunctionCall(UVs=ScreenPosition().ViewportUV, Texture Scale=Constant() {r=0.25}) {materialFunction=&ScaleUVsByCenter}
  n7 = ComponentMask(None=@n6.UVs) {r=true}
  n8 = NamedRerouteUsage()
  n9 = Panner(Coordinate=AppendVector(A=Multiply(A=@n7) {constB=2}, B=ComponentMask(None=@n6.UVs) {g=true}), Time=@n8) {speedY=0.04}
  n10 = Panner(Time=@n8) {speedY=0.04}
  n11 = Add(
   A=Power(Base=TextureSample(UVs=@n9) {samplerType=SAMPLERTYPE_Grayscale, texture=&LowResBlurredNoise}.RGB)
   B=Power(Base=TextureSample(UVs=@n10) {samplerType=SAMPLERTYPE_Grayscale, texture=&LowResBlurredNoise}.RGB)
  )
  NamedRerouteDeclaration(None=Multiply(A=SmoothStep(Value=@n5) {constMax=0.45}, B=Saturate(None=@n11))) {name=CloudGlow, nodeColor=<rgba 0 1 0.2266 1>}
  n12 = MaterialFunctionCall() {materialFunction=&GetUserInterfaceUV}
  n13 = Multiply(
   A=Divide(A=ComponentMask(None=@n12.Pixel Size) {g=true}, B=ComponentMask(None=@n12.Pixel Size) {r=true})
   B=ComponentMask(None=@n12.Normalized UV) {g=true}
  )
  n14 = ComponentMask(None=@n12.Normalized UV) {r=true}
  n15 = Constant() {r=1}
  n16 = Constant() {r=0.6}
  n17 = Add(A=AppendVector(A=Constant() {r=1}, B=Constant() {r=5}), B=Multiply(A=AppendVector(A=@n14, B=@n13), B=AppendVector(A=@n15, B=@n16)))
  n18 = AppendVector(A=Constant(), B=Constant() {r=0.2})
  NamedRerouteDeclaration(None=TextureSampleParameter2D(UVs=Panner(Coordinate=@n17, Speed=@n18)) {parameterName=DistortionTexture, texture=&T_NoisePack_01}.RGB) {name=DistortionTexture, nodeColor=<rgba 0 1 0.1328 1>}
  n19 = Constant() {r=1.5}
  n20 = NamedRerouteUsage()
  n21 = SmoothStep() {constMax=0.8, constMin=0.2}
  n22 = Multiply(A=Saturate(None=NamedRerouteUsage()))
  Saturate(None=Multiply(A=Multiply(A=@n19, B=@n20), B=Add(A=@n21, B=@n22)))
 ]
}
