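# Safflower season top-nav tab button, derived from the MSEE season pass custom button, showing an icon over glow and cloud-texture layers.
# PreConstruct adjusts button style and sizing for mobile.
# Hover scales the icon only while unselected; selection also fades both backing layers, with faster mobile animation variants.
# SetSelectedVisuals additionally sets both layers' brush opacity to match the selected state.
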WidgetBlueprint /Safflower_SeasonData/UI/WBP_Safflower_TopNav_CustomTab
animations [
 OnHover {
  binding_count=3
  length_seconds=0.1
  object=&self:OnHover
  sequence {bindings [
    Image_Safflower_TopNavIcon {
     id=#1
     possessed_class=Image
     tracks [{
       type=MovieScene2DTransformTrack
       sections [{
         type=MovieScene2DTransformSection
         end_frame=2
         start_frame=0
         channels [
          Translation.X {type=MovieSceneScriptingFloatChannel, const [0,1]}
          Translation.Y {type=MovieSceneScriptingFloatChannel, const [0,1]}
          Angle {type=MovieSceneScriptingFloatChannel, const [0,1]}
          Scale.X {type=MovieSceneScriptingFloatChannel, keys [[0,0.9],[2,1]]}
          Scale.Y {type=MovieSceneScriptingFloatChannel, keys [[0,0.9],[2,1]]}
          Shear.X {type=MovieSceneScriptingFloatChannel, const [0,1]}
          Shear.Y {type=MovieSceneScriptingFloatChannel, const [0,1]}
         ]
        }]
      }]
    }
    Image_Glow {
     id=#2
     possessed_class=Image
     tracks [{
       type=MovieSceneFloatTrack
       display_name=Render Opacity
       sections [{type=MovieSceneFloatSection, end_frame=0, start_frame=0, channels [None {type=MovieSceneScriptingFloatChannel, const [0,1]}]}]
      }]
    }
    Image_CloudTexture {
     id=#3
     possessed_class=Image
     tracks [{
       type=MovieSceneFloatTrack
       display_name=Render Opacity
       sections [{type=MovieSceneFloatSection, end_frame=0, start_frame=0, channels [None {type=MovieSceneScriptingFloatChannel, const [0,1]}]}]
      }]
    }
   ]}
 }
 OnUnhover {
  binding_count=3
  length_seconds=0.1
  object=&self:OnUnhover
  sequence {bindings [
    Image_Safflower_TopNavIcon {
     id=#4
     possessed_class=Image
     tracks [{
       type=MovieScene2DTransformTrack
       sections [{
         type=MovieScene2DTransformSection
         end_frame=2
         start_frame=0
         channels [
          Translation.X {type=MovieSceneScriptingFloatChannel, const [0,1]}
          Translation.Y {type=MovieSceneScriptingFloatChannel, const [0,1]}
          Angle {type=MovieSceneScriptingFloatChannel, const [0,1]}
          Scale.X {type=MovieSceneScriptingFloatChannel, keys [[0,1],[2,0.9]]}
          Scale.Y {type=MovieSceneScriptingFloatChannel, keys [[0,1],[2,0.9]]}
          Shear.X {type=MovieSceneScriptingFloatChannel, const [0,1]}
          Shear.Y {type=MovieSceneScriptingFloatChannel, const [0,1]}
         ]
        }]
      }]
    }
    Image_CloudTexture {
     id=#5
     possessed_class=Image
     tracks [{
       type=MovieSceneFloatTrack
       display_name=Render Opacity
       sections [{type=MovieSceneFloatSection, end_frame=0, start_frame=0, channels [None {type=MovieSceneScriptingFloatChannel, const [0,1]}]}]
      }]
    }
    Image_Glow {
     id=#6
     possessed_class=Image
     tracks [{
       type=MovieSceneFloatTrack
       display_name=Render Opacity
       sections [{type=MovieSceneFloatSection, end_frame=0, start_frame=0, channels [None {type=MovieSceneScriptingFloatChannel, const [0,1]}]}]
      }]
    }
   ]}
 }
 OnSelected {
  binding_count=3
  length_seconds=0.25
  object=&self:OnSelected
  sequence {bindings [
    Image_Glow {
     id=#7
     possessed_class=Image
     tracks [{
       type=MovieSceneFloatTrack
       display_name=Render Opacity
       sections [{type=MovieSceneFloatSection, end_frame=5, start_frame=0, channels [None {type=MovieSceneScriptingFloatChannel, keys [[0,0],[5,1]]}]}]
      }]
    }
    Image_CloudTexture {
     id=#8
     possessed_class=Image
     tracks [{
       type=MovieSceneFloatTrack
       display_name=Render Opacity
       sections [{type=MovieSceneFloatSection, end_frame=5, start_frame=0, channels [None {type=MovieSceneScriptingFloatChannel, keys [[0,0],[5,1]]}]}]
      }]
    }
    Image_Safflower_TopNavIcon {
     id=#9
     possessed_class=Image
     tracks [{
       type=MovieScene2DTransformTrack
       sections [{
         type=MovieScene2DTransformSection
         end_frame=5
         start_frame=0
         channels [
          Translation.X {type=MovieSceneScriptingFloatChannel, const [0,1]}
          Translation.Y {type=MovieSceneScriptingFloatChannel, const [0,1]}
          Angle {type=MovieSceneScriptingFloatChannel, const [0,1]}
          Scale.X {type=MovieSceneScriptingFloatChannel, keys [[0,0.9],[5,1]]}
          Scale.Y {type=MovieSceneScriptingFloatChannel, keys [[0,0.9],[5,1]]}
          Shear.X {type=MovieSceneScriptingFloatChannel, const [0,1]}
          Shear.Y {type=MovieSceneScriptingFloatChannel, const [0,1]}
         ]
        }]
      }]
    }
   ]}
 }
 OnDeselected {
  binding_count=3
  length_seconds=0.25
  object=&self:OnDeselected
  sequence {bindings [
    Image_Safflower_TopNavIcon {
     id=#10
     possessed_class=Image
     tracks [{
       type=MovieScene2DTransformTrack
       sections [{
         type=MovieScene2DTransformSection
         end_frame=5
         start_frame=0
         channels [
          Translation.X {type=MovieSceneScriptingFloatChannel, const [0,1]}
          Translation.Y {type=MovieSceneScriptingFloatChannel, const [0,1]}
          Angle {type=MovieSceneScriptingFloatChannel, const [0,1]}
          Scale.X {type=MovieSceneScriptingFloatChannel, keys [[0,1],[5,0.9]]}
          Scale.Y {type=MovieSceneScriptingFloatChannel, keys [[0,1],[5,0.9]]}
          Shear.X {type=MovieSceneScriptingFloatChannel, const [0,1]}
          Shear.Y {type=MovieSceneScriptingFloatChannel, const [0,1]}
         ]
        }]
      }]
    }
    Image_Glow {
     id=#11
     possessed_class=Image
     tracks [{
       type=MovieSceneFloatTrack
       display_name=Render Opacity
       sections [{type=MovieSceneFloatSection, end_frame=5, start_frame=0, channels [None {type=MovieSceneScriptingFloatChannel, keys [[0,1],[5,0]]}]}]
      }]
    }
    Image_CloudTexture {
     id=#12
     possessed_class=Image
     tracks [{
       type=MovieSceneFloatTrack
       display_name=Render Opacity
       sections [{type=MovieSceneFloatSection, end_frame=5, start_frame=0, channels [None {type=MovieSceneScriptingFloatChannel, keys [[0,1],[5,0]]}]}]
      }]
    }
   ]}
 }
 OnSelected_Mobile {
  binding_count=3
  length_seconds=0.05
  object=&self:OnSelected_Mobile
  sequence {bindings [
    Image_CloudTexture {
     id=#13
     possessed_class=Image
     tracks [{
       type=MovieSceneFloatTrack
       display_name=Render Opacity
       sections [{type=MovieSceneFloatSection, end_frame=1, start_frame=0, channels [None {type=MovieSceneScriptingFloatChannel, keys [[0,0],[1,1]]}]}]
      }]
    }
    Image_Glow {
     id=#14
     possessed_class=Image
     tracks [{
       type=MovieSceneFloatTrack
       display_name=Render Opacity
       sections [{type=MovieSceneFloatSection, end_frame=1, start_frame=0, channels [None {type=MovieSceneScriptingFloatChannel, keys [[0,0],[1,1]]}]}]
      }]
    }
    Image_Safflower_TopNavIcon {
     id=#15
     possessed_class=Image
     tracks [{
       type=MovieScene2DTransformTrack
       sections [{
         type=MovieScene2DTransformSection
         end_frame=1
         start_frame=0
         channels [
          Translation.X {type=MovieSceneScriptingFloatChannel, const [0,1]}
          Translation.Y {type=MovieSceneScriptingFloatChannel, const [0,1]}
          Angle {type=MovieSceneScriptingFloatChannel, const [0,1]}
          Scale.X {type=MovieSceneScriptingFloatChannel, keys [[0,0.9],[1,1]]}
          Scale.Y {type=MovieSceneScriptingFloatChannel, keys [[0,0.9],[1,1]]}
          Shear.X {type=MovieSceneScriptingFloatChannel, const [0,1]}
          Shear.Y {type=MovieSceneScriptingFloatChannel, const [0,1]}
         ]
        }]
      }]
    }
   ]}
 }
 OnDeselected_Mobile {
  binding_count=3
  length_seconds=0.05
  object=&self:OnDeselected_Mobile
  sequence {bindings [
    Image_Glow {
     id=#16
     possessed_class=Image
     tracks [{
       type=MovieSceneFloatTrack
       display_name=Render Opacity
       sections [{type=MovieSceneFloatSection, end_frame=0, start_frame=-1, channels [None {type=MovieSceneScriptingFloatChannel, keys [[-1,1],[0,0]]}]}]
      }]
    }
    Image_CloudTexture {
     id=#17
     possessed_class=Image
     tracks [{
       type=MovieSceneFloatTrack
       display_name=Render Opacity
       sections [{type=MovieSceneFloatSection, end_frame=0, start_frame=-1, channels [None {type=MovieSceneScriptingFloatChannel, keys [[-1,1],[0,0]]}]}]
      }]
    }
    Image_Safflower_TopNavIcon {
     id=#18
     possessed_class=Image
     tracks [{
       type=MovieScene2DTransformTrack
       sections [{
         type=MovieScene2DTransformSection
         end_frame=0
         start_frame=-1
         channels [
          Translation.X {type=MovieSceneScriptingFloatChannel, const [0,1]}
          Translation.Y {type=MovieSceneScriptingFloatChannel, const [0,1]}
          Angle {type=MovieSceneScriptingFloatChannel, const [0,1]}
          Scale.X {type=MovieSceneScriptingFloatChannel, keys [[-1,1],[0,0.9]]}
          Scale.Y {type=MovieSceneScriptingFloatChannel, keys [[-1,1],[0,0.9]]}
          Shear.X {type=MovieSceneScriptingFloatChannel, const [0,1]}
          Shear.Y {type=MovieSceneScriptingFloatChannel, const [0,1]}
         ]
        }]
      }]
    }
   ]}
 }
]
graphs [
 EventGraph {nodes [
   Event PreConstruct()
   n1 = IsMobileGame()
   > SetStyle!(
    execute
    self:Common Button Base Object Reference
    InStyle:Common Button Style Class Reference=Select(
     Option 0:Common Button Style Class Reference=/FortUILibrary/Foundation/Buttons/ButtonStyles/Tab/ButtonStyle-Tab-Main-S11.ButtonStyle-Tab-Main-S11_C
     Option 1:Common Button Style Class Reference
     Index=@n1
    ) -> (ReturnValue:Common Button Style Class Reference)
   ) -> (then)
   > Branch(Condition=@n1)
   n2 = Get SizeBox_Mobile() -> (SizeBox_Mobile:Size Box Object Reference)
   > SetMinDesiredWidth!(execute, self:Size Box Object Reference=@n2, InMinDesiredWidth:float=120) -> (then)
   > SetMaxDesiredWidth!(execute, self:Size Box Object Reference=@n2, InMaxDesiredWidth:float=190) -> (then)
   > SetHeightOverride(self=@n2, InHeightOverride=75)
   > SetDesiredSizeOverride!(
    execute
    self:Image Object Reference=Get Image_Safflower_TopNavIcon() -> (Image_Safflower_TopNavIcon:Image Object Reference)
    DesiredSize:Vector 2D Structure=Select(Option 0:Vector 2D Structure=(X=44,Y=44), Option 1:Vector 2D Structure=(X=60,Y=60), Index=@n1) -> (ReturnValue:Vector 2D Structure)
   ) -> (then)
   Event On Hovered()
   n3 = NOT Boolean(A=GetSelected!(self:Common Button Base Object Reference) -> (ReturnValue:bool))
   > Branch(Condition=@n3)
   > PlayAnimationForward(InAnimation=Get OnHover() -> (OnHover:Widget Animation Object Reference))
   Event On Unhovered()
   > Branch(Condition=@n3)
   > PlayAnimationForward(InAnimation=Get OnUnhover() -> (OnUnhover:Widget Animation Object Reference))
   Event On Selected()
   > PlayAnimationForward(InAnimation=Select(
     Option 0:Widget Animation Object Reference=Get OnSelected() -> (OnSelected:Widget Animation Object Reference)
     Option 1:Widget Animation Object Reference=Get OnSelected_Mobile() -> (OnSelected_Mobile:Widget Animation Object Reference)
     Index=IsMobileGame()
    ) -> (ReturnValue:Widget Animation Object Reference))
   > SetSelectedVisuals!(execute, self, IsSelected:bool=true) -> (then)
   Event On Deselected()
   > PlayAnimationForward(InAnimation=Select(
     Option 0=Get OnDeselected() -> (OnDeselected:Widget Animation Object Reference)
     Option 1=Get OnDeselected_Mobile() -> (OnDeselected_Mobile:Widget Animation Object Reference)
     Index=IsMobileGame()
    ))
   > SetSelectedVisuals(IsSelected=false)
  ]}
 SetSelectedVisuals {nodes [
   n1 = SetSelectedVisuals!() -> (then, IsSelected:bool=~) {is=title K2Node_FunctionEntry}
   > SetOpacity!(
    execute
    self:Image Object Reference=Get Image_Glow() -> (Image_Glow:Image Object Reference)
    InOpacity:float=Select(Option 0:float=0, Option 1:float=1, Index=@n1) -> (ReturnValue:float)
   ) -> (then)
   > SetOpacity(self=Get Image_CloudTexture() -> (Image_CloudTexture:Image Object Reference), InOpacity=Select(Option 0=0, Option 1=1, Index=@n1))
  ]}
]
variables []
widgets {
 info {
  inheritedWidgetCount=0
  namedSlotCount=1
  parentClass=/SeasonPass/MSEE/WBP_MSEE_TopNavBar_CustomButton.WBP_MSEE_TopNavBar_CustomButton_C
  rootWidgetClass=null
  widgetCount=5
 }
 widgets [
  SizeBox_Mobile {widgetClassPath=&SizeBox, bIsVariable=true, widget=SizeBox_Mobile}
  Overlay_31 {
   widgetClassPath=&Overlay
   parent=SizeBox_Mobile
   widget=Overlay_31
   slot {class=SizeBoxSlot, properties {horizontalAlignment=HAlign_Center, verticalAlignment=VAlign_Center}}
  }
  Image_Glow {
   widgetClassPath=&Image
   bIsVariable=true
   parent=Overlay_31
   widget=Image_Glow
   properties {
    renderOpacity=0
    brush {
     bIsDynamicallyLoaded=false
     drawAs=Image
     imageType=FullColor
     mirroring=NoMirror
     resourceName=None
     resourceObject=/Safflower_SeasonData/UI/MaterialInstances/MI_UI_Safflower_TopBar_Glow
     tiling=NoTile
     imageSize=<46 46>
     margin=<ltrb 0 0 0 0>
     outlineSettings {
      bUseBrushTransparency=false
      roundingType=HalfHeightRadius
      width=0
      color {colorUseRule=UseColor_Specified, specifiedColor=<rgba 0 0 0 0>}
      cornerRadii=<0 0 0 0>
     }
     tintColor {colorUseRule=UseColor_Specified, specifiedColor=<rgba 1 1 1 1>}
     uVRegion {bIsValid=false, max=<0 0>, min=<0 0>}
    }
    colorAndOpacity=<rgba 1 1 1 0.8>
    renderTransform {angle=0, scale=<2 2>, shear=<0 0>, translation=<0 0>}
   }
   slot {class=OverlaySlot, properties {horizontalAlignment=HAlign_Center, verticalAlignment=VAlign_Center}}
  }
  Image_CloudTexture {
   widgetClassPath=&Image
   bIsVariable=true
   parent=Overlay_31
   widget=Image_CloudTexture
   properties {
    renderOpacity=0
    brush {
     bIsDynamicallyLoaded=false
     drawAs=Image
     imageType=FullColor
     mirroring=NoMirror
     resourceName=None
     resourceObject=/Safflower_SeasonData/UI/Materials/M_UI_Safflower_TopNavGlow
     tiling=NoTile
     imageSize=<42 42>
     margin=<ltrb 0 0 0 0>
     outlineSettings {
      bUseBrushTransparency=false
      roundingType=HalfHeightRadius
      width=0
      color {colorUseRule=UseColor_Specified, specifiedColor=<rgba 0 0 0 0>}
      cornerRadii=<0 0 0 0>
     }
     tintColor {colorUseRule=UseColor_Specified, specifiedColor=<rgba 1 1 1 1>}
     uVRegion {bIsValid=false, max=<0 0>, min=<0 0>}
    }
    renderTransform {angle=0, scale=<2 2>, shear=<0 0>, translation=<0 0>}
   }
   slot {class=OverlaySlot, properties {horizontalAlignment=HAlign_Center, verticalAlignment=VAlign_Center}}
  }
  Image_Safflower_TopNavIcon {
   widgetClassPath=&Image
   bIsVariable=true
   parent=Overlay_31
   widget=Image_Safflower_TopNavIcon
   properties {
    brush {
     bIsDynamicallyLoaded=false
     drawAs=Image
     imageType=FullColor
     mirroring=NoMirror
     resourceName=None
     resourceObject=/Safflower_SeasonData/UI/MaterialInstances/MI_UI_Safflower_TopNavIcon
     tiling=NoTile
     imageSize=<50 50>
     margin=<ltrb 0 0 0 0>
     outlineSettings {
      bUseBrushTransparency=false
      roundingType=HalfHeightRadius
      width=0
      color {colorUseRule=UseColor_Specified, specifiedColor=<rgba 0 0 0 0>}
      cornerRadii=<0 0 0 0>
     }
     tintColor {colorUseRule=UseColor_Specified, specifiedColor=<rgba 1 1 1 1>}
     uVRegion {bIsValid=false, max=<0 0>, min=<0 0>}
    }
    renderTransform {angle=0, scale=<0.9 0.9>, shear=<0 0>, translation=<0 0>}
   }
   slot {class=OverlaySlot, properties {horizontalAlignment=HAlign_Fill, verticalAlignment=VAlign_Fill, padding=<ltrb 4 0 4 0>}}
  }
 ]
}
